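# Niagara cosmetic effect for the CurbRestore player character that puts out snow-like GPU sprites for landing and running.
# Two spawners use probability from a system timeline or sprint state, scaled by Reactivity, zero while aiming down sights, with rates rising over MatchTime.
# Particles spawn on the skeletal mesh, then move under drag, gravity and sprint-dependent curl noise.
# Two sprite renderers draw time-of-day-tinted particles, with derived glow attributes; owner-velocity position offsets are disabled.
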NiagaraSystem /BRCosmetics/Characters/Player/Male/Medium/Bodies/M_MED_CurbRestore/FX/NS_Character_CurbRestore
dependencies {
 usedDataInterfaces [&NiagaraDataInterfaceCurve, &NiagaraDataInterfaceSkeletalMesh]
 usedDynamicInputs [
  &VectorLength
  &RemapRange
  &Boolean_AndOperation
  &SetBoolByFloatComparison
  &FNTimeOfDayColor
 ]
 usedModules [
  &self:SystemSpawnScript.SystemScriptSource.SystemScriptGraph.NiagaraNodeAssignment_3.SetVariables_#1
  &SystemState
  &self:SystemSpawnScript.SystemScriptSource.SystemScriptGraph.NiagaraNodeAssignment_0.SetVariables_#2
  &self:SystemSpawnScript.SystemScriptSource.SystemScriptGraph.NiagaraNodeAssignment_1.SetVariables_#3
  &self:SystemSpawnScript.SystemScriptSource.SystemScriptGraph.NiagaraNodeAssignment_4.SetVariables_#4
  &self:SystemSpawnScript.SystemScriptSource.SystemScriptGraph.NiagaraNodeAssignment_2.SetVariables_#5
  &Timeline
 ]
}
emitters {GPU_SnowLandRun {
  emitterUpdateScript [
   &EmitterState()
   &SpawnRate(
    Spawn Probability=&Multiply_Float(
     A=$System.Timeline.NormalizedTime
     B=&Multiply_Float(A=$User.Reactivity, B=&MakeCustomFloatFromBool(Bool=$User.bIsADS, True Float=0, False Float=1))
    )
    SpawnRate=&FloatFromCurve(
     FloatCurve=DataInterface() {propertyValues {Curve {
        defaultValue=0.1
        keys [
         {arriveTangent=0.3156, interpMode=RCIM_Cubic, leaveTangent=0.3156, tangentMode=RCTM_User, time=0, value=0}
         {arriveTangent=2.4808, interpMode=RCIM_Cubic, leaveTangent=2.4808, tangentMode=RCTM_User, time=1, value=1}
        ]
       }}}
     CurveIndex=&Multiply_Float(A=$User.MatchTime, B=0.65)
     Scale Curve=&Lerp_Float(A=250, B=1500, Alpha=$User.Cape)
    )
    Use Spawn Probability=true
   ) {module=SpawnRate001}
   &SpawnRate(
    Spawn Probability=&Multiply_Float(
     A=$System.bIsSprinting
     B=&Multiply_Float(A=$User.Reactivity, B=&MakeCustomFloatFromBool(Bool=$User.bIsADS, True Float=0, False Float=1))
    )
    SpawnGroup=1
    SpawnRate=&FloatFromCurve(
     FloatCurve=DataInterface() {propertyValues {Curve {
        defaultValue=0.1
        keys [
         {arriveTangent=0.3156, interpMode=RCIM_Cubic, leaveTangent=0.3156, tangentMode=RCTM_User, time=0, value=0}
         {arriveTangent=2.4808, interpMode=RCIM_Cubic, leaveTangent=2.4808, tangentMode=RCTM_User, time=1, value=1}
        ]
       }}}
     CurveIndex=&Multiply_Float(A=$User.MatchTime, B=0.65)
     Scale Curve=100
    )
    Use Spawn Probability=true
   ) {module=SpawnRate002}
  ]
  particleSpawnScript [
   /SetVariables(Particles.PerParticleRandom:NiagaraFloat=&RandomRangeFloat(Minimum=0, Maximum=1)) {isSetParameters=true}
   &InitializeParticle(
    Color=$System.Color ToD
    Lifetime=&RandomRangeFloat(Minimum=0.75, Maximum=1.25)
    Sprite Size Max=<0.6 0.6>
    Sprite Size Min=<0.4 0.4>
    Sprite Size Mode=::NewEnumerator2
    UsePositionOffset=false
    Write Material Random=true
   )
   &SkeletalMeshLocation(
    Bone Sampling Mode=::NewEnumerator4
    Interpolate Mesh Sampling=false
    MaskSpawnGroup=false
    Orientation Sampling=::NewEnumerator1
    Sampled Position Offset=<50 0 0>
    Skeletal Mesh=$System.Skeletal Mesh
   )
   &ShapeLocation(MaskSpawnGroup=false, Non Uniform Scale=<5 3 1>, Override Seed=false, Sphere Radius=5, UseOwnerScale=true)
   &AddVelocity(
    Cone Angle=75
    Cone Axis=<0 0 1>
    Module.UseConeFalloff=false
    Module.UseVelDistribution=false
    Override Seed=false
    Rotation Mode=::NewEnumerator2
    Velocity Mode=::NewEnumerator2
    Velocity Speed=&Multiply_Float(A=175, B=&RandomRangeFloat(Minimum=0.25, Maximum=1))
    Yaw / Pitch / Roll:Vector3f=<0 0 -30>
   )
  ]
  particleUpdateScript [
   &ParticleState()
   &Drag(
    Drag=&Multiply_Float(A=2, B=&RandomRangeFloat(Minimum=0.5, Maximum=1))
    Ignore Mass=false
    Use Linear Drag=true
    Use Rotational Drag=false
   )
   &CurlNoiseForce(
    Bias Noise=false
    Noise Frequency=100
    Noise Strength=&FloatFromCurve(
     FloatCurve=DataInterface() {propertyValues {Curve {keys [
         {arriveTangent=1.3847, interpMode=RCIM_Cubic, leaveTangent=1.3847, tangentMode=RCTM_User, time=0, value=0.3}
         {interpMode=RCIM_Cubic, tangentMode=RCTM_User, time=1, value=1}
        ]}}}
     Scale Curve=&Lerp_Float(A=300, B=100, Alpha=$System.bIsSprinting)
    )
    Pan Noise=false
    Random Seed=33
   ) {module=CurlNoiseForce001}
   &GravityForce(Gravity=&Multiply_Vector(A=<0 0 -49>, B=&VectorFromFloat(Value=4)))
   &SolveForcesAndVelocity(Clamp Velocity=false, Limit Acceleration=false)
   &ScaleColor(
    Scale Alpha=&FloatFromCurve(FloatCurve=DataInterface() {propertyValues {Curve {
        defaultValue=1
        keys [
         {arriveTangent=9.0725, interpMode=RCIM_Cubic, leaveTangent=9.0725, tangentMode=RCTM_User, time=0, value=0}
         <at 0.2 0.9847 RCIM_Cubic>
         {arriveTangent=-3.3227, interpMode=RCIM_Cubic, leaveTangent=-3.3227, tangentMode=RCTM_User, time=1, value=0}
        ]
       }}})
    ScaleA=true
    ScaleRGB=true
   )
   &ScaleSpriteSize(
    Uniform Curve Scale=&RandomRangeFloat(Minimum=1, Maximum=2.5)
    Uniform Curve Sprite Scale=DataInterface() {propertyValues {Curve {keys [
        {tangentMode=RCTM_User, time=0, value=1}
        {arriveTangent=-0.0405, interpMode=RCIM_Cubic, leaveTangent=-0.0405, tangentMode=RCTM_User, time=0.7009, value=1}
        {arriveTangent=-8.0089, interpMode=RCIM_Cubic, leaveTangent=-8.0089, tangentMode=RCTM_User, time=1, value=0}
       ]}}}
   )
   /SetVariables(Particles.SpriteSizeGlow:Vector2f=&Multiply_Vector2D_ByFloat(Float=1.2, Vector2D=$Particles.SpriteSize)) {isSetParameters=true}
   /SetVariables(Particles.Glow Color=&MultiplyLinearColor(Linear Color=$Particles.Color, Scale Factor=<2 2 2 0.15>)) {isSetParameters=true}
   /SetVariables(Particles.Position=&AddVectorToPosition(
     Position=$Particles.Position
     Vector=&Multiply_Vector(
      A=&TransformVector(Vector=$Engine.Owner.Velocity, Source Space=::World, Destination Space=::Local)
      B=&Multiply_VectorByFloat(
       Vector=<-1 -1 0>
       Float=&FloatFromCurve(
        FloatCurve=DataInterface() {propertyValues {Curve {keys [
            {arriveTangent=2.0546, interpMode=RCIM_Cubic, leaveTangent=2.0546, tangentMode=RCTM_User, time=0, value=0}
            {interpMode=RCIM_Cubic, tangentMode=RCTM_User, time=1, value=1}
           ]}}}
        Scale Curve=0.0025
       )
      )
     )
    )) {enabled=false, isSetParameters=true}
   /SetVariables(Particles.Position=&AddVectorToPosition(
     Position=$Particles.Position
     Vector=&Multiply_Vector(
      A=&TransformVector(Vector=$Engine.Owner.Velocity, Source Space=::World, Destination Space=::World)
      B=&Multiply_VectorByFloat(
       Vector=<0.1 0.1 0.08>
       Float=&FloatFromCurve(
        FloatCurve=DataInterface() {propertyValues {Curve {keys [<at 0 0.255>, <at 0.5 0.2599 RCIM_Constant>, {arriveTangent=-0.7941, interpMode=RCIM_Constant, leaveTangent=-0.7941, time=0.51, value=0.1}]}}}
        CurveIndex=$System.SpeedNormalized
        Scale Curve=&MaskFloat(Mask Source=0, Mask Parameter=?, Mask=0, Masked Float=0, Passthrough Float=0.7)
       )
      )
     )
    )) {enabled=false, isSetParameters=true}
  ]
  rendererClasses [&NiagaraSpriteRendererProperties]
  renderers [
   {rendererClass=&NiagaraSpriteRendererProperties}
   {rendererClass=&NiagaraSpriteRendererProperties, rendererIndex=1}
  ]
 }}
summary {
 systemName=NS_Character_CurbRestore
 emitters [{emitterName=GPU_SnowLandRun, rendererClasses [&NiagaraSpriteRendererProperties]}]
 userVariables [
  User.bIsADS {type=NiagaraBool}
  User.bIsFrontEnd {type=NiagaraBool}
  User.bIsFrontEndPreview {type=NiagaraBool}
  User.bPawnLanded {type=NiagaraBool}
  User.Cape {type=NiagaraFloat, defaultValue=1, description=""}
  User.MatchTime {type=NiagaraFloat, defaultValue=0, description=""}
  User.Reactivity {type=NiagaraFloat, defaultValue=1, description=""}
 ]
}
