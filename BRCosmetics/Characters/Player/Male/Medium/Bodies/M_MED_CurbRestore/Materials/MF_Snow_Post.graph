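# Material function that layers snow over the input material attributes for the CurbRestore male character body.
# A tiled snow mask texture varies the snow's color, metallic and roughness, and an inverted Fresnel term adds tinted emissive.
# The snow attributes blend over the original by the SnowMask input, scaled by the CR_IsReactive parameter.
# Detail-normal and switchable vertex-offset branches are present but not wired into the snow layer.
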MaterialFunction /BRCosmetics/Characters/Player/Male/Medium/Bodies/M_MED_CurbRestore/Materials/MF_Snow_Post
graph {nodes [
  Constant()
  n1 = GetMaterialAttributes(None=FunctionInput() {inputName=Material, inputType=FunctionInput_MaterialAttributes}) {attributeGetTypes [
    #1
    #2
    #3
    #4
   ]}
  n2 = Multiply(
   A=OneMinus(None=Fresnel() {baseReflectFraction=0, exponent=1})
   B=ScalarParameter() {defaultValue=0.15, group=Snow, parameterName=Snow_FresnelIntensity}
  )
  n3 = VectorParameter() {parameterName=SnowEmissive_FresnelColor, defaultValue=<rgba 0.9635 0.9635 0.9635 1>}
  n4 = SetMaterialAttributes!(
   MaterialAttributes=@n1.MaterialAttributes
   Normal
   EmissiveColor=Add(A=@n1.EmissiveColor, B=Multiply(A=@n2, B=@n3.RGB))
   BaseColor=NamedRerouteUsage()
   Specular=ScalarParameter() {defaultValue=1, group=Snow, parameterName=Snow_Specular}
   Metallic=NamedRerouteUsage()
   Roughness=NamedRerouteUsage()
   WorldPositionOffset
  ) {attributeSetTypes [
    #1
    #3
    #5
    #6
    #7
    #8
    #4
   ]}
  n5 = FunctionInput() {inputName=SnowMask, inputType=FunctionInput_Scalar, sortPriority=1}
  n6 = ScalarParameter() {defaultValue=1, group=Snow, parameterName=CR_IsReactive}
  FunctionOutput(None=BlendMaterialAttributes(A=@n1.MaterialAttributes, B=@n4, Alpha=LinearInterpolate(B=@n5, Alpha=@n6) {constB=0}))
  n7 = TextureSample(UVs=Multiply(A=TextureCoordinate(), B=ScalarParameter() {defaultValue=6, group=Snow, parameterName=Snow_Tiling3})) {samplerType=SAMPLERTYPE_Masks, texture=/BRCosmetics/Characters/Player/Male/Medium/Bodies/M_MED_CurbRestore/Textures/T_CurbRestore_SnowMask}
  n8 = SmoothStep(Value=@n7.B)
  n9 = LinearInterpolate(Alpha=@n8) {constA=0.5, constB=0}
  Min(A=@n9) {constB=0.2}
  NamedRerouteDeclaration(None=Multiply(
    A=VectorParameter() {group=Snow, parameterName=Snow_Color, defaultValue=<rgba 0.9635 0.9635 0.9635 1>}.RGB
    B=LinearInterpolate(Alpha=@n8) {constA=0.8}
   )) {name=Basecolor_OUT, nodeColor=<rgba 0.8516 1 0 1>}
  NamedRerouteDeclaration(None=@n1.Normal) {name=Normal_IN, nodeColor=<rgba 1 0 0.5859 1>}
  n10 = Multiply(A=TextureCoordinate()) {constB=4}
  NamedRerouteDeclaration(None=MaterialFunctionCall(BaseNormal=NamedRerouteUsage(), AdditionalNormal=TextureSample(UVs=@n10) {samplerType=SAMPLERTYPE_Normal}.RGB) {materialFunction=&BlendAngleCorrectedNormals}.Result) {name=Normal_OUT, nodeColor=<rgba 0 0.9062 1 1>}
  NamedRerouteDeclaration(None=LinearInterpolate(Alpha=@n8) {constB=0.5}) {name=Metallic_OUT, nodeColor=<rgba 0 1 0.1797 1>}
  n11 = Multiply(A=PreSkinnedNormal(), B=Constant3Vector() {constant=<rgba 0 0 1 0>})
  NamedRerouteDeclaration(None=Multiply(A=Abs(None=@n11), B=ScalarParameter() {defaultValue=0.5, parameterName=SnowWPO})) {name=WPO_OUT, nodeColor=<rgba 0 1 0.0156 1>}
  NamedRerouteDeclaration(None=@n9) {name=Roughness_OUT, nodeColor=<rgba 1 0 0.2109 1>}
  NamedRerouteUsage()
  StaticSwitchParameter(True=NamedRerouteUsage(), False=@n1.WorldPositionOffset) {parameterName=UseSnowWPO}
 ]}
