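# Material function that builds a snow mask for the CurbRestore character body, outputting it alongside an upward-face mask.
# The upward mask raises a normal's Z to a power, apparently the pre-skinned local normal, so snow stays put through animation.
# Opacity-scaled, noise-eroded upward coverage adds to a channel-selected TechArtMask layer.
# A tiled mask texture offset by MatchTime gates the result, so snow coverage grows as that parameter rises.
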MaterialFunction /BRCosmetics/Characters/Player/Male/Medium/Bodies/M_MED_CurbRestore/Materials/MF_Snow_Pre
graph {nodes [
  n1 = ComponentMask(None=NamedRerouteUsage()) {b=true}
  n2 = ScalarParameter() {defaultValue=2, group=Snow, parameterName=Snow_UpwardNormalPower}
  n3 = NamedRerouteDeclaration(None=Saturate(None=Power(Base=@n1, Exp=@n2))) {name=VertexNormalB, nodeColor=<rgba 0 1 0.2266 1>}
  FunctionOutput(None=@n3) {outputName=UpwardFace, sortPriority=5}
  n4 = FunctionInput() {inputName=TechArtMask, sortPriority=1}
  n5 = Multiply(A=@n3, B=ScalarParameter() {defaultValue=0.75, group=Snow, parameterName=Snow_Layer2Opacity})
  n6 = TextureObject() {samplerType=SAMPLERTYPE_Masks, texture=/BRCosmetics/Characters/Player/Male/Medium/Bodies/M_MED_CurbRestore/Textures/T_CurbRestore_SnowMask}
  n7 = TextureSample(UVs=Multiply(A=TextureCoordinate(), B=ScalarParameter() {defaultValue=5, group=Snow, parameterName=Snow_Tiling2}), Tex=@n6) {samplerSource=SSM_Wrap_WorldGroupSettings}
  n8 = Subtract(A=NamedRerouteUsage(), B=Multiply(A=@n7.R) {constB=5})
  n9 = Multiply(
   A=ComponentMask(None=@n7.RGB) {g=true}
   B=LinearInterpolate(A=ScalarParameter() {defaultValue=6, group=Snow, parameterName=Snow_NoiseSub}, Alpha=Saturate(None=@n8)) {constB=0}
  ) {constB=0.5}
  n10 = Add(
   A=Multiply(
    A=ChannelMaskParameter(None=@n4) {group=Snow, parameterName=Snow_MaskChannel}.R
    B=ScalarParameter() {defaultValue=0.7, group=Snow, parameterName=Snow_Layer1Opacity}
   )
   B=Saturate(None=Subtract(A=@n5, B=@n9))
  )
  n11 = TextureCoordinate()
  n12 = ScalarParameter() {defaultValue=8, group=Snow, parameterName=Snow_Tiling}
  n13 = TextureObject() {samplerType=SAMPLERTYPE_Masks, texture=&T_CurbRestore_SnowMask}
  n14 = ComponentMask(None=TextureSample(UVs=Multiply(A=@n11, B=@n12), Tex=@n13) {samplerSource=SSM_Wrap_WorldGroupSettings}.RGB) {r=true}
  n15 = ConstantBiasScale(None=ScalarParameter() {desc=0-1, group=Snow, parameterName=MatchTime}) {bias=-0.5, scale=2}
  n16 = Multiply(A=@n10, B=Saturate(None=Add(A=@n14, B=@n15))) {constA=1}
  FunctionOutput(None=NamedRerouteDeclaration(None=Saturate(None=@n16)) {name=Snow Mask, nodeColor=<rgba 0.5938 0 1 1>}) {outputName=SnowMask}
  n17 = PreSkinnedNormal()
  NamedRerouteDeclaration(None=ShaderStageSwitch(PixelShader=VertexInterpolator(VS=@n17).PS, VertexShader=@n17)) {name=PS-LocalNormal, nodeColor=<rgba 1 0.4868 0.4485 1>}
  Saturate(None=@n10)
  ScalarParameter() {defaultValue=1, group=Snow, parameterName=DebugTime, sliderMax=1}
 ]}
