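# Parent material for the CurbRestore male character body, adding snow to the shared 2023 character shader with fur.
# Character PreFX supplies material and masks to a fur chain of custom color, adjustments, painting, snow-aware shells and WPO shrink, then Snow_Post and a PostFX stage.
# The UseFurShells static switch picks the fur or base snow mask.
# A separate FluffyFabric-plus-snow branch has no visible consumer.
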Material /BRCosmetics/Characters/Player/Male/Medium/Bodies/M_MED_CurbRestore/Materials/M_CurbRestore_Snow_Parent
graph {
 properties {
  bAutomaticallySetUsageInEditor=false
  bUseMaterialAttributes=true
  bUsedWithMeshDeformer=true
  bUsedWithMorphTargets=true
  bUsedWithSkeletalMesh=true
  pixelDepthOffsetMode=PDOM_Legacy
  shadingModel=MSM_FromMaterialExpression
  usedShadingModels=Default Lit | Subsurface | Clear Coat | Subsurface Profile | Cloth
 }
 nodes [
  n1 = MaterialFunctionCall() {materialFunction=&MF_CharacterParent_2023_PreFX}
  MaterialFunctionCall(Material=NamedRerouteUsage(), M=@n1.M, Pre Skinned Local Position=@n1.M, Distance Blend=@n1.M, Fresnel=@n1.M) {materialFunction=&MF_CharacterParent_2023_PostFX}
  MaterialFunctionCall!(
   Material=MaterialFunctionCall!(Material=@n1.Result, TechArtMask=@n1.Result) {materialFunction=/Game/Global/Materials/Character_Parent_2023/FluffyFabric_Parent_2023/MF_FluffyFabric}.Result
   SnowMask=MaterialFunctionCall!(TechArtMask=@n1.Tech Art Mask) {materialFunction=&/MF_Snow_Pre}.SnowMask
  ) {materialFunction=&/MF_Snow_Post}
  NamedRerouteDeclaration(None=@n1.Tech Art Mask) {name=TechArtMask, nodeColor=<rgba 0.2188 0 1 1>}
  NamedRerouteDeclaration(None=@n1.Result) {name=Material, nodeColor=<rgba 0 1 0.1094 1>}
  NamedRerouteDeclaration(None=@n1.M) {name=Mask, nodeColor=<rgba 1 0 0.2109 1>}
  n2 = MaterialFunctionCall!(Material=NamedRerouteUsage(), TechArtMask=NamedRerouteUsage()) {materialFunction=/Game/Global/Materials/Character_Parent_2023/Fur_Parent_2023/MF_Fur_CustomColor}
  n3 = NamedRerouteUsage()
  n4 = NamedRerouteUsage()
  n5 = MaterialFunctionCall(TechArtMask=@n4) {materialFunction=&MF_Snow_Pre}
  n6 = MaterialFunctionCall!(
   In=MaterialFunctionCall!(
    In=MaterialFunctionCall!(MA=@n2.Result, TechArtMask=@n3) {materialFunction=/Game/Global/Materials/Character_Parent_2023/Fur_Parent_2023/MF_Fur_ParamAdjustments}.Result
    TechArtMask=NamedRerouteUsage()
    Mask Input=NamedRerouteUsage()
   ) {materialFunction=/Game/Global/Materials/Character_Parent_2023/Fur_Parent_2023/MF_Fur_Painted}.Result
   TechArtMask=@n4
   SnowMask=@n5.SnowMask
   UpwardFace=@n5.SnowMask
  ) {materialFunction=&/MF_Fur_Shells_Snow}
  n7 = NamedRerouteUsage()
  NamedRerouteDeclaration(None=MaterialFunctionCall(
    Material=MaterialFunctionCall!(Material=@n6.Result, Mask=@n7, Fur Length=@n6.Result) {materialFunction=/Game/Global/Materials/Character_Parent_2023/Fur_Parent_2023/MF_FurWPOShrink}.Result
    SnowMask=StaticSwitchParameter(True=@n6.FurSnowMask, False=@n5.SnowMask) {group=Fur_Shell_01 - Base Settings, parameterName=UseFurShells, sortPriority=-2}
   ) {materialFunction=&MF_Snow_Post}.Result) {name=OUT_Material, nodeColor=<rgba 0 0.5312 1 1>}
 ]
}
