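# Control Rig for the M_MED_CurbRestore player body that corrects deform bones and offsets MSTR skirt, cape, fur, belt and left sword bones.
# A shared DeformRig function runs first, then sequenced offsets weighted by smoothed crouch, jump/fall and rail-grind flags.
# Spherical pose readers on arms and legs drive further offsets, some gated by crouch or parachute state.
# GoldenGloryEmote scales spine-driven sword offsets; emotion enables a right-fur parent constraint.
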ControlRigBlueprint /BRCosmetics/Characters/Player/Male/Medium/Bodies/M_MED_CurbRestore/Meshes/M_MED_CurbRestore_CtrlRig
graphs [
 RigVMFunctionLibrary {links [], nodes []}
 RigVMFunction_Sequence_1_ContainedGraph {nodes [
   n1 = RigVMFunctionEntryNode()
   RigVMFunction_Sequence = &RigVMFunction_Sequence(ExecuteContext=@n1.ExecuteContext)
   n2 = &RigVMFunction_Sequence(ExecuteContext=@RigVMFunction_Sequence.B)
   n3 = &RigVMFunction_Sequence(ExecuteContext=@n2.B)
   n4 = &RigVMFunction_Sequence(ExecuteContext=@n3.B)
   n5 = &RigVMFunction_Sequence(ExecuteContext=@n4.B)
   n6 = &RigVMFunction_Sequence(ExecuteContext=@n5.B)
   RigVMFunctionReturnNode(A=@RigVMFunction_Sequence.A, B=@n2.A, C=@n3.A, D=@n4.A, E=@n5.A, F=@n6.A, G=@n6.B)
  ]}
 RigVMFunction_Sequence_2_ContainedGraph {nodes [
   n1 = RigVMFunctionEntryNode()
   n2 = &RigVMFunction_Sequence(ExecuteContext=@n1.ExecuteContext)
   n3 = &RigVMFunction_Sequence(ExecuteContext=@n2.B)
   RigVMFunctionReturnNode(A=@n2.A, B=@n3.A, C=@n3.B)
  ]}
 RigVMModel {nodes [
   RigUnit_BeginExecution = &RigUnit_BeginExecution()
   RigVMFunction_Sequence = RigVMAggregateNode(ExecuteContext=@RigUnit_BeginExecution.ExecutePin)
   FN_DeformRig_FNC = RigVMFunctionReferenceNode(
    DeformRig_ Curve="(deformRig_l_upperarm_up_alpha_72_#1=true,deformRig_r_upperarm_up_alpha_73_#2=true,deformRig_l_upperarm_fr_alpha_74_#3=False,deformRig_r_upperarm_fr_alpha_75_CB51B80C ...(truncated)"
    DeformRig_Bool="(upperarm_1_#4=True,elbow_3_#5=True,wrist_7_#6=True,groin_8_#7=True,glute_17_#8=True,knee_9_40C408324CA261 ...(truncated)"
    DeformRig_Debug=(upperarm_debug_18_#4=False,wrist_debug_20_#6=False,lat_debug_26_#9=False,groinGlute_debug_27_#7=False)
    DeformRig_Value="(upperarm_up_18_#10=(Rotation=(X=0,Y=0,Z=0,W=1),Translation=(X=0,Y=0,Z=8),Scale3D=(X=0.7,Y=1,Z=1)),upperarm_fwd_fr_19_#11=(Rotation=(X=0,Y=0,Z=0,W=1),Translation=(X=0,Y=10,Z=0),Scale3D=(X=0.5,Y=1,Z= ...(truncated)"
    ExecuteContext=@RigVMFunction_Sequence.A
   )
   n1 = &RigVMFunction_AlphaInterp(
    InterpSpeedDecreasing=6
    InterpSpeedIncreasing=6
    Scale=1
    bClampResult=false
    bMapRange=false
    Value=&RigVMFunction_MathBoolToFloat(Value=RigVMVariableNode() {variable=jumpOrFalling}.Value).Result
   )
   RigVMVariableNode(Value=@n1.Result, ExecuteContext=@RigVMFunction_Sequence.C) {variable=jumpOrFallingAlpha}
   n2 = &RigVMFunction_AlphaInterp(
    InterpSpeedDecreasing=6
    InterpSpeedIncreasing=6
    Scale=1
    bClampResult=false
    bMapRange=false
    Value=&RigVMFunction_MathBoolToFloat(Value=RigVMVariableNode() {variable=grindRail}.Value).Result
   )
   > RigVMVariableNode(Value=@n2.Result) {variable=grindRailAlpha}
   RigVMFunction_MathBoolToFloat = &RigVMFunction_MathBoolToFloat(Value=RigVMVariableNode() {variable=crouchingAndNotMoving}.Value)
   AlphaInterp = &RigVMFunction_AlphaInterp(InterpSpeedDecreasing=8, InterpSpeedIncreasing=8, Scale=1, bClampResult=false, bMapRange=false, Value=@RigVMFunction_MathBoolToFloat.Result)
   RigUnit_SetRelativeTransformForItem = /Script/ControlRig.RigUnit_SetRelativeTransformForItem(
    Child=(Type=Bone,Name="MSTR_skirt_back_l")
    Parent=(Type=Bone,Name="MSTR_skirt_back_l")
    Value=(Rotation=(X=0,Y=0,Z=-0.342,W=0.9397),Translation=(X=-27,Y=0,Z=0),Scale3D=(X=1,Y=1,Z=1))
    bParentInitial=False
    bPropagateToChildren=True
    Weight=@AlphaInterp.Result
    ExecutePin=@RigVMFunction_Sequence.D
   ) {
    is=struct RigVMUnitNode
    template="Set Relative Transform::Execute(in Child,in Parent,in bParentInitial,in Value,in Weight,in bPropagateToChildren)"
   }
   > &RigUnit_SetRelativeTransformForItem(
    Child=(Type=Bone,Name="MSTR_skirt_back_r")
    Parent=(Type=Bone,Name="MSTR_skirt_back_r")
    Value=(Rotation=(X=0.25,Y=-0.25,Z=0.067,W=0.933),Translation=(X=0,Y=0,Z=-18),Scale3D=(X=1,Y=1,Z=1))
    bParentInitial=False
    bPropagateToChildren=True
    Weight=@AlphaInterp.Result
   ) {template="Set Relative Transform::Execute(in Child,in Parent,in bParentInitial,in Value,in Weight,in bPropagateToChildren)"}
   n3 = &RigVMFunction_AlphaInterp(
    InterpSpeedDecreasing=10
    InterpSpeedIncreasing=10
    Scale=1
    bClampResult=false
    bMapRange=false
    Value=&RigVMFunction_MathBoolToFloat(Value=RigVMVariableNode() {variable=crouchingAndMoving}.Value).Result
   )
   > &RigUnit_SetRelativeTransformForItem(
    Child=(Type=Bone,Name="MSTR_skirt_back_l")
    Parent=(Type=Bone,Name="MSTR_skirt_back_l")
    Value=(Rotation=(X=0,Y=0,Z=-0.2504,W=0.9681),Translation=(X=0,Y=0,Z=0),Scale3D=(X=1,Y=1,Z=1))
    bParentInitial=False
    bPropagateToChildren=True
    Weight=@n3.Result
   ) {template="Set Relative Transform::Execute(in Child,in Parent,in bParentInitial,in Value,in Weight,in bPropagateToChildren)"}
   > &RigUnit_SetRelativeTransformForItem(
    Child=(Type=Bone,Name="MSTR_skirt_back_r")
    Parent=(Type=Bone,Name="MSTR_skirt_back_r")
    Value=(Rotation=(X=0.0858,Y=-0.0151,Z=-0.173,W=0.9811),Translation=(X=-8,Y=0,Z=-5),Scale3D=(X=1,Y=1,Z=1))
    bParentInitial=False
    bPropagateToChildren=True
    Weight=@n3.Result
   ) {template="Set Relative Transform::Execute(in Child,in Parent,in bParentInitial,in Value,in Weight,in bPropagateToChildren)"}
   RigVMFunction_MathQuaternionToEuler = &RigVMFunction_MathQuaternionToEuler(
    RotationOrder=ZYX
    Value=&RigUnit_GetRelativeTransformForItem(Child=(Type=Bone,Name="spine_01"), Parent=(Type=Bone,Name="pelvis"), bChildInitial=False, bParentInitial=False).RelativeTransform.Rotation
   )
   Remap = &RigVMFunction_MathDoubleRemap(SourceMaximum=10, SourceMinimum=0, TargetMaximum=1, TargetMinimum=0, Value=@RigVMFunction_MathQuaternionToEuler.Result.Y)
   Multiply = &RigVMFunction_MathDoubleMul(B=@Remap.Result, A=RigVMVariableNode() {variable=GoldenGloryEmote}.Value)
   > &RigUnit_ModifyTransforms(
    ItemToModify=((Item=(Type=Bone,Name="MSTR_sword_l"),Transform=(Rotation=(X=-0.1256,Y=-0.1374,Z=-0.158,W=0.9697),Translation=(X=-0.2293,Y=0.1331,Z=1.2479),Scale3D=(X=1,Y=1,Z=1))))
    WeightMinimum=0
    Weight=@Multiply.Result
   )
   > &RigUnit_ModifyTransforms(
    ItemToModify=((Item=(Type=Bone,Name="MSTR_sword_l"),Transform=(Rotation=(X=0.0116,Y=0.0598,Z=0.2305,W=-0.9712),Translation=(X=0,Y=0,Z=0),Scale3D=(X=1,Y=1,Z=1))))
    WeightMinimum=0
    Weight=&RigVMFunction_MathDoubleMul(
     B=&RigVMFunction_MathDoubleRemap(SourceMaximum=10, SourceMinimum=0, TargetMaximum=1, TargetMinimum=0, Value=@RigVMFunction_MathQuaternionToEuler.Result.X).Result
     A=RigVMVariableNode() {variable=GoldenGloryEmote}.Value
    ).Result
   )
   RigVMFunction_MathBoolNot = &RigVMFunction_MathBoolNot(Value=RigVMVariableNode() {variable=crouching}.Value)
   RigVMFunction_ControlFlowBranch = &RigVMFunction_ControlFlowBranch(Condition=@RigVMFunction_MathBoolNot.Result, ExecuteContext=@RigVMFunction_Sequence.E)
   &RigUnit_SetRelativeTransformForItem(
    Child=(Type=Bone,Name="MSTR_skirt_back_l")
    Parent=(Type=Bone,Name="MSTR_skirt_back_l")
    Value=(Rotation=(X=0.0038,Y=0.0435,Z=-0.0871,W=0.9952),Translation=(X=0,Y=0,Z=5),Scale3D=(X=1,Y=1,Z=1))
    Weight=1
    bParentInitial=False
    bPropagateToChildren=True
    ExecutePin=@RigVMFunction_ControlFlowBranch.False
   ) {template="Set Relative Transform::Execute(in Child,in Parent,in bParentInitial,in Value,in Weight,in bPropagateToChildren)"}
   n4 = &RigUnit_SphericalPoseReader(
    ActiveRegionScaleFactors=(PositiveWidth=1,NegativeWidth=1,PositiveHeight=1,NegativeHeight=1)
    ActiveRegionSize=0.1
    Debug=(bDrawDebug=false,bDraw2D=False,bDrawLocalAxes=False,DebugScale=25,DebugSegments=20,DebugThickness=0.25)
    DriverAxis=(X=-1,Y=0,Z=0)
    DriverItem=(Type=Bone,Name="calf_l")
    FalloffRegionScaleFactors=(PositiveWidth=1,NegativeWidth=1.5,PositiveHeight=1,NegativeHeight=1)
    FalloffSize=0.2
    FlipHeightScaling=False
    FlipWidthScaling=False
    OptionalParentItem=(Type=Bone,Name="None")
    RotationOffset=(X=-90,Y=0,Z=10)
    ExecutePin=@RigVMFunction_ControlFlowBranch.True
   )
   > &RigUnit_SetRelativeTransformForItem(
    Child=(Type=Bone,Name="MSTR_skirt_back_l")
    Parent=(Type=Bone,Name="MSTR_skirt_back_l")
    Value=(Rotation=(X=0,Y=0,Z=-0.1045,W=0.9945),Translation=(X=0,Y=0,Z=0),Scale3D=(X=1,Y=1,Z=1))
    bParentInitial=False
    bPropagateToChildren=True
    Weight=@n4.OutputParam
   ) {template="Set Relative Transform::Execute(in Child,in Parent,in bParentInitial,in Value,in Weight,in bPropagateToChildren)"}
   n5 = RigVMVariableNode() {variable=jumpOrFallingAlpha}
   &RigUnit_SetRelativeTransformForItem(
    Child=(Type=Bone,Name="MSTR_skirt_back_l")
    Parent=(Type=Bone,Name="MSTR_skirt_back_l")
    Value=(Rotation=(X=-0.4957,Y=0.0653,Z=-0.113,W=0.8586),Translation=(X=0,Y=0,Z=9),Scale3D=(X=1,Y=1,Z=1))
    bParentInitial=False
    bPropagateToChildren=True
    ExecutePin=@RigVMFunction_Sequence.F
    Weight=@n5.Value
   ) {template="Set Relative Transform::Execute(in Child,in Parent,in bParentInitial,in Value,in Weight,in bPropagateToChildren)"}
   > &RigUnit_SetRelativeTransformForItem(
    Child=(Type=Bone,Name="MSTR_sword_l")
    Parent=(Type=Bone,Name="MSTR_sword_l")
    Value=(Rotation=(X=0,Y=0,Z=-0.1305,W=0.9914),Translation=(X=0,Y=0,Z=0),Scale3D=(X=1,Y=1,Z=1))
    bParentInitial=False
    bPropagateToChildren=True
    Weight=@n5.Value
   ) {template="Set Relative Transform::Execute(in Child,in Parent,in bParentInitial,in Value,in Weight,in bPropagateToChildren)"}
   n6 = RigVMVariableNode() {variable=grindRailAlpha}
   > &RigUnit_SetRelativeTransformForItem(
    Child=(Type=Bone,Name="MSTR_skirt_back_l")
    Parent=(Type=Bone,Name="MSTR_skirt_back_l")
    Value=(Rotation=(X=-0.4226,Y=0,Z=0,W=0.9063),Translation=(X=0,Y=0,Z=12),Scale3D=(X=1,Y=1,Z=1))
    bParentInitial=False
    bPropagateToChildren=True
    Weight=@n6.Value
   ) {template="Set Relative Transform::Execute(in Child,in Parent,in bParentInitial,in Value,in Weight,in bPropagateToChildren)"}
   > &RigUnit_SetRelativeTransformForItem(
    Child=(Type=Bone,Name="MSTR_skirt_side_l")
    Parent=(Type=Bone,Name="MSTR_skirt_side_l")
    Value=(Rotation=(X=0,Y=0,Z=0.0872,W=0.9962),Translation=(X=0,Y=0,Z=0),Scale3D=(X=1,Y=1,Z=1))
    bParentInitial=False
    bPropagateToChildren=True
    Weight=@n6.Value
   ) {template="Set Relative Transform::Execute(in Child,in Parent,in bParentInitial,in Value,in Weight,in bPropagateToChildren)"}
   > &RigUnit_SetRelativeTransformForItem(
    Child=(Type=Bone,Name="MSTR_skirt_back_r")
    Parent=(Type=Bone,Name="MSTR_skirt_back_r")
    Value=(Rotation=(X=0.0858,Y=-0.0151,Z=-0.173,W=0.9811),Translation=(X=-8,Y=0,Z=-5),Scale3D=(X=1,Y=1,Z=1))
    bParentInitial=False
    bPropagateToChildren=True
    Weight=@n6.Value
   ) {template="Set Relative Transform::Execute(in Child,in Parent,in bParentInitial,in Value,in Weight,in bPropagateToChildren)"}
   RigVMFunction_Sequence_1 = RigVMAggregateNode(ExecuteContext=@RigVMFunction_Sequence.G)
   n7 = &RigVMFunction_ControlFlowBranch(ExecuteContext=@RigVMFunction_Sequence_1.A, Condition=RigVMVariableNode() {variable=crouching}.Value)
   n8 = &RigUnit_SphericalPoseReader(
    ActiveRegionScaleFactors=(PositiveWidth=1,NegativeWidth=2,PositiveHeight=1,NegativeHeight=1)
    ActiveRegionSize=0.1
    Debug=(bDrawDebug=false,bDraw2D=False,bDrawLocalAxes=False,DebugScale=25,DebugSegments=20,DebugThickness=0.25)
    DriverAxis=(X=1,Y=0,Z=0)
    DriverItem=(Type=Bone,Name="calf_r")
    FalloffRegionScaleFactors=(PositiveWidth=0.75,NegativeWidth=1,PositiveHeight=1,NegativeHeight=1)
    FalloffSize=0.3
    FlipHeightScaling=False
    FlipWidthScaling=False
    OptionalParentItem=(Type=Bone,Name="None")
    RotationOffset=(X=90,Y=0,Z=0)
    ExecutePin=@n7.False
   )
   n9 = &RigVMFunction_AlphaInterp(InterpSpeedDecreasing=10, InterpSpeedIncreasing=10, Scale=1, bClampResult=false, bMapRange=false, Value=@n8.OutputParam)
   > &RigUnit_SetRelativeTransformForItem(
    Child=(Type=Bone,Name="MSTR_skirt_back_r")
    Parent=(Type=Bone,Name="MSTR_skirt_back_r")
    Value=(Rotation=(X=0.0376,Y=0.2132,Z=-0.1695,W=0.9615),Translation=(X=-15,Y=0,Z=0),Scale3D=(X=1,Y=1,Z=1))
    bParentInitial=False
    bPropagateToChildren=True
    Weight=@n9.Result
   ) {template="Set Relative Transform::Execute(in Child,in Parent,in bParentInitial,in Value,in Weight,in bPropagateToChildren)"}
   n10 = &RigUnit_SphericalPoseReader(
    ActiveRegionScaleFactors=(PositiveWidth=1,NegativeWidth=1,PositiveHeight=1,NegativeHeight=1)
    ActiveRegionSize=0.1
    Debug=(bDrawDebug=false,bDraw2D=False,bDrawLocalAxes=False,DebugScale=25,DebugSegments=20,DebugThickness=0.25)
    DriverAxis=(X=1,Y=0,Z=0)
    DriverItem=(Type=Bone,Name="thigh_r")
    FalloffRegionScaleFactors=(PositiveWidth=1,NegativeWidth=1,PositiveHeight=1,NegativeHeight=1)
    FalloffSize=0.2
    FlipHeightScaling=False
    FlipWidthScaling=False
    OptionalParentItem=(Type=Bone,Name="None")
    RotationOffset=(X=-85,Y=0,Z=0)
    ExecutePin=@n7.True
   )
   > &RigUnit_SetRelativeTransformForItem(
    Child=(Type=Bone,Name="MSTR_belt")
    Parent=(Type=Bone,Name="MSTR_belt")
    Value=(Rotation=(X=0,Y=-0.6756,Z=0,W=0.7373),Translation=(X=0,Y=0,Z=0),Scale3D=(X=1,Y=1,Z=1))
    bParentInitial=False
    bPropagateToChildren=True
    Weight=@n10.OutputParam
   ) {template="Set Relative Transform::Execute(in Child,in Parent,in bParentInitial,in Value,in Weight,in bPropagateToChildren)"}
   n11 = &RigVMFunction_Sequence(ExecuteContext=@RigVMFunction_Sequence_1.B)
   n12 = &RigVMFunction_ControlFlowBranch(ExecuteContext=@n11.A, Condition=RigVMVariableNode() {variable=parachuteOpened}.Value)
   n13 = &RigUnit_SphericalPoseReader(
    ActiveRegionScaleFactors=(PositiveWidth=1,NegativeWidth=1,PositiveHeight=1,NegativeHeight=1)
    ActiveRegionSize=0.1
    Debug=(bDrawDebug=false,bDraw2D=False,bDrawLocalAxes=False,DebugScale=25,DebugSegments=20,DebugThickness=0.25)
    DriverAxis=(X=0,Y=-1,Z=0)
    DriverItem=(Type=Bone,Name="upperarm_r")
    FalloffRegionScaleFactors=(PositiveWidth=1,NegativeWidth=1,PositiveHeight=2,NegativeHeight=1.5)
    FalloffSize=0.25
    FlipHeightScaling=False
    FlipWidthScaling=False
    OptionalParentItem=(Type=Bone,Name="None")
    RotationOffset=(X=-210,Y=35,Z=15)
    ExecutePin=@n12.False
   )
   > &RigUnit_SetRelativeTransformForItem(
    Child=(Type=Bone,Name="MSTR_fur_r")
    Parent=(Type=Bone,Name="MSTR_fur_r")
    Value=(Rotation=(X=0.4981,Y=0.0755,Z=-0.0436,W=0.8627),Translation=(X=0,Y=-2,Z=8),Scale3D=(X=1,Y=1,Z=1))
    bParentInitial=False
    bPropagateToChildren=True
    Weight=@n13.OutputParam
   ) {template="Set Relative Transform::Execute(in Child,in Parent,in bParentInitial,in Value,in Weight,in bPropagateToChildren)"}
   n14 = &RigVMFunction_ControlFlowBranch(ExecuteContext=@n11.B, Condition=RigVMVariableNode() {variable=emotion}.Value)
   &RigUnit_SphericalPoseReader(
    ActiveRegionScaleFactors=(PositiveWidth=1,NegativeWidth=1,PositiveHeight=1,NegativeHeight=1)
    ActiveRegionSize=0.1
    Debug=(bDrawDebug=true,bDraw2D=False,bDrawLocalAxes=False,DebugScale=25,DebugSegments=20,DebugThickness=0.25)
    DriverAxis=(X=-1,Y=0,Z=0)
    DriverItem=(Type=Bone,Name="upperarm_r")
    FalloffRegionScaleFactors=(PositiveWidth=1,NegativeWidth=1,PositiveHeight=1,NegativeHeight=1)
    FalloffSize=0.2
    FlipHeightScaling=False
    FlipWidthScaling=False
    OptionalParentItem=(Type=Bone,Name="None")
    RotationOffset=(X=-25,Y=150,Z=0)
    ExecutePin=@n14.True
   )
   > &RigUnit_ParentConstraint(
    AdvancedSettings=(RotationOrderForFilter=XZY)
    Child=(Type=Bone,Name="MSTR_fur_r")
    Parents="((Item=(Type=Bone,Name=\"spine_04\"),Weight=0.0409),(Item=(Type=Bone,Name=\"upperarm_twist_02_r\"),Weight=0.0473),(Item=(Type=Bone,Name=\"deform_pec_r\"),Weight=0.0503),(Item=(Type=Bone,Name=\"spine_05\"),Weight=0.1107),(Item=(Type=Bone,Name=\"deform_upperarm_fr_r\" ...(truncated)"
    Weight=1
   )
   n15 = &RigUnit_SphericalPoseReader(
    ActiveRegionScaleFactors=(PositiveWidth=1,NegativeWidth=1,PositiveHeight=1,NegativeHeight=1)
    ActiveRegionSize=0.1
    Debug=(bDrawDebug=false,bDraw2D=False,bDrawLocalAxes=False,DebugScale=25,DebugSegments=20,DebugThickness=0.25)
    DriverAxis=(X=-1,Y=0,Z=0)
    DriverItem=(Type=Bone,Name="upperarm_r")
    FalloffRegionScaleFactors=(PositiveWidth=1,NegativeWidth=1,PositiveHeight=1,NegativeHeight=1)
    FalloffSize=0.25
    FlipHeightScaling=False
    FlipWidthScaling=False
    OptionalParentItem=(Type=Bone,Name="None")
    RotationOffset=(X=100,Y=0,Z=50)
    ExecutePin=@RigVMFunction_Sequence_1.C
   )
   n16 = &RigVMFunction_AlphaInterp(InterpSpeedDecreasing=10, InterpSpeedIncreasing=10, Scale=1, bClampResult=false, bMapRange=false, Value=@n15.OutputParam)
   > &RigUnit_SetRelativeTransformForItem(
    Child=(Type=Bone,Name="MSTR_cape_r")
    Parent=(Type=Bone,Name="MSTR_cape_r")
    Value=(Rotation=(X=0,Y=0,Z=-0.342,W=0.9397),Translation=(X=0,Y=-8,Z=0),Scale3D=(X=1,Y=1,Z=1))
    bParentInitial=False
    bPropagateToChildren=True
    Weight=@n16.Result
   ) {template="Set Relative Transform::Execute(in Child,in Parent,in bParentInitial,in Value,in Weight,in bPropagateToChildren)"}
   > n17 = &RigUnit_SphericalPoseReader(
    ActiveRegionScaleFactors=(PositiveWidth=1,NegativeWidth=1,PositiveHeight=1,NegativeHeight=1)
    ActiveRegionSize=0.1
    Debug=(bDrawDebug=false,bDraw2D=False,bDrawLocalAxes=False,DebugScale=25,DebugSegments=20,DebugThickness=0.25)
    DriverAxis=(X=1,Y=0,Z=0)
    DriverItem=(Type=Bone,Name="upperarm_l")
    FalloffRegionScaleFactors=(PositiveWidth=1,NegativeWidth=1,PositiveHeight=1,NegativeHeight=1)
    FalloffSize=0.25
    FlipHeightScaling=False
    FlipWidthScaling=False
    OptionalParentItem=(Type=Bone,Name="None")
    RotationOffset=(X=-35,Y=-80,Z=0)
   )
   n18 = &RigVMFunction_AlphaInterp(InterpSpeedDecreasing=10, InterpSpeedIncreasing=10, Scale=1, bClampResult=false, bMapRange=false, Value=@n17.OutputParam)
   > &RigUnit_SetRelativeTransformForItem(
    Child=(Type=Bone,Name="MSTR_cape_l")
    Parent=(Type=Bone,Name="MSTR_cape_l")
    Value=(Rotation=(X=0,Y=0,Z=-0.342,W=0.9397),Translation=(X=0,Y=-8,Z=0),Scale3D=(X=1,Y=1,Z=1))
    bParentInitial=False
    bPropagateToChildren=True
    Weight=@n18.Result
   ) {template="Set Relative Transform::Execute(in Child,in Parent,in bParentInitial,in Value,in Weight,in bPropagateToChildren)"}
   &RigUnit_PrepareForExecution()
   &RigUnit_SetRelativeTransformForItem(
    Child=(Type=Bone,Name="MSTR_fur_r")
    ExecutePin=()
    Parent=(Type=Bone,Name="MSTR_fur_r")
    Value=(Rotation=(X=0,Y=0,Z=0,W=1),Translation=(X=-2,Y=2,Z=0),Scale3D=(X=1,Y=1,Z=1))
    Weight=1
    bParentInitial=False
    bPropagateToChildren=True
   ) {template="Set Relative Transform::Execute(in Child,in Parent,in bParentInitial,in Value,in Weight,in bPropagateToChildren)"}
   &RigUnit_SetRelativeTransformForItem(
    Child=(Type=Bone,Name="MSTR_skirt_back_r")
    ExecutePin=()
    Parent=(Type=Bone,Name="MSTR_skirt_back_r")
    Value=(Rotation=(X=0.4957,Y=-0.0653,Z=-0.113,W=0.8586),Translation=(X=0,Y=0,Z=-9),Scale3D=(X=1,Y=1,Z=1))
    bParentInitial=False
    bPropagateToChildren=True
    Weight=@n5.Value
   ) {template="Set Relative Transform::Execute(in Child,in Parent,in bParentInitial,in Value,in Weight,in bPropagateToChildren)"}
  ]}
]
hierarchy [
 root {loc=<0 0 0>, rot=<0 0 0 1>}
 pelvis {parent=root, loc=<0 -0.7071 93.1191>, rot=<0.0127 0.707 -0.0127 -0.707>}
 spine_01 {parent=pelvis, loc=<6.7494 -0.6388 0>, rot=<0 0 0.034 -0.9994>}
 spine_02 {parent=spine_01, loc=<7.2518 -0.035 0>, rot=<0 0 0.0238 -0.9997>}
 spine_03 {parent=spine_02, loc=<6.0104 0 0>, rot=<0 0 0.0149 -0.9999>}
 spine_04 {parent=spine_03, loc=<8.1024 0 0>, rot=<0 0 -0.0132 -0.9999>}
 spine_05 {parent=spine_04, loc=<10.0146 0 0>, rot=<0 0 -0.0015 -1>}
 neck_01 {parent=spine_05, loc=<11.3138 0.1529 0>, rot=<0 0 -0.2003 -0.9797>}
 neck_02 {parent=neck_01, loc=<4.5627 0 0>, rot=<0 0 0.0504 -0.9987>}
 head {parent=neck_02, loc=<5.3914 0 0>, rot=<0 0 0.115 -0.9934>}
 MSTR_hair_l {parent=head, loc=<4.905 0.2861 -7.6753>, rot=<0.0041 -1 0 0>}
 dyn_hair_01_l {parent=MSTR_hair_l, loc=<0 0 0>, rot=<0 0 0 1>}
 dyn_hair_02_l {parent=dyn_hair_01_l, loc=<4.3063 0 0>, rot=<0 0 0 1>}
 MSTR_hair_r {parent=head, loc=<4.905 0.2861 7.8451>, rot=<0.0041 -1 0 0>}
 dyn_hair_01_r {parent=MSTR_hair_r, loc=<0 0 0>, rot=<0 0 0 1>}
 dyn_hair_02_r {parent=dyn_hair_01_r, loc=<4.5743 0 0>, rot=<0 0 0 1>}
 MSTR_hair_c {parent=head, loc=<5.6142 -7.0804 0>, rot=<0.0041 -1 0 0>}
 dyn_hair_01_c {parent=MSTR_hair_c, loc=<0 0 0>, rot=<0 0 0 1>}
 dyn_hair_02_c {parent=dyn_hair_01_c, loc=<4.9337 0 0>, rot=<0 0 0 1>}
 attach_cape {parent=spine_05, loc=<9.017 -11.714 0>, rot=<0.0616 -0.7044 -0.0616 -0.7044>}
 cape {parent=attach_cape, type=Null, loc=<0 0 0>, rot=<0 0 0 1>}
 attach_backpack {parent=spine_05, loc=<-0.123 -13.473 0>, rot=<-0.0093 -0.707 0.0093 -0.707>}
 Backpack_BR {parent=attach_backpack, type=Null, loc=<0 -8.5869 0.9171>, rot=<0 0 0 1>}
 clavicle_l {parent=spine_05, loc=<4.3575 -0.1466 -2.3223>, rot=<0.0465 -0.7289 -0.009 -0.6829>}
 upperarm_l {parent=clavicle_l, loc=<15.5111 0 0>, rot=<0.0226 -0.3234 -0.0088 -0.9459>}
 lowerarm_l {parent=upperarm_l, loc=<26.6514 0 0>, rot=<0.0391 -0.0099 -0.2453 -0.9686>}
 hand_l {parent=lowerarm_l, loc=<22.3491 0 0>, rot=<-0.7081 -0.0507 0.0975 -0.6975>}
 thumb_01_l {parent=hand_l, loc=<1.3118 -0.3191 -2.459>, rot=<0.6546 -0.4042 0.0664 -0.6354>}
 thumb_02_l {parent=thumb_01_l, loc=<5.9336 -0.0448 0.0401>, rot=<-0.0037 0.0213 0.298 -0.9543>}
 thumb_03_l {parent=thumb_02_l, loc=<2.2223 0.3682 -0.0907>, rot=<0.0118 -0.013 0.0066 -0.9998>}
 index_metacarpal_l {parent=hand_l, loc=<1.6535 0.0183 -2.96>, rot=<0.0158 -0.0124 -0.0173 -0.9996>}
 index_01_l {parent=index_metacarpal_l, loc=<8.3008 0.0357 -0.0524>, rot=<-0.0044 0.0485 0.0924 -0.9945>}
 index_02_l {parent=index_01_l, loc=<3.3587 -0.3004 0.0452>, rot=<-0.002 0.0027 0.3225 -0.9466>}
 index_03_l {parent=index_02_l, loc=<2.5578 -0.123 -0.0252>, rot=<-0.002 0.0067 0.2558 -0.9667>}
 middle_metacarpal_l {parent=hand_l, loc=<2.1856 -0.1725 -0.5984>, rot=<0.006 0.0168 -0.0329 -0.9993>}
 middle_01_l {parent=middle_metacarpal_l, loc=<7.3275 -0.2447 0.0129>, rot=<-0.0208 0.0042 0.0858 -0.9961>}
 middle_02_l {parent=middle_01_l, loc=<4.0123 0.1401 0.1175>, rot=<-0.0218 0.0258 0.3512 -0.9357>}
 middle_03_l {parent=middle_02_l, loc=<2.6077 0.0534 0.1373>, rot=<0.0082 0.0188 0.1708 -0.9851>}
 ring_metacarpal_l {parent=hand_l, loc=<2.2074 -0.3625 1.3638>, rot=<-0.0284 0.0722 -0.0458 -0.9959>}
 ring_01_l {parent=ring_metacarpal_l, loc=<7.0295 -0.6197 -0.1446>, rot=<-0.0489 -0.0252 0.1892 -0.9804>}
 ring_02_l {parent=ring_01_l, loc=<3.6174 0.4647 -0.1027>, rot=<0 0 0.2806 -0.9598>}
 ring_03_l {parent=ring_02_l, loc=<2.4152 0.114 0.0199>, rot=<0 0 0.1189 -0.9929>}
 pinky_metacarpal_l {parent=hand_l, loc=<1.7616 -0.9753 3.3814>, rot=<-0.0397 0.0814 -0.069 -0.9935>}
 pinky_01_l {parent=pinky_metacarpal_l, loc=<7.1709 -0.4831 -0.1094>, rot=<-0.0654 -0.0035 0.2242 -0.9723>}
 pinky_02_l {parent=pinky_01_l, loc=<2.8274 0.0893 -0.1043>, rot=<0.0037 0.0011 0.2854 -0.9584>}
 pinky_03_l {parent=pinky_02_l, loc=<2.2046 0.0307 -0.0078>, rot=<-0.0002 0.0098 0.1445 -0.9895>}
 deform_wrist_dn_base_l {parent=hand_l, loc=<0 -2.2342 0>, rot=<0 0 0 -1>}
 deform_wrist_dn_l {parent=deform_wrist_dn_base_l, loc=<0 0 0>, rot=<0 0 0 1>}
 deform_wrist_up_base_l {parent=hand_l, loc=<0 2.2682 0>, rot=<0 0 0 -1>}
 deform_wrist_up_l {parent=deform_wrist_up_base_l, loc=<0 0 0>, rot=<0 0 0 1>}
 weapon_l {parent=hand_l, loc=<-0.1974 -3.4111 0.3811>, rot=<0 0 0 1>}
 hand_attach_l {parent=hand_l, loc=<8.5 -3 0>, rot=<0.7071 0.7071 0 0>}
 lowerarm_twist_02_l {parent=lowerarm_l, loc=<9.6199 0 0>, rot=<0 0 0 -1>}
 lowerarm_twist_01_l {parent=lowerarm_l, loc=<17.8549 0 0>, rot=<0 0 0 -1>}
 deform_elbow_in_base_l {parent=lowerarm_l, loc=<0.2112 5.259 0.8274>, rot=<0 0 0 -1>}
 deform_elbow_in_l {parent=deform_elbow_in_base_l, loc=<0 0 0>, rot=<0 0 0 1>}
 deform_elbow_out_base_l {parent=lowerarm_l, loc=<0 -4 0>, rot=<0 0 0 -1>}
 deform_elbow_out_l {parent=deform_elbow_out_base_l, loc=<0 0 0>, rot=<0 0 0 1>}
 upperarm_twist_01_l {parent=upperarm_l, loc=<0.0041 0 0>, rot=<0 0 0 -1>}
 deform_upperarm_bk_base_l {parent=upperarm_twist_01_l, loc=<1.0813 -7.2193 -1.9453>, rot=<0 0 0 1>}
 deform_upperarm_bk_l {parent=deform_upperarm_bk_base_l, loc=<0 0 0>, rot=<0 0 0 1>}
 deform_upperarm_fr_base_l {parent=upperarm_twist_01_l, loc=<-0.3691 7.0506 -2.765>, rot=<0 0 0 1>}
 deform_upperarm_fr_l {parent=deform_upperarm_fr_base_l, loc=<0 0 0>, rot=<0 0 0 1>}
 deform_upperarm_up_base_l {parent=upperarm_twist_01_l, loc=<-3.2216 -1.1391 3.3311>, rot=<0 0 0 1>}
 deform_upperarm_up_l {parent=deform_upperarm_up_base_l, loc=<0 0 0>, rot=<0 0 0 1>}
 upperarm_twist_02_l {parent=upperarm_l, loc=<13.7941 0 0>, rot=<0 0 0 -1>}
 clavicle_r {parent=spine_05, loc=<4.3575 -0.1466 2.3223>, rot=<0.7289 0.0465 -0.6829 0.009>}
 upperarm_r {parent=clavicle_r, loc=<-15.5111 0 0>, rot=<0.0226 -0.3234 -0.0088 -0.9459>}
 lowerarm_r {parent=upperarm_r, loc=<-26.6514 0 0>, rot=<0.0391 -0.0099 -0.2453 -0.9686>}
 hand_r {parent=lowerarm_r, loc=<-22.3491 0 0>, rot=<-0.7081 -0.0507 0.0975 -0.6975>}
 thumb_01_r {parent=hand_r, loc=<-1.3118 0.3191 2.459>, rot=<0.6546 -0.4042 0.0664 -0.6354>}
 thumb_02_r {parent=thumb_01_r, loc=<-5.9336 0.0448 -0.0401>, rot=<-0.0037 0.0213 0.298 -0.9543>}
 thumb_03_r {parent=thumb_02_r, loc=<-2.2223 -0.3682 0.0907>, rot=<0.0118 -0.013 0.0066 -0.9998>}
 index_metacarpal_r {parent=hand_r, loc=<-1.6535 -0.0183 2.96>, rot=<0.0158 -0.0124 -0.0173 -0.9996>}
 index_01_r {parent=index_metacarpal_r, loc=<-8.3008 -0.0357 0.0524>, rot=<-0.0044 0.0485 0.0924 -0.9945>}
 index_02_r {parent=index_01_r, loc=<-3.3587 0.3004 -0.0452>, rot=<-0.002 0.0027 0.3225 -0.9466>}
 index_03_r {parent=index_02_r, loc=<-2.5578 0.123 0.0252>, rot=<-0.002 0.0067 0.2558 -0.9667>}
 middle_metacarpal_r {parent=hand_r, loc=<-2.1856 0.1725 0.5984>, rot=<0.006 0.0168 -0.0329 -0.9993>}
 middle_01_r {parent=middle_metacarpal_r, loc=<-7.3275 0.2447 -0.0129>, rot=<-0.0208 0.0042 0.0858 -0.9961>}
 middle_02_r {parent=middle_01_r, loc=<-4.0123 -0.1401 -0.1175>, rot=<-0.0218 0.0258 0.3512 -0.9357>}
 middle_03_r {parent=middle_02_r, loc=<-2.6077 -0.0534 -0.1373>, rot=<0.0082 0.0188 0.1708 -0.9851>}
 ring_metacarpal_r {parent=hand_r, loc=<-2.2074 0.3625 -1.3638>, rot=<-0.0284 0.0722 -0.0458 -0.9959>}
 ring_01_r {parent=ring_metacarpal_r, loc=<-7.0295 0.6197 0.1446>, rot=<-0.0489 -0.0252 0.1892 -0.9804>}
 ring_02_r {parent=ring_01_r, loc=<-3.6174 -0.4647 0.1027>, rot=<0 0 0.2806 -0.9598>}
 ring_03_r {parent=ring_02_r, loc=<-2.4152 -0.114 -0.0199>, rot=<0 0 0.1189 -0.9929>}
 pinky_metacarpal_r {parent=hand_r, loc=<-1.7616 0.9753 -3.3814>, rot=<-0.0397 0.0814 -0.069 -0.9935>}
 pinky_01_r {parent=pinky_metacarpal_r, loc=<-7.1709 0.4831 0.1094>, rot=<-0.0654 -0.0035 0.2242 -0.9723>}
 pinky_02_r {parent=pinky_01_r, loc=<-2.8274 -0.0893 0.1043>, rot=<0.0037 0.0011 0.2854 -0.9584>}
 pinky_03_r {parent=pinky_02_r, loc=<-2.2046 -0.0307 0.0078>, rot=<-0.0002 0.0098 0.1445 -0.9895>}
 deform_wrist_dn_base_r {parent=hand_r, loc=<-0.0004 2.2345 -0.0001>, rot=<0 0 0 -1>}
 deform_wrist_dn_r {parent=deform_wrist_dn_base_r, loc=<0 0 0>, rot=<0 0 0 1>}
 deform_wrist_up_base_r {parent=hand_r, loc=<0.0003 -2.2684 0.0001>, rot=<0 0 0 -1>}
 deform_wrist_up_r {parent=deform_wrist_up_base_r, loc=<0 0 0>, rot=<0 0 0 1>}
 weapon_r {parent=hand_r, loc=<0.1974 3.4111 -0.3811>, rot=<0 0 0 1>}
 hand_attach_r {parent=hand_r, loc=<-8.5001 2.9999 0>, rot=<0 0 -0.7071 -0.7071>}
 lowerarm_twist_02_r {parent=lowerarm_r, loc=<-9.6199 0 0>, rot=<0 0 0 -1>}
 lowerarm_twist_01_r {parent=lowerarm_r, loc=<-17.8549 0 0>, rot=<0 0 0 -1>}
 deform_elbow_in_base_r {parent=lowerarm_r, loc=<-0.2111 -5.2591 -0.8275>, rot=<0 0 0 -1>}
 deform_elbow_in_r {parent=deform_elbow_in_base_r, loc=<0 0 0>, rot=<0 0 0 1>}
 deform_elbow_out_base_r {parent=lowerarm_r, loc=<0 4 -0.0001>, rot=<0 0 0 -1>}
 deform_elbow_out_r {parent=deform_elbow_out_base_r, loc=<0 0 0>, rot=<0 0 0 1>}
 upperarm_twist_01_r {parent=upperarm_r, loc=<-0.0041 0 0>, rot=<0 0 0 -1>}
 deform_upperarm_bk_base_r {parent=upperarm_twist_01_r, loc=<-1.0814 7.2192 1.9455>, rot=<0 0 0 1>}
 deform_upperarm_bk_r {parent=deform_upperarm_bk_base_r, loc=<0 0 0>, rot=<0 0 0 1>}
 deform_upperarm_fr_base_r {parent=upperarm_twist_01_r, loc=<0.3689 -7.0506 2.7652>, rot=<0 0 0 1>}
 deform_upperarm_fr_r {parent=deform_upperarm_fr_base_r, loc=<0 0 0>, rot=<0 0 0 1>}
 deform_upperarm_up_base_r {parent=upperarm_twist_01_r, loc=<3.2214 1.1391 -3.3308>, rot=<0 0 0 1>}
 deform_upperarm_up_r {parent=deform_upperarm_up_base_r, loc=<0 0 0>, rot=<0 0 0 -1>}
 upperarm_twist_02_r {parent=upperarm_r, loc=<-13.7941 0 0>, rot=<0 0 0 -1>}
 joint1 {parent=clavicle_r, loc=<-12.1288 -8.7056 -5.8606>, rot=<0.9996 0.0252 0.0122 0.0003>}
 MSTR_fur_r {parent=clavicle_r, loc=<-12.1288 -8.7056 -5.8606>, rot=<0.9996 0.0252 0.0122 0.0003>}
 MSTR_cape_r {parent=spine_05, loc=<3.7805 -13.441 12.9048>, rot=<0.0847 -0.996 0.0024 0.0282>}
 dyn_cape_01_r {parent=MSTR_cape_r, loc=<0 0 0>, rot=<0 0 0 1>}
 dyn_cape_02_r {parent=dyn_cape_01_r, loc=<20.3179 0 0>, rot=<0 0 0 1>}
 dyn_cape_03_r {parent=dyn_cape_02_r, loc=<20.3178 0 0>, rot=<0 0 0 1>}
 dyn_cape_04_r {parent=dyn_cape_03_r, loc=<21.9068 0 0>, rot=<0 0 0 1>}
 dyn_cape_05_r {parent=dyn_cape_04_r, loc=<17.8199 0 0>, rot=<0 0 0 1>}
 dyn_fur_back_r {parent=dyn_cape_01_r, loc=<9.6885 -3.7451 -0.1789>, rot=<0.0013 0.0283 -0.0447 -0.9986>}
 MSTR_cape_c {parent=spine_05, loc=<3.7805 -13.441 -0.057>, rot=<0.0848 -0.9964 0 0>}
 dyn_cape_01_c {parent=MSTR_cape_c, loc=<0 0 0>, rot=<0 0 0 1>}
 dyn_cape_02_c {parent=dyn_cape_01_c, loc=<20.3179 0 0>, rot=<0 0 0 1>}
 dyn_cape_03_c {parent=dyn_cape_02_c, loc=<20.3178 0 0>, rot=<0 0 0 1>}
 dyn_cape_04_c {parent=dyn_cape_03_c, loc=<21.9068 0 0>, rot=<0 0 0 1>}
 dyn_cape_05_c {parent=dyn_cape_04_c, loc=<17.8199 0 0>, rot=<0 0 0 1>}
 dyn_fur_back_c {parent=dyn_cape_01_c, loc=<9.6628 -3.7451 -0.057>, rot=<0 0 -0.0447 -0.999>}
 MSTR_cape_l {parent=spine_05, loc=<3.7805 -13.441 -12.1217>, rot=<0.0847 -0.9959 -0.0027 -0.0315>}
 dyn_cape_01_l {parent=MSTR_cape_l, loc=<0 0 0>, rot=<0 0 0 1>}
 dyn_cape_02_l {parent=dyn_cape_01_l, loc=<20.3179 0 0>, rot=<0 0 0 1>}
 dyn_cape_03_l {parent=dyn_cape_02_l, loc=<20.3178 0 0>, rot=<0 0 0 1>}
 dyn_cape_04_l {parent=dyn_cape_03_l, loc=<21.9068 0 0>, rot=<0 0 0 1>}
 dyn_cape_05_l {parent=dyn_cape_04_l, loc=<17.8199 0 0>, rot=<0 0 0 1>}
 dyn_fur_back_l {parent=dyn_cape_01_l, loc=<9.6746 -3.7451 -0.1203>, rot=<-0.0014 -0.0316 -0.0447 -0.9985>}
 deform_pec_aim_base_l {parent=spine_04, loc=<7.8732 12.6388 -2.4123>, rot=<0.0102 -0.4602 0.3979 -0.7936>}
 deform_pec_aim_l {parent=deform_pec_aim_base_l, loc=<0 0 0>, rot=<0 0 0 1>}
 deform_pec_base_l {parent=deform_pec_aim_l, loc=<10.766 0 0>, rot=<0 0 0 1>}
 deform_pec_l {parent=deform_pec_base_l, loc=<0.055 0 0>, rot=<0 0 0 1>}
 deform_pec_aim_base_r {parent=spine_04, loc=<7.8727 12.6388 2.4123>, rot=<-0.4602 -0.0102 0.7936 0.3979>}
 deform_pec_aim_r {parent=deform_pec_aim_base_r, loc=<0 0 0>, rot=<0 0 0 1>}
 deform_pec_base_r {parent=deform_pec_aim_r, loc=<-10.7662 -0.0006 -0.0007>, rot=<0 0 0 1>}
 deform_pec_r {parent=deform_pec_base_r, loc=<0 0 0>, rot=<0 0 0 1>}
 MSTR_sword_l {parent=spine_01, loc=<0.952 3.427 -16.609>, rot=<0.2066 -0.9605 -0.1612 -0.0941>}
 sword_hidebone_l {parent=MSTR_sword_l, loc=<-15.0157 -0.2126 -0.8173>, rot=<-0.0165 -0.0092 0.0112 -0.9998>}
 MSTR_sword_r {parent=spine_01, loc=<-0.7651 -0.7738 17.5348>, rot=<0.1577 -0.9781 -0.1056 0.0857>}
 sword_hidebone_r {parent=MSTR_sword_r, loc=<0 0 0>, rot=<0 0 0 1>}
 MSTR_belt {parent=spine_01, loc=<3.0785 8.8877 8.0499>, rot=<0.0853 0.7019 0.7051 0.0537>}
 dyn_belt_01 {parent=MSTR_belt, loc=<0 0 0>, rot=<0 0 0 1>}
 dyn_belt_02 {parent=dyn_belt_01, loc=<7.6046 0 0>, rot=<-0.0008 -0.0203 0.0047 -0.9998>}
 dyn_belt_03 {parent=dyn_belt_02, loc=<10.6215 0 0>, rot=<-0.0016 -0.0422 -0.0019 -0.9991>}
 dyn_belt_04 {parent=dyn_belt_03, loc=<14.0246 0 0>, rot=<-0.0008 -0.0229 0.0033 -0.9997>}
 dyn_belt_05 {parent=dyn_belt_04, loc=<14.2628 0 0>, rot=<0 0 0 1>}
 thigh_l {parent=pelvis, loc=<-2.5693 -0.0109 -7.7501>, rot=<0.0012 0.0572 0.0155 -0.9982>}
 calf_l {parent=thigh_l, loc=<-39.0484 0 0>, rot=<0.0006 -0.0161 -0.0391 -0.9991>}
 foot_l {parent=calf_l, loc=<-42.561 0.1077 0.4292>, rot=<0.0418 -0.0455 0.0343 -0.9975>}
 ball_l {parent=foot_l, loc=<-6.1599 13.8264 0.6806>, rot=<0.0001 -0.0001 -0.7063 -0.7079>}
 calf_twist_02_l {parent=calf_l, loc=<-18.8064 -0.0967 0.0397>, rot=<0.0237 0.0095 -0.0035 -0.9997>}
 calf_twist_01_l {parent=calf_l, loc=<-35.813 -0.0967 0.0397>, rot=<0.0237 0.0095 -0.0035 -0.9997>}
 deform_knee_in_base_l {parent=calf_l, loc=<1.298 -5.2957 -0.0496>, rot=<0 0 0 -1>}
 deform_knee_in_l {parent=deform_knee_in_base_l, loc=<0 0 0>, rot=<0 0 0 1>}
 deform_knee_out_base_l {parent=calf_l, loc=<0 4 0>, rot=<0 0 0 -1>}
 deform_knee_out_l {parent=deform_knee_out_base_l, loc=<0 0 0>, rot=<0 0 0 1>}
 thigh_twist_01_l {parent=thigh_l, loc=<-21.6303 0 0>, rot=<0.0013 0 0 -1>}
 deform_groin_base_l {parent=thigh_l, loc=<-2.1908 8.1142 -1.7442>, rot=<0 0 0 -1>}
 deform_groin_l {parent=deform_groin_base_l, loc=<0 0 0>, rot=<0.9932 -0.0048 -0.1141 -0.0213>}
 MSTR_skirt_front_l {parent=thigh_l, loc=<-10.6969 9.2786 -1.7875>, rot=<-0.0498 -0.9986 0.0008 0.0203>}
 dyn_skirt_01_front_l {parent=MSTR_skirt_front_l, loc=<0 0 0>, rot=<0 0 0 1>}
 dyn_skirt_02_front_l {parent=dyn_skirt_01_front_l, loc=<8.4747 0 0>, rot=<0 0 0 1>}
 dyn_skirt_03_front_l {parent=dyn_skirt_02_front_l, loc=<10.1146 0 0>, rot=<0 0 0 1>}
 MSTR_skirt_side_l {parent=thigh_l, loc=<-9.7853 -2.2801 -9.4995>, rot=<0.0023 -0.7079 0.7057 -0.03>}
 dyn_skirt_01_side_l {parent=MSTR_skirt_side_l, loc=<0 0 0>, rot=<0 0 0 1>}
 dyn_skirt_02_side_l {parent=dyn_skirt_01_side_l, loc=<8.4747 0 0>, rot=<0 0 0 1>}
 dyn_skirt_03_side_l {parent=dyn_skirt_02_side_l, loc=<10.1146 0 0>, rot=<0 0 0 1>}
 MSTR_skirt_back_l {parent=thigh_l, loc=<-10.1794 -14.61 -0.604>, rot=<0.0655 -0.9978 -0.0025 0.0051>}
 dyn_skirt_01_back_l {parent=MSTR_skirt_back_l, loc=<0 0 0>, rot=<0 0 0 1>}
 dyn_skirt_02_back_l {parent=dyn_skirt_01_back_l, loc=<8.4747 0 0>, rot=<0 0 0 1>}
 dyn_skirt_03_back_l {parent=dyn_skirt_02_back_l, loc=<10.1146 0 0>, rot=<0 0 0 1>}
 thigh_r {parent=pelvis, loc=<-2.5693 -0.0109 7.7501>, rot=<0.0572 -0.0012 0.9982 0.0155>}
 calf_r {parent=thigh_r, loc=<39.0484 0 0>, rot=<0.0006 -0.0161 -0.0391 -0.9991>}
 foot_r {parent=calf_r, loc=<42.561 -0.1077 -0.4292>, rot=<0.0418 -0.0455 0.0343 -0.9975>}
 ball_r {parent=foot_r, loc=<6.1599 -13.8264 -0.6806>, rot=<0.0001 -0.0001 -0.7063 -0.7079>}
 calf_twist_02_r {parent=calf_r, loc=<18.8064 0.0967 -0.0397>, rot=<0.0237 0.0095 -0.0035 -0.9997>}
 calf_twist_01_r {parent=calf_r, loc=<35.813 0.0967 -0.0397>, rot=<0.0237 0.0095 -0.0035 -0.9997>}
 deform_knee_in_base_r {parent=calf_r, loc=<-1.298 5.2957 0.0496>, rot=<0 0 0 -1>}
 deform_knee_in_r {parent=deform_knee_in_base_r, loc=<0 0 0>, rot=<0 0 0 1>}
 deform_knee_out_base_r {parent=calf_r, loc=<0 -4 0>, rot=<0 0 0 -1>}
 deform_knee_out_r {parent=deform_knee_out_base_r, loc=<0 0 0>, rot=<0 0 0 1>}
 thigh_twist_01_r {parent=thigh_r, loc=<21.6303 0 0>, rot=<0.0013 0 0 -1>}
 deform_groin_base_r {parent=thigh_r, loc=<2.1908 -8.1142 1.7442>, rot=<0 0 0 -1>}
 deform_groin_r {parent=deform_groin_base_r, loc=<0 0 0>, rot=<0 0 0 1>}
 MSTR_skirt_side_r {parent=thigh_r, loc=<9.6525 2.281 8.3436>, rot=<0.7028 0.0365 0.0049 -0.7105>}
 dyn_skirt_01_side_r {parent=MSTR_skirt_side_r, loc=<0 0 0>, rot=<0 0 0 1>}
 dyn_skirt_02_side_r {parent=dyn_skirt_01_side_r, loc=<8.4747 0 0>, rot=<0 0 0 1>}
 dyn_skirt_03_side_r {parent=dyn_skirt_02_side_r, loc=<10.1146 0 0>, rot=<0 0 0 1>}
 MSTR_skirt_front_r {parent=thigh_r, loc=<10.6905 -9.2785 1.7313>, rot=<0.9943 -0.0496 -0.0939 0.0042>}
 dyn_skirt_01_front_r {parent=MSTR_skirt_front_r, loc=<0 0 0>, rot=<0 0 0 1>}
 dyn_skirt_02_front_r {parent=dyn_skirt_01_front_r, loc=<8.4747 0 0>, rot=<0 0 0 1>}
 dyn_skirt_03_front_r {parent=dyn_skirt_02_front_r, loc=<10.1146 0 0>, rot=<0 0 0 1>}
 MSTR_skirt_back_r {parent=thigh_r, loc=<10.3767 14.6086 3.1227>, rot=<0.9987 0.0494 0.0025 -0.013>}
 dyn_skirt_01_back_r {parent=MSTR_skirt_back_r, loc=<0 0 0>, rot=<0 0 0 1>}
 dyn_skirt_02_back_r {parent=dyn_skirt_01_back_r, loc=<8.4747 0 0>, rot=<0 0 0 1>}
 dyn_skirt_03_back_r {parent=dyn_skirt_02_back_r, loc=<10.1146 0 0>, rot=<0 0 0 1>}
 deform_glute_base_l {parent=pelvis, loc=<-8.8064 -8.8222 -8.4287>, rot=<0.0275 0.0501 -0.4472 -0.8926>}
 deform_glute_l {parent=deform_glute_base_l, loc=<0 0 0>, rot=<0 0 0 1>}
 deform_glute_base_r {parent=pelvis, loc=<-8.8062 -8.8222 8.4287>, rot=<0.0501 -0.0275 0.8926 -0.4472>}
 deform_glute_r {parent=deform_glute_base_r, loc=<0 0 0>, rot=<0 0 0 1>}
 attach {parent=root, loc=<0 0 0>, rot=<0 0 0 1>}
 ik_hand_root {parent=root, loc=<0 0 0>, rot=<0 0 0 1>}
 ik_hand_gun {parent=ik_hand_root, loc=<-51.9727 4.0883 102.6935>, rot=<0.6084 0.1793 0.4613 -0.6205>}
 ik_hand_r {parent=ik_hand_gun, loc=<0 0 0>, rot=<0 0 0 1>}
 ik_hand_l {parent=ik_hand_gun, loc=<53.0285 82.1781 35.2051>, rot=<-0.5102 -0.7906 -0.3387 0>}
 ik_foot_root {parent=root, loc=<0 0 0>, rot=<0 0 0 1>}
 ik_foot_l {parent=ik_foot_root, loc=<15.2608 -4.4176 9.4432>, rot=<0.0357 0.7045 0.0284 -0.7082>}
 ik_foot_r {parent=ik_foot_root, loc=<-15.2608 -4.4176 9.4432>, rot=<0.7082 0.0284 -0.7045 0.0357>}
 root_fp {parent=root, loc=<0 0 0>, rot=<0 0 0 1>}
 attach_fp {parent=root_fp, loc=<0 0 0>, rot=<0 0 0 1>}
 camera_root {parent=root_fp, loc=<0 0 0>, rot=<0 0 0 1>}
 camera {parent=camera_root, loc=<0 0 0>, rot=<0 0 0 1>}
 attach_emote_01 {parent=root, loc=<0 0 0>, rot=<0 0 0 1>}
 attach_emote_02 {parent=root, loc=<0 0 0>, rot=<0 0 0 1>}
 attach_emote_03 {parent=root, loc=<0 0 0>, rot=<0 0 0 1>}
 warp_point_01 {parent=root, loc=<0 0 0>, rot=<0 0 0 1>}
 warp_point_02 {parent=root, loc=<0 0 0>, rot=<0 0 0 1>}
 warp_point_03 {parent=root, loc=<0 0 0>, rot=<0 0 0 1>}
]
variables [
 GoldenGloryEmote {type=double, default=0}
 crouching {type=bool, default=False}
 crouchingAndMoving {type=bool, default=False}
 crouchingAndNotMoving {type=bool, default=False}
 emotion {type=bool, default=False}
 grindRail {type=bool, default=False}
 grindRailAlpha {type=float, default=0}
 jumpOrFalling {type=bool, default=False}
 jumpOrFallingAlpha {type=float, default=0}
 parachuteOpened {type=bool, default=False}
 swimming {type=bool, default=False}
]
