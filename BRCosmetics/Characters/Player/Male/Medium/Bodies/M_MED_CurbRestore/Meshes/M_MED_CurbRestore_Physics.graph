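# Physics asset for the M_MED_CurbRestore male player body cosmetic, defining collision and joint limits for a three-strand cape.
# Center, left and right five-bone cape chains hang from MSTR_cape anchors with limited swing and twist; free constraints link outer tips to the center tip.
# Capsules cover torso, legs, left arm, both swords and back skirt bones.
# A large root sphere, offset downward, serves as floor collision.
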PhysicsAsset /BRCosmetics/Characters/Player/Male/Medium/Bodies/M_MED_CurbRestore/Meshes/M_MED_CurbRestore_Physics
bodies {
 MSTR_cape_c [{shapeType=Capsule, length=0.2, radius=0.2, shapeName=MSTR_cape_c_capsule}]
 MSTR_cape_l [{shapeType=Capsule, length=0.2, radius=0.2, shapeName=MSTR_cape_l_capsule, rotation=<rot 0 -90 0>}]
 MSTR_cape_r [{shapeType=Capsule, length=0.2, radius=0.2, shapeName=MSTR_cape_r_capsule}]
 MSTR_sword_l [{
   shapeType=Capsule
   length=69.0191
   radius=7.7968
   shapeName=MSTR_sword_l_capsule
   center=<34.103 2.7263 4.4427>
   rotation=<rot 0 -90.6441 -89.6423>
  }]
 MSTR_sword_r [{
   shapeType=Capsule
   length=49.0023
   radius=7.8598
   shapeName=MSTR_sword_r_capsule
   center=<29.3377 2.3683 -3.7709>
   rotation=<rot 0.0113 -90.5776 -89.6385>
  }]
 calf_l [{shapeType=Capsule, length=39.5754, radius=10.7174, shapeName=calf_l_capsule, center=<-20.3365 -2.4168 0.7888>, rotation=<rot 90 -2.3153 0>}]
 calf_r [{shapeType=Capsule, length=39.5754, radius=10.7174, shapeName=calf_r_capsule, center=<20.3365 2.4168 -0.7888>, rotation=<rot 90 177.6847 0>}]
 dyn_cape_01_c [{shapeType=Capsule, length=16.343, radius=11.344, shapeName=dyn_cape_01_c_capsule, center=<13.654 -11.6226 0>, rotation=<rot 90 0 0>}]
 dyn_cape_01_l [{shapeType=Capsule, length=17.1536, radius=11.2251, shapeName=dyn_cape_01_l_capsule, center=<13.2298 -8.9948 1.3076>, rotation=<rot 90 0 0>}]
 dyn_cape_01_r [{shapeType=Capsule, length=17.1536, radius=11.2251, shapeName=dyn_cape_01_r_capsule, center=<13.2396 -9.4592 -3.0386>, rotation=<rot 90 0 0>}]
 dyn_cape_02_c [{shapeType=Capsule, length=15.2329, radius=10.4424, shapeName=dyn_cape_02_c_capsule, center=<17.5759 -11.2815 0>, rotation=<rot 90 0 0>}]
 dyn_cape_02_l [{shapeType=Capsule, length=12.8673, radius=11.4305, shapeName=dyn_cape_02_l_capsule, center=<18.2356 -8.9431 2.9348>, rotation=<rot 90 0 0>}]
 dyn_cape_02_r [{shapeType=Capsule, length=13.7519, radius=10.9698, shapeName=dyn_cape_02_r_capsule, center=<18.071 -8 -3.3007>, rotation=<rot 90 0 0>}]
 dyn_cape_03_c [{shapeType=Capsule, length=12.5204, radius=11.006, shapeName=dyn_cape_03_c_capsule, center=<12.4607 -7.9331 0>, rotation=<rot 90 2.3225 0>}]
 dyn_cape_03_l [{shapeType=Capsule, length=11.9328, radius=11.2739, shapeName=dyn_cape_03_l_capsule, center=<14.5602 -10.0054 3.6448>, rotation=<rot 90 0 0>}]
 dyn_cape_03_r [{shapeType=Capsule, length=10.6033, radius=11.416, shapeName=dyn_cape_03_r_capsule, center=<13.0417 -7.8122 -4.6974>, rotation=<rot 90 0 0>}]
 dyn_cape_04_c [{shapeType=Capsule, length=16.5855, radius=11.0043, shapeName=dyn_cape_04_c_capsule, center=<14.3563 -7.1406 0>, rotation=<rot 90 1.4579 0>}]
 dyn_cape_04_l [
  {
   shapeType=Capsule
   length=10.1244
   radius=10.7363
   shapeName=dyn_cape_04_l_capsule
   center=<17.143 -6.9423 6.2716>
   rotation=<rot 82.5573 92.4304 90>
  }
  {
   shapeType=Capsule
   length=13.7266
   radius=9.604
   shapeName=dyn_cape_04_l_capsule
   center=<17.5843 -5.8144 -0.7465>
   rotation=<rot 89.6895 0.5957 -1.0287>
  }
 ]
 dyn_cape_04_r [{
   shapeType=Capsule
   length=15.0645
   radius=14.1306
   shapeName=dyn_cape_04_r_capsule
   center=<15.6394 -8.6207 -4.4587>
   rotation=<rot 89.1921 0.0007 0.0005>
  }]
 dyn_cape_05_c [{shapeType=Capsule, length=16.5995, radius=8.0935, shapeName=dyn_cape_05_c_capsule, center=<16.058 -3.2361 0>}]
 dyn_cape_05_l [{
   shapeType=Capsule
   length=14.7721
   radius=8.4627
   shapeName=dyn_cape_05_l_capsule
   center=<16.1606 -1.4055 3.7759>
   rotation=<rot 5.5788 -90 -2.07>
  }]
 dyn_cape_05_r [{
   shapeType=Capsule
   length=11.7625
   radius=8.2857
   shapeName=dyn_cape_05_r_capsule
   center=<16.6831 -1.9171 -2.9594>
   rotation=<rot -5.7823 -89.9997 2.3771>
  }]
 dyn_skirt_02_back_l [{shapeType=Capsule, length=18.0125, radius=6.5886, shapeName=dyn_skirt_02_back_l_capsule, center=<6.465 1.5468 -4.0034>}]
 dyn_skirt_02_back_r [{
   shapeType=Capsule
   length=18.0125
   radius=6.5886
   shapeName=dyn_skirt_02_back_r_capsule
   center=<10.7326 4.88 5.1398>
   rotation=<rot 0 0.0004 172.627>
  }]
 foot_l [{
   shapeType=Capsule
   length=16.4834
   radius=9.5125
   shapeName=foot_l_capsule
   center=<-5.2334 5.2867 0.3759>
   rotation=<rot 0 -1.0822 -82.1322>
  }]
 foot_r [{
   shapeType=Capsule
   length=16.4834
   radius=9.5125
   shapeName=foot_r_capsule
   center=<5.2334 -5.2867 -0.3759>
   rotation=<rot 0 -1.0822 97.8678>
  }]
 lowerarm_l [{shapeType=Capsule, length=18.474, radius=6.8202, shapeName=lowerarm_l_capsule, center=<9.9954 0 0>, rotation=<rot 90 -6.3781 0>}]
 pelvis [{
   shapeType=Capsule
   length=14.5718
   radius=13.5648
   shapeName=pelvis_capsule
   center=<-3.3878 -3.9688 0.2152>
   rotation=<rot 11.0152 -90 2.535>
  }]
 root [{shapeType=Sphere, radius=281.2503, shapeName=dyn_floor_collision_sphere, center=<0 0 -278.8158>}]
 spine_03 [{
   shapeType=Capsule
   length=11.6848
   radius=12.2622
   shapeName=spine_03_capsule
   center=<6.7843 0.6764 0.1169>
   rotation=<rot -3.422 -90 1.6906>
  }]
 spine_05 [{shapeType=Capsule, length=20.7729, radius=11.9381, shapeName=spine_05_capsule}]
 thigh_l [
  {shapeType=Capsule, length=30.3031, radius=12.9402, shapeName=thigh_l_capsule, center=<-17.5135 -5.5358 0>, rotation=<rot 90 4.7373 0>}
  {
   shapeType=Capsule
   length=33.1793
   radius=10.8362
   shapeName=thigh_l_capsule
   center=<-24.7139 -4.4379 -0.076>
   rotation=<rot -88.9981 -179.9997 179.9996>
  }
 ]
 thigh_r [{shapeType=Capsule, length=30.3031, radius=12.9402, shapeName=thigh_r_capsule, center=<17.5135 5.5358 0>, rotation=<rot 90 -175.2627 0>}]
 upperarm_l [{shapeType=Capsule, length=21.6773, radius=7.0834, shapeName=upperarm_l_capsule, center=<9.452 0 0>, rotation=<rot 90 0 0>}]
}
constraints [
 {bone1Name=dyn_cape_01_c, bone2Name=MSTR_cape_c, swing1LimitDegrees=45, swing2LimitDegrees=25, twistLimitDegrees=0}
 {bone1Name=dyn_cape_02_c, bone2Name=dyn_cape_01_c, swing1LimitDegrees=45, swing2LimitDegrees=25, twistLimitDegrees=10}
 {bone1Name=dyn_cape_03_c, bone2Name=dyn_cape_02_c, swing1LimitDegrees=45, swing2LimitDegrees=45, twistLimitDegrees=10}
 {bone1Name=dyn_cape_04_c, bone2Name=dyn_cape_03_c, swing1LimitDegrees=40, swing2LimitDegrees=40, twistLimitDegrees=10}
 {bone1Name=dyn_cape_05_c, bone2Name=dyn_cape_04_c, swing1LimitDegrees=40, swing2LimitDegrees=40, twistLimitDegrees=0}
 {bone1Name=dyn_cape_01_l, bone2Name=MSTR_cape_l, swing1LimitDegrees=45, swing2LimitDegrees=45, swing2Motion=Locked, twistLimitDegrees=5}
 {bone1Name=dyn_cape_02_l, bone2Name=dyn_cape_01_l, swing1LimitDegrees=45, swing2LimitDegrees=25, twistLimitDegrees=10}
 {bone1Name=dyn_cape_03_l, bone2Name=dyn_cape_02_l, swing1LimitDegrees=45, swing2LimitDegrees=45, twistLimitDegrees=10}
 {bone1Name=dyn_cape_04_l, bone2Name=dyn_cape_03_l, swing1LimitDegrees=40, swing2LimitDegrees=40, twistLimitDegrees=10}
 {bone1Name=dyn_cape_05_l, bone2Name=dyn_cape_04_l, swing1LimitDegrees=40, swing2LimitDegrees=40, twistLimitDegrees=0}
 {bone1Name=dyn_cape_01_r, bone2Name=MSTR_cape_r, swing1LimitDegrees=45, swing2LimitDegrees=25, twistLimitDegrees=0}
 {bone1Name=dyn_cape_02_r, bone2Name=dyn_cape_01_r, swing1LimitDegrees=45, swing2LimitDegrees=25, twistLimitDegrees=10}
 {bone1Name=dyn_cape_03_r, bone2Name=dyn_cape_02_r, swing1LimitDegrees=45, swing2LimitDegrees=45, twistLimitDegrees=10}
 {bone1Name=dyn_cape_04_r, bone2Name=dyn_cape_03_r, swing1LimitDegrees=40, swing2LimitDegrees=40, twistLimitDegrees=10}
 {bone1Name=dyn_cape_05_r, bone2Name=dyn_cape_04_r, swing1LimitDegrees=40, swing2LimitDegrees=40, twistLimitDegrees=0}
 {
  bone1Name=MSTR_sword_r
  bone2Name=root
  swing1LimitDegrees=45
  swing1Motion=Free
  swing2LimitDegrees=45
  swing2Motion=Free
  twistLimitDegrees=45
  twistMotion=Free
 }
 {
  bone1Name=dyn_cape_05_r
  bone2Name=dyn_cape_05_c
  swing1LimitDegrees=45
  swing1Motion=Free
  swing2LimitDegrees=45
  swing2Motion=Free
  twistLimitDegrees=45
  twistMotion=Free
 }
 {
  bone1Name=dyn_cape_05_l
  bone2Name=dyn_cape_05_c
  swing1LimitDegrees=45
  swing1Motion=Free
  swing2LimitDegrees=45
  swing2Motion=Free
  twistLimitDegrees=45
  twistMotion=Free
 }
]
