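# Physics asset for the belt mesh of the M_MED_CurbRestore medium male player body cosmetic, set up for a dangling belt.
# Five dyn_belt capsules chain from MSTR_belt through constraints with tight twist and swing limits that narrow toward the tip.
# Capsules on the pelvis, both thighs and three front-right skirt bones have no constraints, apparently serving as colliders for the belt.
# The thighs stack multiple capsules, with an extra one on the right.
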PhysicsAsset /BRCosmetics/Characters/Player/Male/Medium/Bodies/M_MED_CurbRestore/Meshes/M_MED_CurbRestore_belt_Physics
bodies {
 MSTR_belt [{shapeType=Capsule, length=0.505, radius=0.505, shapeName=MSTR_belt_capsule, rotation=<rot 0 -90 0>}]
 dyn_belt_01 [{shapeType=Capsule, length=4.2288, radius=1.4209, shapeName=dyn_belt_01_capsule, center=<4.8 0 0.6055>, rotation=<rot 90 0 0>}]
 dyn_belt_02 [{shapeType=Capsule, length=7.4401, radius=2.3534, shapeName=dyn_belt_02_capsule, center=<6.5 0 1.7436>, rotation=<rot 90 0 0>}]
 dyn_belt_03 [{shapeType=Capsule, length=11.4117, radius=2.3395, shapeName=dyn_belt_03_capsule, center=<8.6837 0 1.7782>, rotation=<rot 90 0 0>}]
 dyn_belt_04 [{shapeType=Capsule, length=10.3633, radius=2.0428, shapeName=dyn_belt_04_capsule, center=<8.15 -0.1811 0.7738>, rotation=<rot 90 0 0>}]
 dyn_belt_05 [{shapeType=Capsule, length=7.6342, radius=2.5501, shapeName=dyn_belt_05_capsule, center=<7.0986 0 0.2728>, rotation=<rot 90 0 0>}]
 dyn_skirt_01_front_r [{
   shapeType=Capsule
   length=12.5012
   radius=5.2757
   shapeName=dyn_skirt_01_front_r_capsule
   center=<2.6333 -3.1806 0.256>
   rotation=<rot -8.1905 -1.0491 10.0007>
  }]
 dyn_skirt_02_front_r [{
   shapeType=Capsule
   length=10.254
   radius=6.1623
   shapeName=dyn_skirt_02_front_r_capsule
   center=<1.9735 -4.3128 0.3646>
   rotation=<rot -7.3674 -1.0493 14.8733>
  }]
 dyn_skirt_03_front_r [{
   shapeType=Capsule
   length=11.0606
   radius=6.2966
   shapeName=dyn_skirt_03_front_r_capsule
   center=<2.2164 -4.5517 0.1763>
   rotation=<rot -9.4791 -1.6517 9.9321>
  }]
 pelvis [{
   shapeType=Capsule
   length=14.5718
   radius=13.5648
   shapeName=pelvis_capsule
   center=<-3.3878 -5.0046 0.4168>
   rotation=<rot 11.0152 -90 2.535>
  }]
 thigh_l [
  {shapeType=Capsule, length=30.3031, radius=12.9402, shapeName=thigh_l_capsule, center=<-17.5135 -5.5358 0>, rotation=<rot 90 4.7373 0>}
  {
   shapeType=Capsule
   length=33.1793
   radius=10.8362
   shapeName=thigh_l_capsule
   center=<-24.7139 -4.4379 -0.076>
   rotation=<rot -88.9981 -179.9997 179.9996>
  }
 ]
 thigh_r [
  {shapeType=Capsule, length=30.3031, radius=12.9402, shapeName=thigh_r_capsule, center=<17.5135 5.5358 0>, rotation=<rot 90 -175.2627 0>}
  {
   shapeType=Capsule
   length=17.1704
   radius=5.5811
   shapeName=thigh_r_capsule
   center=<4.9229 -5.7582 -6.1831>
   rotation=<rot -5.0489 -0.0002 5.4623>
  }
  {
   shapeType=Capsule
   length=27.9953
   radius=9.5377
   shapeName=thigh_r_capsule
   center=<22.533 -1.011 -1.8038>
   rotation=<rot 89.4732 179.9979 -175.4063>
  }
 ]
}
constraints [
 {bone1Name=dyn_belt_01, bone2Name=MSTR_belt, swing1LimitDegrees=25, swing2LimitDegrees=30, twistLimitDegrees=2}
 {bone1Name=dyn_belt_02, bone2Name=dyn_belt_01, swing1LimitDegrees=24, swing2LimitDegrees=25, twistLimitDegrees=2}
 {bone1Name=dyn_belt_03, bone2Name=dyn_belt_02, swing1LimitDegrees=23, swing2LimitDegrees=20, twistLimitDegrees=2}
 {bone1Name=dyn_belt_04, bone2Name=dyn_belt_03, swing1LimitDegrees=22, swing2LimitDegrees=18, twistLimitDegrees=2}
 {bone1Name=dyn_belt_05, bone2Name=dyn_belt_04, swing1LimitDegrees=20, swing2LimitDegrees=16, twistLimitDegrees=2}
]
