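# Anim Blueprint for the M_MED_CurbRestore cosmetic player body, adding secondary motion to its cape, skirt, swords and backpack.
# The attached parent's pose feeds a state-driven Control Rig, then additive bone offsets.
# Bools switch between that pose, two RigidBody setups and cape Trail controllers, feeding per-bone AnimDynamics for the skirt.
# Post-dynamics offsets adjust attachments, and a custom-idle weight layers DefaultSlot onto left-sword bones.
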AnimBlueprint /BRCosmetics/Characters/Player/Male/Medium/Bodies/M_MED_CurbRestore/Meshes/M_MED_CurbRestore_AnimBP
target_skeleton=&/M_MED_CurbRestore_Skeleton
graphs [
 AnimGraph {nodes [
   n1 = "Transform (Modify) Bone - Bone: MSTR_cape_r"!(
    ComponentPose=Local To Component(LocalPose=Use cached pose 'postDynamics'())
    bAlphaBoolEnabled:bool=True
    Translation="0,0,0"
    Rotation="0,0,10"
    Scale="1,1,1"
   ) -> (Pose) {
    alphaInputType=Bool
    rotationMode=BMM_Additive
    translationMode=BMM_Additive
    alphaBoolBlend {blendInTime=1, blendOutTime=0.5}
    boneToModify {boneName=MSTR_cape_r}
    rotation {roll=10}
   }
   n2 = "Transform (Modify) Bone - Bone: MSTR_cape_l"!(ComponentPose=@n1, bAlphaBoolEnabled:bool=True, Translation="0,0,0", Rotation="0,0,10", Scale="1,1,1") -> (Pose) {
    alphaInputType=Bool
    rotationMode=BMM_Additive
    translationMode=BMM_Additive
    alphaBoolBlend {blendInTime=1, blendOutTime=0.5}
    boneToModify {boneName=MSTR_cape_l}
    rotation {roll=10}
   }
   n3 = "Transform (Modify) Bone - Bone: MSTR_skirt_back_r"!(ComponentPose=@n2, bAlphaBoolEnabled:bool=True, Translation="0,0,0", Rotation="0,0,-20", Scale="1,1,1") -> (Pose) {
    alphaInputType=Bool
    rotationMode=BMM_Additive
    translationMode=BMM_Additive
    alphaBoolBlend {blendInTime=1, blendOutTime=0.5}
    boneToModify {boneName=MSTR_skirt_back_r}
    rotation {roll=-20}
   }
   n4 = "Transform (Modify) Bone - Bone: MSTR_skirt_back_l"!(ComponentPose=@n3, bAlphaBoolEnabled:bool=True, Translation="0,0,0", Rotation="0,0,-20", Scale="1,1,1") -> (Pose) {
    alphaInputType=Bool
    rotationMode=BMM_Additive
    translationMode=BMM_Additive
    alphaBoolBlend {blendInTime=1, blendOutTime=0.5}
    boneToModify {boneName=MSTR_skirt_back_l}
    rotation {roll=-20}
   }
   n5 = "Transform (Modify) Bone - Bone: MSTR_sword_r"!(ComponentPose=@n4, bAlphaBoolEnabled:bool=True, Translation="-0.717,2.233,-0.020", Rotation="22.000,-0.522,15.000", Scale="1,1,1") -> (Pose) {
    alphaInputType=Bool
    rotationMode=BMM_Additive
    translationMode=BMM_Additive
    translationSpace=BCS_BoneSpace
    alphaBoolBlend {blendInTime=2, blendOption=CubicInOut, blendOutTime=0.5}
    boneToModify {boneName=MSTR_sword_r}
    rotation=<rot 22 -0.522 15>
    translation=<-0.7169 2.2331 -0.0205>
   }
   n6 = "Transform (Modify) Bone - Bone: MSTR_sword_l"!(ComponentPose=@n5, bAlphaBoolEnabled:bool=True, Translation="0,0,0", Rotation="0.000,0.000,-18.727", Scale="1,1,1") -> (Pose) {
    alphaInputType=Bool
    rotationMode=BMM_Additive
    translationMode=BMM_Additive
    translationSpace=BCS_BoneSpace
    alphaBoolBlend {blendInTime=4, blendOption=CubicInOut, blendOutTime=0.3}
    boneToModify {boneName=MSTR_sword_l}
    rotation {roll=-18.727}
   }
   n7 = "Transform (Modify) Bone - Bone: MSTR_sword_r"(ComponentPose=@n6, bAlphaBoolEnabled=True, Translation="1.589,1.330,-2.794", Rotation="-17.789,1.933,-6.073", Scale="1,1,1") {
    alphaInputType=Bool
    rotationMode=BMM_Additive
    translationMode=BMM_Additive
    translationSpace=BCS_BoneSpace
    alphaBoolBlend {blendInTime=0.1, blendOption=Sinusoidal, blendOutTime=1}
    boneToModify {boneName=MSTR_sword_r}
    rotation=<rot -17.789 1.933 -6.073>
    translation=<1.589 1.33 -2.794>
   }
   n8 = "Transform (Modify) Bone - Bone: MSTR_cape_r"(ComponentPose=@n7, bAlphaBoolEnabled=True, Translation="0,0,0", Rotation="0,0,10", Scale="1,1,1") {
    alphaInputType=Bool
    rotationMode=BMM_Additive
    translationMode=BMM_Additive
    alphaBoolBlend {blendInTime=1, blendOutTime=0.5}
    boneToModify {boneName=MSTR_cape_r}
    rotation {roll=10}
   }
   n9 = "Transform (Modify) Bone - Bone: MSTR_cape_l"(ComponentPose=@n8, bAlphaBoolEnabled=True, Translation="0,0,0", Rotation="0,0,10", Scale="1,1,1") {
    alphaInputType=Bool
    rotationMode=BMM_Additive
    translationMode=BMM_Additive
    alphaBoolBlend {blendInTime=1, blendOutTime=0.5}
    boneToModify {boneName=MSTR_cape_l}
    rotation {roll=10}
   }
   n10 = "Transform (Modify) Bone - Bone: MSTR_cape_l"(ComponentPose=@n9, bAlphaBoolEnabled=True, Translation="0,0,0", Rotation="0,0,10", Scale="1,1,1") {
    alphaInputType=Bool
    rotationMode=BMM_Additive
    translationMode=BMM_Additive
    alphaBoolBlend {blendInTime=1, blendOutTime=0.5}
    boneToModify {boneName=MSTR_cape_l}
    rotation {roll=10}
   }
   n11 = "Transform (Modify) Bone - Bone: attach_backpack"(ComponentPose=@n10, Alpha=1, Translation="0,8,0", Rotation="0,0,0", Scale="1,1,1") {translationMode=BMM_Additive, translationSpace=BCS_BoneSpace, boneToModify {boneName=attach_backpack}, translation {y=8}}
   n12 = "Transform (Modify) Bone - Bone: attach_backpack"(ComponentPose=@n11, Alpha=1, Translation="0,-6,2", Rotation="0,0,10", Scale="1,1,1") {
    rotationMode=BMM_Additive
    rotationSpace=BCS_BoneSpace
    translationMode=BMM_Additive
    translationSpace=BCS_BoneSpace
    alphaScaleBiasClamp {bInterpResult=true, interpSpeedDecreasing=1, interpSpeedIncreasing=1}
    boneToModify {boneName=attach_backpack}
    rotation {roll=10}
    translation {y=-6, z=2}
   }
   n13 = "Transform (Modify) Bone - Bone: attach_backpack"(ComponentPose=@n12, Alpha=1, Translation="-2,-4,3", Rotation="0,0,0", Scale="1,1,1") {
    translationMode=BMM_Additive
    translationSpace=BCS_BoneSpace
    alphaScaleBiasClamp {bInterpResult=true, interpSpeedDecreasing=1}
    boneToModify {boneName=attach_backpack}
    translation=<-2 -4 3>
   }
   n14 = "Transform (Modify) Bone - Bone: attach_backpack"(ComponentPose=@n13, Alpha=1, Translation="0,-4.8,0", Rotation="0,0,0", Scale="1,1,1") {
    translationMode=BMM_Additive
    translationSpace=BCS_BoneSpace
    alphaScaleBiasClamp {bInterpResult=true, interpSpeedDecreasing=1}
    boneToModify {boneName=attach_backpack}
    translation {y=-4.8}
   }
   n15 = "Transform (Modify) Bone - Bone: MSTR_cape_r"(ComponentPose=@n14, bAlphaBoolEnabled=True, Translation="0,0,0", Rotation="0,0,10", Scale="1,1,1") {
    alphaInputType=Bool
    rotationMode=BMM_Additive
    translationMode=BMM_Additive
    alphaBoolBlend {blendInTime=1, blendOutTime=0.5}
    boneToModify {boneName=MSTR_cape_r}
    rotation {roll=10}
   }
   n16 = "Transform (Modify) Bone - Bone: MSTR_cape_l"(ComponentPose=@n15, bAlphaBoolEnabled=True, Translation="0,0,0", Rotation="0,0,10", Scale="1,1,1") {
    alphaInputType=Bool
    rotationMode=BMM_Additive
    translationMode=BMM_Additive
    alphaBoolBlend {blendInTime=1, blendOutTime=0.5}
    boneToModify {boneName=MSTR_cape_l}
    rotation {roll=10}
   }
   n17 = "Transform (Modify) Bone - Bone: MSTR_cape_l"(ComponentPose=@n16, bAlphaBoolEnabled=True, Translation="0,0,0", Rotation="0,0,10", Scale="1,1,1") {
    alphaInputType=Bool
    rotationMode=BMM_Additive
    translationMode=BMM_Additive
    alphaBoolBlend {blendInTime=1, blendOutTime=0.5}
    boneToModify {boneName=MSTR_cape_l}
    rotation {roll=10}
   }
   n18 = "Transform (Modify) Bone - Bone: MSTR_skirt_back_r"(ComponentPose=@n17, bAlphaBoolEnabled=True, Translation="2,0,0", Rotation="0,0,-10", Scale="1,1,1") {
    alphaInputType=Bool
    rotationMode=BMM_Additive
    translationMode=BMM_Additive
    alphaBoolBlend {blendInTime=1, blendOutTime=0.5}
    boneToModify {boneName=MSTR_skirt_back_r}
    rotation {roll=-10}
    translation {x=2}
   }
   n19 = "Transform (Modify) Bone - Bone: MSTR_skirt_back_l"(ComponentPose=@n18, bAlphaBoolEnabled=True, Translation="2,0,0", Rotation="0,0,-8", Scale="1,1,1") {
    alphaInputType=Bool
    rotationMode=BMM_Additive
    translationMode=BMM_Additive
    alphaBoolBlend {blendInTime=1, blendOutTime=0.5}
    boneToModify {boneName=MSTR_skirt_back_l}
    rotation {roll=-8}
    translation {x=2}
   }
   n20 = "Transform (Modify) Bone - Bone: upperarm_l"(ComponentPose=@n19, Alpha=1, Translation="0,0,0", Rotation="0,0,0", Scale="1,1,1") {
    rotationMode=BMM_Additive
    rotationSpace=BCS_BoneSpace
    alphaScaleBiasClamp {bInterpResult=true, interpSpeedDecreasing=5, interpSpeedIncreasing=5}
    boneToModify {boneName=upperarm_l}
   }
   n21 = "Transform (Modify) Bone - Bone: MSTR_cape_r"(ComponentPose=@n20, bAlphaBoolEnabled=True, Translation="0,0,0", Rotation="0,0,-30", Scale="1,1,1") {
    alphaInputType=Bool
    rotationMode=BMM_Additive
    translationMode=BMM_Additive
    alphaBoolBlend {blendInTime=1, blendOutTime=0.5}
    boneToModify {boneName=MSTR_cape_r}
    rotation {roll=-30}
   }
   n22 = "Transform (Modify) Bone - Bone: MSTR_cape_c"!(ComponentPose=@n21, bAlphaBoolEnabled:bool=True, Translation="0,0,0", Rotation="0,0,-10", Scale="1,1,1") -> (Pose) {
    alphaInputType=Bool
    rotationMode=BMM_Additive
    translationMode=BMM_Additive
    alphaBoolBlend {blendInTime=1, blendOutTime=0.5}
    boneToModify {boneName=MSTR_cape_c}
    rotation {roll=-10}
   }
   n23 = "Transform (Modify) Bone - Bone: MSTR_cape_l"(ComponentPose=@n22, bAlphaBoolEnabled=True, Translation="0,0,0", Rotation="0,0,-30", Scale="1,1,1") {
    alphaInputType=Bool
    rotationMode=BMM_Additive
    translationMode=BMM_Additive
    alphaBoolBlend {blendInTime=1, blendOutTime=0.5}
    boneToModify {boneName=MSTR_cape_l}
    rotation {roll=-30}
   }
   n24 = "Transform (Modify) Bone - Bone: dyn_cape_02_r"!(ComponentPose=@n23, bAlphaBoolEnabled:bool=True, Translation="0,0,0", Rotation="0,0,-10", Scale="1,1,1") -> (Pose) {
    alphaInputType=Bool
    rotationMode=BMM_Additive
    translationMode=BMM_Additive
    alphaBoolBlend {blendInTime=1, blendOutTime=0.5}
    boneToModify {boneName=dyn_cape_02_r}
    rotation {roll=-10}
   }
   n25 = "Transform (Modify) Bone - Bone: dyn_cape_02_c"!(ComponentPose=@n24, bAlphaBoolEnabled:bool=True, Translation="0,0,0", Rotation="0,0,-10", Scale="1,1,1") -> (Pose) {
    alphaInputType=Bool
    rotationMode=BMM_Additive
    translationMode=BMM_Additive
    alphaBoolBlend {blendInTime=1, blendOutTime=0.5}
    boneToModify {boneName=dyn_cape_02_c}
    rotation {roll=-10}
   }
   n26 = "Transform (Modify) Bone - Bone: dyn_cape_02_l"!(ComponentPose=@n25, bAlphaBoolEnabled:bool=True, Translation="0,0,0", Rotation="0,0,-10", Scale="1,1,1") -> (Pose) {
    alphaInputType=Bool
    rotationMode=BMM_Additive
    translationMode=BMM_Additive
    alphaBoolBlend {blendInTime=1, blendOutTime=0.5}
    boneToModify {boneName=dyn_cape_02_l}
    rotation {roll=-10}
   }
   n27 = "Transform (Modify) Bone - Bone: MSTR_cape_l"(ComponentPose=@n26, bAlphaBoolEnabled=True, Translation="-10,0,10", Rotation="0,-10,10", Scale="1,1,1") {
    alphaInputType=Bool
    rotationMode=BMM_Additive
    translationMode=BMM_Additive
    alphaBoolBlend {blendInTime=0.2, blendOutTime=1}
    boneToModify {boneName=MSTR_cape_l}
    rotation {roll=10, yaw=-10}
    translation {x=-10, z=10}
   }
   Output Pose(Result=Layered blend per bone(
     BasePose=Component To Local(ComponentPose=@n27)
     BlendPoses_0=Slot 'DefaultSlot'()
     BlendWeights_0=Get Custom Idle Is Playing() -> (Custom Idle Is Playing:double)
    ) {layerSetup [{branchFilters [{boneName=MSTR_sword_l}, {boneName=sword_hidebone_l}]}]}) {layerGroup=None}
   n28 = Control Rig!(
    Source=Copy Pose From Mesh() {bUseAttachedParent=true}
    Alpha
    crouchingAndNotMoving:bool=Get crounchingAndNotMoving() -> (crounchingAndNotMoving:bool)
    crouchingAndMoving:bool=Get crouchingAndMoving() -> (crouchingAndMoving:bool)
    crouching:bool=Get bIsCrouching()
    jumpOrFalling:bool=Get jumpOrFalling() -> (jumpOrFalling:bool)
    grindRail:bool=Get bIsOnGrindRail()
    swimming:bool=Get bIsSurfaceSwimming()
    parachuteOpened:bool=Get bIsParachuteOpened()
    emotion:bool=Get bIsPlayingEmote()
    GoldenGloryEmote:double=Get GoldenGloryEmote() -> (GoldenGloryEmote:double)
   ) -> (Pose) {
    bTransferPoseInGlobalSpace=true
    controlRigAssetReference {blueprintRigClass=&/M_MED_CurbRestore_CtrlRig.M_MED_CurbRestore_CtrlRig_C}
   }
   Save cached pose 'preControlRig'(Pose=@n28) {cacheName=preControlRig}
   n29 = "Transform (Modify) Bone - Bone: dyn_cape_01_l"!(
    ComponentPose=Local To Component(LocalPose=Use cached pose 'preDynamics'())
    bAlphaBoolEnabled:bool=True
    Translation="0,0,0"
    Rotation="0,-45,0"
    Scale="1,1,1"
   ) -> (Pose) {
    alphaInputType=Bool
    rotationMode=BMM_Additive
    rotationSpace=BCS_BoneSpace
    alphaBoolBlend {blendInTime=1, blendOption=Sinusoidal, blendOutTime=1}
    boneToModify {boneName=dyn_cape_01_l}
    rotation {yaw=-45}
   }
   n30 = "Transform (Modify) Bone - Bone: dyn_cape_01_c"!(ComponentPose=@n29, bAlphaBoolEnabled:bool=True, Translation="0,0,0", Rotation="0,-45,0", Scale="1,1,1") -> (Pose) {
    alphaInputType=Bool
    rotationMode=BMM_Additive
    rotationSpace=BCS_BoneSpace
    alphaBoolBlend {blendInTime=1, blendOption=Sinusoidal, blendOutTime=1}
    boneToModify {boneName=dyn_cape_01_c}
    rotation {yaw=-45}
   }
   n31 = "Transform (Modify) Bone - Bone: dyn_cape_01_r"!(ComponentPose=@n30, bAlphaBoolEnabled:bool=True, Translation="0,0,0", Rotation="0,-45,0", Scale="1,1,1") -> (Pose) {
    alphaInputType=Bool
    rotationMode=BMM_Additive
    rotationSpace=BCS_BoneSpace
    alphaBoolBlend {blendInTime=1, blendOption=Sinusoidal, blendOutTime=1}
    boneToModify {boneName=dyn_cape_01_r}
    rotation {yaw=-45}
   }
   n32 = Get BwdMovement() -> (BwdMovement:double)
   n33 = Trail controller - Bone: dyn_cape_05_l(ComponentPose=@n31, Alpha=@n32) {
    bUsePlanarLimit=true
    chainLength=5
    debugLifeTime=0.1
    maxDeltaTime=0.0333
    alphaScaleBiasClamp {bInterpResult=true}
    baseJoint {boneName=attach_backpack}
    planarLimits [{drivingBone {boneName=attach_cape}, planeTransform {location=<0 0 0>, rotation=<rot 0 0 -90>, scale=<1 1 1>}}]
    rotationLimits [
     {limitMax=<5 25 45>, limitMin=<-5 -25 -45>}
     {limitMax=<5 25 45>, limitMin=<-5 -25 -45>}
     {limitMax=<5 25 45>, limitMin=<-5 -25 -45>}
     {limitMax=<5 25 45>, limitMin=<-5 -25 -45>}
     {limitMax=<5 25 45>, limitMin=<-5 -25 -45>}
    ]
    rotationOffsets [<0 0 0>, <0 0 0>, <0 0 0>, <0 0 0>, <0 0 0>]
    trailBone {boneName=dyn_cape_05_l}
    trailRelaxationSpeed {editorCurveData {keys [<at 0 23 RCIM_Linear>,<at 0.284 19.3799 RCIM_Linear>,<at 0.6554 8.5924 RCIM_Linear>,<at 1 5 RCIM_Linear>]}}
   }
   n34 = Trail controller - Bone: dyn_cape_05_c(ComponentPose=@n33, Alpha=@n32) {
    bUsePlanarLimit=true
    chainLength=5
    debugLifeTime=0.1
    maxDeltaTime=0.0333
    alphaScaleBiasClamp {bInterpResult=true}
    baseJoint {boneName=attach_backpack}
    planarLimits [{drivingBone {boneName=attach_cape}, planeTransform {location=<0 0 0>, rotation=<rot 0 0 -90>, scale=<1 1 1>}}]
    rotationLimits [
     {limitMax=<180 180 45>, limitMin=<-180 -180 -45>}
     {limitMax=<180 180 45>, limitMin=<-180 -180 -45>}
     {limitMax=<180 90 45>, limitMin=<-180 -90 -45>}
     {limitMax=<180 180 45>, limitMin=<-180 -180 -45>}
     {limitMax=<180 180 45>, limitMin=<-180 -180 -45>}
    ]
    rotationOffsets [<0 0 0>, <0 0 0>, <0 0 0>, <0 0 0>, <0 0 0>]
    trailBone {boneName=dyn_cape_05_c}
    trailRelaxationSpeed {editorCurveData {keys [<at 0 23 RCIM_Linear>,<at 0.284 19.3799 RCIM_Linear>,<at 0.6554 8.5924 RCIM_Linear>,<at 1 5 RCIM_Linear>]}}
   }
   n35 = Trail controller - Bone: dyn_cape_05_r(ComponentPose=@n34, Alpha=@n32) {
    bUsePlanarLimit=true
    chainLength=5
    debugLifeTime=0.1
    maxDeltaTime=0.0333
    alphaScaleBiasClamp {bInterpResult=true}
    baseJoint {boneName=attach_backpack}
    planarLimits [{drivingBone {boneName=attach_cape}, planeTransform {location=<0 0 0>, rotation=<rot 0 0 -90>, scale=<1 1 1>}}]
    rotationLimits [
     {limitMax=<180 180 45>, limitMin=<-180 -180 -45>}
     {limitMax=<180 180 45>, limitMin=<-180 -180 -45>}
     {limitMax=<180 90 45>, limitMin=<-180 -90 -45>}
     {limitMax=<180 180 45>, limitMin=<-180 -180 -45>}
     {limitMax=<180 180 45>, limitMin=<-180 -180 -45>}
    ]
    rotationOffsets [<0 0 0>, <0 0 0>, <0 0 0>, <0 0 0>, <0 0 0>]
    trailBone {boneName=dyn_cape_05_r}
    trailRelaxationSpeed {editorCurveData {keys [<at 0 23 RCIM_Linear>,<at 0.284 19.3799 RCIM_Linear>,<at 0.6554 8.5924 RCIM_Linear>,<at 1 5 RCIM_Linear>]}}
   }
   n36 = "Transform (Modify) Bone - Bone: dyn_cape_01_l"(ComponentPose=@n35, bAlphaBoolEnabled=True, Translation="0,0,0", Rotation="0,60,0", Scale="1,1,1") {
    alphaInputType=Bool
    rotationMode=BMM_Additive
    rotationSpace=BCS_BoneSpace
    alphaBoolBlend {blendInTime=1, blendOption=Sinusoidal, blendOutTime=1}
    boneToModify {boneName=dyn_cape_01_l}
    rotation {yaw=60}
   }
   n37 = "Transform (Modify) Bone - Bone: dyn_cape_01_c"(ComponentPose=@n36, bAlphaBoolEnabled=True, Translation="0,0,0", Rotation="0,45,0", Scale="1,1,1") {
    alphaInputType=Bool
    rotationMode=BMM_Additive
    rotationSpace=BCS_BoneSpace
    alphaBoolBlend {blendInTime=1, blendOption=Sinusoidal, blendOutTime=1}
    boneToModify {boneName=dyn_cape_01_c}
    rotation {yaw=45}
   }
   n38 = "Transform (Modify) Bone - Bone: dyn_cape_01_r"(ComponentPose=@n37, bAlphaBoolEnabled=True, Translation="0,0,0", Rotation="0,45,0", Scale="1,1,1") {
    alphaInputType=Bool
    rotationMode=BMM_Additive
    rotationSpace=BCS_BoneSpace
    alphaBoolBlend {blendInTime=1, blendOption=Sinusoidal, blendOutTime=1}
    boneToModify {boneName=dyn_cape_01_r}
    rotation {yaw=45}
   }
   Save cached pose 'trailCache'(Pose=Component To Local(ComponentPose=@n38)) {cacheName=trailCache}
   n39 = "Transform (Modify) Bone - Bone: MSTR_sword_r"(
    ComponentPose=Local To Component(LocalPose=Use cached pose 'preControlRig'())
    bAlphaBoolEnabled=Get crounchingAndNotMoving()
    Translation="0,0,0"
    Rotation="0,0,10"
    Scale="1,1,1"
   ) {
    alphaInputType=Bool
    rotationMode=BMM_Additive
    translationMode=BMM_Additive
    translationSpace=BCS_BoneSpace
    alphaBoolBlend {blendInTime=1, blendOutTime=0.5}
    boneToModify {boneName=MSTR_sword_r}
    rotation {roll=10}
   }
   n40 = "Transform (Modify) Bone - Bone: MSTR_sword_l"(ComponentPose=@n39, bAlphaBoolEnabled=True, Translation="-2.445,2.965,4.203", Rotation="-3.793,-3.727,17.050", Scale="1,1,1") {
    alphaInputType=Bool
    rotationMode=BMM_Additive
    rotationSpace=BCS_BoneSpace
    translationMode=BMM_Additive
    translationSpace=BCS_BoneSpace
    alphaBoolBlend {blendInTime=0.35, blendOutTime=0.5}
    boneToModify {boneName=MSTR_sword_l}
    rotation=<rot -3.793 -3.727 17.05>
    translation=<-2.445 2.965 4.203>
   }
   n41 = "Transform (Modify) Bone - Bone: MSTR_sword_l"(ComponentPose=@n40, bAlphaBoolEnabled=True, Translation="-3.483,-0.175,-0.019", Rotation="-6.086,7.885,27.226", Scale="1,1,1") {
    alphaInputType=Bool
    rotationMode=BMM_Additive
    translationMode=BMM_Additive
    translationSpace=BCS_BoneSpace
    alphaBoolBlend {blendInTime=0.2, blendOutTime=1}
    boneToModify {boneName=MSTR_sword_l}
    rotation=<rot -6.086 7.885 27.226>
    translation=<-3.483 -0.175 -0.019>
   }
   n42 = "Transform (Modify) Bone - Bone: MSTR_sword_l"(ComponentPose=@n41, bAlphaBoolEnabled=True, Translation="1.5,-1.7,1.5", Rotation="-6.389,-5.584,-7.483", Scale="1,1,1") {
    alphaInputType=Bool
    rotationMode=BMM_Additive
    translationMode=BMM_Additive
    translationSpace=BCS_BoneSpace
    alphaBoolBlend {blendInTime=0.5, blendOutTime=0.5}
    boneToModify {boneName=MSTR_sword_l}
    rotation=<rot -6.389 -5.584 -7.483>
    translation=<1.5 -1.7 1.5>
   }
   n43 = "Transform (Modify) Bone - Bone: MSTR_sword_r"(ComponentPose=@n42, bAlphaBoolEnabled=True, Translation="-0.028,0.399,-1.639", Rotation="-16.719,-1.196,0.000", Scale="1,1,1") {
    alphaInputType=Bool
    rotationMode=BMM_Additive
    translationMode=BMM_Additive
    translationSpace=BCS_BoneSpace
    alphaBoolBlend {blendInTime=0.2, blendOutTime=0.2}
    boneToModify {boneName=MSTR_sword_r}
    rotation=<rot -16.719 -1.196 0>
    translation=<-0.028 0.399 -1.639>
   }
   n44 = "Transform (Modify) Bone - Bone: MSTR_cape_r"(ComponentPose=@n43, bAlphaBoolEnabled=True, Translation="0,-4,0", Rotation="0,0,-30", Scale="1,1,1") {
    alphaInputType=Bool
    rotationMode=BMM_Additive
    translationMode=BMM_Additive
    alphaBoolBlend {blendInTime=1, blendOutTime=0.5}
    boneToModify {boneName=MSTR_cape_r}
    rotation {roll=-30}
    translation {y=-4}
   }
   n45 = "Transform (Modify) Bone - Bone: MSTR_sword_l"(ComponentPose=@n44, bAlphaBoolEnabled=True, Translation="0,0,0", Rotation="0,0,15", Scale="1,1,1") {
    alphaInputType=Bool
    rotationMode=BMM_Additive
    translationMode=BMM_Additive
    translationSpace=BCS_BoneSpace
    alphaBoolBlend {blendInTime=0.1, blendOutTime=1}
    boneToModify {boneName=MSTR_sword_l}
    rotation {roll=15}
   }
   n46 = "Transform (Modify) Bone - Bone: MSTR_cape_l"(ComponentPose=@n45, bAlphaBoolEnabled=True, Translation="0,-4,0", Rotation="0,0,-30", Scale="1,1,1") {
    alphaInputType=Bool
    rotationMode=BMM_Additive
    translationMode=BMM_Additive
    alphaBoolBlend {blendInTime=1, blendOutTime=0.5}
    boneToModify {boneName=MSTR_cape_l}
    rotation {roll=-30}
    translation {y=-4}
   }
   n47 = "Transform (Modify) Bone - Bone: MSTR_cape_l"(ComponentPose=@n46, bAlphaBoolEnabled=True, Translation="0,-4,0", Rotation="0,0,-30", Scale="1,1,1") {
    alphaInputType=Bool
    rotationMode=BMM_Additive
    translationMode=BMM_Additive
    alphaBoolBlend {blendInTime=1, blendOutTime=0.5}
    boneToModify {boneName=MSTR_cape_l}
    rotation {roll=-30}
    translation {y=-4}
   }
   n48 = "Transform (Modify) Bone - Bone: MSTR_sword_l"(ComponentPose=@n47, bAlphaBoolEnabled=True, Translation="0,0,0", Rotation="0,0,-10", Scale="1,1,1") {
    alphaInputType=Bool
    rotationMode=BMM_Additive
    translationMode=BMM_Additive
    translationSpace=BCS_BoneSpace
    alphaBoolBlend {blendInTime=0.2, blendOutTime=1}
    boneToModify {boneName=MSTR_sword_l}
    rotation {roll=-10}
   }
   n49 = "Transform (Modify) Bone - Bone: MSTR_sword_l"(ComponentPose=@n48, bAlphaBoolEnabled=True, Translation="0,0,0", Rotation="0,0,8", Scale="1,1,1") {
    alphaInputType=Bool
    rotationMode=BMM_Additive
    translationMode=BMM_Additive
    translationSpace=BCS_BoneSpace
    alphaBoolBlend {blendInTime=0.4, blendOutTime=1}
    boneToModify {boneName=MSTR_sword_l}
    rotation {roll=8}
   }
   n50 = "Transform (Modify) Bone - Bone: MSTR_sword_r"(ComponentPose=@n49, bAlphaBoolEnabled=True, Translation="0,0,0", Rotation="0,0,5", Scale="1,1,1") {
    alphaInputType=Bool
    rotationMode=BMM_Additive
    translationMode=BMM_Additive
    alphaBoolBlend {blendInTime=1, blendOutTime=0.5}
    boneToModify {boneName=MSTR_sword_r}
    rotation {roll=5}
   }
   Save cached pose 'preDynamics'(Pose=Component To Local(ComponentPose=@n50)) {cacheName=preDynamics}
   n51 = RigidBody!(ComponentPose=Local To Component(LocalPose=Use cached pose 'preDynamics'()), Alpha, OverrideWorldGravity:Vector="0,0,-980", ExternalForce) -> (Pose) {
    bClampLinearTranslationLimitToRefPose=true
    bOverrideWorldGravity=true
    overridePhysicsAsset=&/M_MED_CurbRestore_Physics
    baseBoneRef {boneName=spine_05}
    componentAppliedLinearAccClamp=<3000 3000 3000>
    overrideWorldGravity {z=-980}
    simSpaceSettings {
     dampingAlpha=0
     maxAngularAcceleration=24
     maxAngularVelocity=6
     maxLinearAcceleration=3200
     maxLinearVelocity=800
     velocityScaleZ=0.65
     worldAlpha=1
     externalLinearDragV=<2.3 0.2 0.5>
    }
   }
   n52 = RigidBody(ComponentPose=Local To Component(LocalPose=Use cached pose 'preDynamics'()), OverrideWorldGravity="0,0,-980") {
    bOverrideWorldGravity=true
    overridePhysicsAsset=&/M_MED_CurbRestore_BackpackSolution_Physics
    simulationSpace=BaseBoneSpace
    baseBoneRef {boneName=spine_05}
    componentAppliedLinearAccClamp=<3000 3000 3000>
    overrideWorldGravity {z=-980}
    simSpaceSettings {
     dampingAlpha=0
     maxAngularAcceleration=16
     maxAngularVelocity=8
     maxLinearAcceleration=3200
     maxLinearVelocity=800
     velocityScaleZ=0.45
     worldAlpha=1
     externalLinearDragV=<1 0.5 0.5>
    }
   }
   n53 = Blend Poses by bool(
    BlendPose_0=Component To Local(ComponentPose=@n52)
    BlendPose_1=Component To Local(ComponentPose=@n51)
    bActiveValue=Get IsWearingBackpackOrCape() -> (IsWearingBackpackOrCape:bool)
   ) {blendType=Linear}
   Save cached pose 'Rban'(Pose=Component To Local(ComponentPose=Local To Component(LocalPose=@n53))) {cacheName=Rban}
   n54 = Blend Poses by bool(BlendPose_0=Use cached pose 'preDynamics'(), BlendPose_1=Use cached pose 'Rban'(), bActiveValue=Get bClothEnabled()) {blendType=Linear}
   n55 = Use cached pose 'trailCache'()
   n56 = Get RbanOrCloth() -> (RbanOrCloth:bool)
   n57 = Get Sim Space Settings Skirt() -> (Sim Space Settings Skirt:Anim Phys Sim Space Settings Structure)
   n58 = Anim Dynamics - Bone: dyn_skirt_01_front_l!(
    ComponentPose=Local To Component(LocalPose=Blend Poses by bool(BlendPose_0=@n54, BlendPose_1=@n55, bActiveValue=@n56) {blendType=Linear})
    Alpha=1
    SimSpaceSettings:Anim Phys Sim Space Settings Structure=@n57
    ExternalForce="0,0,0"
   ) -> (Pose) {
    angularDampingOverride=1
    angularSpringConstant=10000
    bAngularSpring=true
    bOverrideAngularDamping=true
    bUseGravityOverride=true
    gravityScale=0.35
    simulationSpace=RootRelative
    boundBone {boneName=dyn_skirt_01_front_l}
    componentAppliedLinearAccClamp=<3000 3000 3000>
    componentLinearAccScale=<1 1 1>
    componentLinearVelScale=<1 1 1>
    gravityOverride {z=-980}
    physicsBodyDefinitions [{
      boundBone {boneName=dyn_skirt_01_front_l}
      boxExtents=<20 30 30>
      constraintSetup {angularLimitsMax=<15 5 15>, angularLimitsMin=<-15 -5 -10>, angularTarget=<1 0 0>}
      localJointOffset=<2 0 0>
     }]
    simSpaceSettings {maxAngularAcceleration=2500, maxAngularVelocity=2500, simSpaceAngularAlpha=1}
   }
   n59 = Anim Dynamics - Bone: dyn_skirt_02_front_l!(ComponentPose=@n58, Alpha=1, SimSpaceSettings:Anim Phys Sim Space Settings Structure=@n57, ExternalForce="0,0,0") -> (Pose) {
    angularDampingOverride=1
    angularSpringConstant=10000
    bAngularSpring=true
    bOverrideAngularDamping=true
    bUseGravityOverride=true
    gravityScale=0.35
    simulationSpace=RootRelative
    boundBone {boneName=dyn_skirt_02_front_l}
    componentAppliedLinearAccClamp=<3000 3000 3000>
    componentLinearAccScale=<1 1 1>
    componentLinearVelScale=<1 1 1>
    gravityOverride {z=-980}
    physicsBodyDefinitions [{
      boundBone {boneName=dyn_skirt_02_front_l}
      boxExtents=<20 30 30>
      constraintSetup {angularLimitsMax=<15 5 15>, angularLimitsMin=<-15 -5 -3>, angularTarget=<1 0 0>}
      localJointOffset=<2 0 0>
     }]
    simSpaceSettings {maxAngularAcceleration=2500, maxAngularVelocity=2500, simSpaceAngularAlpha=1}
   }
   n60 = Anim Dynamics - Bone: dyn_skirt_03_front_l!(ComponentPose=@n59, Alpha=1, SimSpaceSettings:Anim Phys Sim Space Settings Structure=@n57, ExternalForce="0,0,0") -> (Pose) {
    angularDampingOverride=1
    angularSpringConstant=10000
    bAngularSpring=true
    bOverrideAngularDamping=true
    bUseGravityOverride=true
    bUseSphericalLimits=true
    gravityScale=0.35
    simulationSpace=RootRelative
    boundBone {boneName=dyn_skirt_03_front_l}
    componentAppliedLinearAccClamp=<3000 3000 3000>
    componentLinearAccScale=<1 1 1>
    componentLinearVelScale=<1 1 1>
    gravityOverride {z=-980}
    physicsBodyDefinitions [{
      boundBone {boneName=dyn_skirt_03_front_l}
      boxExtents=<20 30 30>
      constraintSetup {angularLimitsMax=<15 5 15>, angularLimitsMin=<-15 -5 -3>, angularTarget=<1 0 0>}
      localJointOffset=<2 0 0>
     }]
    simSpaceSettings {maxAngularAcceleration=2500, maxAngularVelocity=2500, simSpaceAngularAlpha=1}
   }
   n61 = Anim Dynamics - Bone: dyn_skirt_01_side_l!(ComponentPose=@n60, Alpha=1, SimSpaceSettings:Anim Phys Sim Space Settings Structure=@n57, ExternalForce="0,0,0") -> (Pose) {
    angularDampingOverride=1
    angularSpringConstant=12000
    bAngularSpring=true
    bOverrideAngularDamping=true
    bUseGravityOverride=true
    gravityScale=0.35
    simulationSpace=RootRelative
    boundBone {boneName=dyn_skirt_01_side_l}
    componentLinearAccScale=<1 1 1>
    componentLinearVelScale=<1 1 1>
    gravityOverride {z=-980}
    physicsBodyDefinitions [{
      boundBone {boneName=dyn_skirt_01_side_l}
      boxExtents=<20 30 30>
      constraintSetup {angularLimitsMax=<0 15 15>, angularLimitsMin=<0 -15 0>, angularTarget=<1 0 0>}
      localJointOffset=<2 0 0>
     }]
    simSpaceSettings {maxAngularAcceleration=2500, maxAngularVelocity=2500, simSpaceAngularAlpha=1}
   }
   n62 = Anim Dynamics - Bone: dyn_skirt_02_side_l!(ComponentPose=@n61, Alpha=1, SimSpaceSettings:Anim Phys Sim Space Settings Structure=@n57, ExternalForce="0,0,0") -> (Pose) {
    angularDampingOverride=1
    angularSpringConstant=10000
    bAngularSpring=true
    bOverrideAngularDamping=true
    bUseGravityOverride=true
    bUseSphericalLimits=true
    gravityScale=0.35
    simulationSpace=RootRelative
    boundBone {boneName=dyn_skirt_02_side_l}
    componentLinearAccScale=<1 1 1>
    componentLinearVelScale=<1 1 1>
    gravityOverride {z=-980}
    physicsBodyDefinitions [{
      boundBone {boneName=dyn_skirt_02_side_l}
      boxExtents=<20 30 30>
      constraintSetup {angularLimitsMax=<0 15 15>, angularLimitsMin=<0 -15 0>, angularTarget=<1 0 0>}
      localJointOffset=<2 0 0>
     }]
    simSpaceSettings {maxAngularAcceleration=2500, maxAngularVelocity=2500, simSpaceAngularAlpha=1}
    sphericalLimits [{limitRadius=12, drivingBone {boneName=thigh_l}, sphereLocalOffset=<-27.3807 0 1.5309>}]
   }
   n63 = Anim Dynamics - Bone: dyn_skirt_03_side_l!(ComponentPose=@n62, Alpha=1, SimSpaceSettings:Anim Phys Sim Space Settings Structure=@n57, ExternalForce="0,0,0") -> (Pose) {
    angularDampingOverride=1
    angularSpringConstant=10000
    bAngularSpring=true
    bOverrideAngularDamping=true
    bUseGravityOverride=true
    gravityScale=0.35
    simulationSpace=RootRelative
    boundBone {boneName=dyn_skirt_03_side_l}
    componentLinearAccScale=<1 1 1>
    componentLinearVelScale=<1 1 1>
    gravityOverride {z=-980}
    physicsBodyDefinitions [{
      boundBone {boneName=dyn_skirt_03_side_l}
      boxExtents=<20 30 30>
      constraintSetup {angularLimitsMax=<0 15 15>, angularLimitsMin=<0 -15 0>, angularTarget=<1 0 0>}
      localJointOffset=<2 0 0>
     }]
    simSpaceSettings {maxAngularAcceleration=2500, maxAngularVelocity=2500, simSpaceAngularAlpha=1}
    sphericalLimits [{limitRadius=10, drivingBone {boneName=thigh_l}, sphereLocalOffset=<-35.5035 -4.9091 -4.6745>}]
   }
   n64 = Anim Dynamics - Bone: dyn_skirt_01_back_l!(ComponentPose=@n63, Alpha=1, SimSpaceSettings:Anim Phys Sim Space Settings Structure=@n57, ExternalForce="0,0,0") -> (Pose) {
    angularDampingOverride=1
    angularSpringConstant=12000
    bAngularSpring=true
    bOverrideAngularDamping=true
    bUseGravityOverride=true
    gravityScale=0.35
    simulationSpace=RootRelative
    boundBone {boneName=dyn_skirt_01_back_l}
    componentLinearAccScale=<1 1 1>
    componentLinearVelScale=<1 1 1>
    gravityOverride {z=-980}
    physicsBodyDefinitions [{
      boundBone {boneName=dyn_skirt_01_back_l}
      boxExtents=<20 30 30>
      constraintSetup {angularLimitsMax=<0 0 5>, angularLimitsMin=<-15 -15 -10>, angularTarget=<1 0 0>}
      localJointOffset=<2 0 0>
     }]
    simSpaceSettings {maxAngularAcceleration=2500, maxAngularVelocity=2500, simSpaceAngularAlpha=1}
   }
   n65 = Anim Dynamics - Bone: dyn_skirt_02_back_l!(ComponentPose=@n64, Alpha=1, SimSpaceSettings:Anim Phys Sim Space Settings Structure=@n57, ExternalForce="0,0,0") -> (Pose) {
    angularDampingOverride=1
    angularSpringConstant=12000
    bAngularSpring=true
    bOverrideAngularDamping=true
    bUseGravityOverride=true
    gravityScale=0.35
    simulationSpace=RootRelative
    boundBone {boneName=dyn_skirt_02_back_l}
    componentLinearAccScale=<1 1 1>
    componentLinearVelScale=<1 1 1>
    gravityOverride {z=-980}
    physicsBodyDefinitions [{
      boundBone {boneName=dyn_skirt_02_back_l}
      boxExtents=<20 30 30>
      constraintSetup {angularLimitsMax=<0 0 5>, angularLimitsMin=<-15 -15 -10>, angularTarget=<1 0 0>}
      localJointOffset=<2 0 0>
     }]
    simSpaceSettings {maxAngularAcceleration=2500, maxAngularVelocity=2500, simSpaceAngularAlpha=1}
   }
   n66 = Anim Dynamics - Bone: dyn_skirt_03_back_l!(ComponentPose=@n65, Alpha=1, SimSpaceSettings:Anim Phys Sim Space Settings Structure=@n57, ExternalForce="0,0,0") -> (Pose) {
    angularDampingOverride=1
    angularSpringConstant=12000
    bAngularSpring=true
    bOverrideAngularDamping=true
    bUseGravityOverride=true
    bUseSphericalLimits=true
    gravityScale=0.35
    simulationSpace=RootRelative
    boundBone {boneName=dyn_skirt_03_back_l}
    componentLinearAccScale=<1 1 1>
    componentLinearVelScale=<1 1 1>
    gravityOverride {z=-980}
    physicsBodyDefinitions [{
      boundBone {boneName=dyn_skirt_03_back_l}
      boxExtents=<20 30 30>
      constraintSetup {angularLimitsMax=<0 0 5>, angularLimitsMin=<-15 0 -10>, angularTarget=<1 0 0>}
      localJointOffset=<2 0 0>
     }]
    simSpaceSettings {maxAngularAcceleration=2500, maxAngularVelocity=2500, simSpaceAngularAlpha=1}
    sphericalLimits [
     {limitRadius=16, drivingBone {boneName=calf_l}, sphereLocalOffset=<-19.0884 -5.7536 0.4653>}
     {limitRadius=15, drivingBone {boneName=calf_twist_01_l}, sphereLocalOffset=<-5.2468 -2.6995 0.0976>}
     {limitRadius=10, drivingBone {boneName=MSTR_sword_l}, sphereLocalOffset=<59.9166 0 -1.0737>}
    ]
   }
   n67 = Get Sim Space Settings Skirt()
   n68 = Anim Dynamics - Bone: dyn_skirt_01_front_r!(ComponentPose=@n66, Alpha=1, SimSpaceSettings:Anim Phys Sim Space Settings Structure=@n67, ExternalForce="0,0,0") -> (Pose) {
    angularDampingOverride=1
    angularSpringConstant=10000
    bAngularSpring=true
    bOverrideAngularDamping=true
    bUseGravityOverride=true
    gravityScale=0.35
    simulationSpace=RootRelative
    boundBone {boneName=dyn_skirt_01_front_r}
    componentLinearAccScale=<1 1 1>
    componentLinearVelScale=<1 1 1>
    gravityOverride {z=-980}
    physicsBodyDefinitions [{
      boundBone {boneName=dyn_skirt_01_front_r}
      boxExtents=<20 30 30>
      constraintSetup {angularLimitsMax=<15 15 7>, angularLimitsMin=<-15 -15 -3>, angularTarget=<1 0 0>}
      localJointOffset=<2 0 0>
     }]
    simSpaceSettings {maxAngularAcceleration=2500, maxAngularVelocity=2500, simSpaceAngularAlpha=1}
   }
   n69 = Anim Dynamics - Bone: dyn_skirt_02_front_r!(ComponentPose=@n68, Alpha=1, SimSpaceSettings:Anim Phys Sim Space Settings Structure=@n67, ExternalForce="0,0,0") -> (Pose) {
    angularDampingOverride=1
    angularSpringConstant=10000
    bAngularSpring=true
    bOverrideAngularDamping=true
    bUseGravityOverride=true
    gravityScale=0.35
    simulationSpace=RootRelative
    boundBone {boneName=dyn_skirt_02_front_r}
    componentLinearAccScale=<1 1 1>
    componentLinearVelScale=<1 1 1>
    gravityOverride {z=-980}
    physicsBodyDefinitions [{
      boundBone {boneName=dyn_skirt_02_front_r}
      boxExtents=<20 30 30>
      constraintSetup {angularLimitsMax=<15 15 10>, angularLimitsMin=<-15 -15 -3>, angularTarget=<1 0 0>}
      localJointOffset=<2 0 0>
     }]
    simSpaceSettings {maxAngularAcceleration=2500, maxAngularVelocity=2500, simSpaceAngularAlpha=1}
   }
   n70 = Anim Dynamics - Bone: dyn_skirt_03_front_r!(ComponentPose=@n69, Alpha=1, SimSpaceSettings:Anim Phys Sim Space Settings Structure=@n67, ExternalForce="0,0,0") -> (Pose) {
    angularDampingOverride=1
    angularSpringConstant=10000
    bAngularSpring=true
    bOverrideAngularDamping=true
    bUseGravityOverride=true
    gravityScale=0.35
    simulationSpace=RootRelative
    boundBone {boneName=dyn_skirt_03_front_r}
    componentLinearAccScale=<1 1 1>
    componentLinearVelScale=<1 1 1>
    gravityOverride {z=-980}
    physicsBodyDefinitions [{
      boundBone {boneName=dyn_skirt_03_front_r}
      boxExtents=<20 30 30>
      constraintSetup {angularLimitsMax=<10 15 10>, angularLimitsMin=<-10 -15 -3>, angularTarget=<1 0 0>}
      localJointOffset=<2 0 0>
     }]
    simSpaceSettings {maxAngularAcceleration=2500, maxAngularVelocity=2500, simSpaceAngularAlpha=1}
   }
   n71 = Anim Dynamics - Bone: dyn_skirt_01_side_r!(ComponentPose=@n70, Alpha=1, SimSpaceSettings:Anim Phys Sim Space Settings Structure=@n67, ExternalForce="0,0,0") -> (Pose) {
    angularDampingOverride=1
    angularSpringConstant=10000
    bAngularSpring=true
    bOverrideAngularDamping=true
    bUseGravityOverride=true
    gravityScale=0.35
    simulationSpace=RootRelative
    boundBone {boneName=dyn_skirt_01_side_r}
    componentLinearAccScale=<1 1 1>
    componentLinearVelScale=<1 1 1>
    gravityOverride {z=-980}
    physicsBodyDefinitions [{
      boundBone {boneName=dyn_skirt_01_side_r}
      boxExtents=<20 30 30>
      constraintSetup {angularLimitsMax=<15 10 2>, angularLimitsMin=<-15 -15 -10>, angularTarget=<1 0 0>}
      localJointOffset=<1 0 0>
     }]
    simSpaceSettings {maxAngularAcceleration=2500, maxAngularVelocity=2500, simSpaceAngularAlpha=1}
   }
   n72 = Anim Dynamics - Bone: dyn_skirt_02_side_r!(ComponentPose=@n71, Alpha=1, SimSpaceSettings:Anim Phys Sim Space Settings Structure=@n67, ExternalForce="0,0,0") -> (Pose) {
    angularDampingOverride=1
    angularSpringConstant=10000
    bAngularSpring=true
    bOverrideAngularDamping=true
    bUseGravityOverride=true
    gravityScale=0.35
    simulationSpace=RootRelative
    boundBone {boneName=dyn_skirt_02_side_r}
    componentLinearAccScale=<1 1 1>
    componentLinearVelScale=<1 1 1>
    gravityOverride {z=-980}
    physicsBodyDefinitions [{
      boundBone {boneName=dyn_skirt_02_side_r}
      boxExtents=<20 30 30>
      constraintSetup {angularLimitsMax=<15 10 2>, angularLimitsMin=<-15 -15 -10>, angularTarget=<1 0 0>}
      localJointOffset=<1 0 0>
     }]
    simSpaceSettings {maxAngularAcceleration=2500, maxAngularVelocity=2500, simSpaceAngularAlpha=1}
   }
   n73 = Anim Dynamics - Bone: dyn_skirt_03_side_r!(ComponentPose=@n72, Alpha=1, SimSpaceSettings:Anim Phys Sim Space Settings Structure=@n67, ExternalForce="0,0,0") -> (Pose) {
    angularDampingOverride=1
    angularSpringConstant=10000
    bAngularSpring=true
    bOverrideAngularDamping=true
    bUseGravityOverride=true
    gravityScale=0.35
    simulationSpace=RootRelative
    boundBone {boneName=dyn_skirt_03_side_r}
    componentLinearAccScale=<1 1 1>
    componentLinearVelScale=<1 1 1>
    gravityOverride {z=-980}
    physicsBodyDefinitions [{
      boundBone {boneName=dyn_skirt_03_side_r}
      boxExtents=<20 30 30>
      constraintSetup {angularLimitsMax=<15 10 2>, angularLimitsMin=<-15 -15 -10>, angularTarget=<1 0 0>}
      localJointOffset=<1 0 0>
     }]
    simSpaceSettings {maxAngularAcceleration=2500, maxAngularVelocity=2500, simSpaceAngularAlpha=1}
   }
   n74 = Anim Dynamics - Bone: dyn_skirt_01_back_r!(ComponentPose=@n73, Alpha=1, SimSpaceSettings:Anim Phys Sim Space Settings Structure=@n67, ExternalForce="0,0,0") -> (Pose) {
    angularDampingOverride=1
    angularSpringConstant=12000
    bAngularSpring=true
    bOverrideAngularDamping=true
    bUseGravityOverride=true
    gravityScale=0.35
    simulationSpace=RootRelative
    boundBone {boneName=dyn_skirt_01_back_r}
    componentLinearAccScale=<1 1 1>
    componentLinearVelScale=<1 1 1>
    gravityOverride {z=-980}
    physicsBodyDefinitions [{
      boundBone {boneName=dyn_skirt_01_back_r}
      boxExtents=<20 30 30>
      constraintSetup {angularLimitsMax=<5 10 0>, angularLimitsMin=<-5 -15 -10>, angularTarget=<1 0 0>}
      localJointOffset=<1 0 0>
     }]
    simSpaceSettings {maxAngularAcceleration=2500, maxAngularVelocity=2500, simSpaceAngularAlpha=1}
   }
   n75 = Anim Dynamics - Bone: dyn_skirt_02_back_r!(ComponentPose=@n74, Alpha=1, SimSpaceSettings:Anim Phys Sim Space Settings Structure=@n67, ExternalForce="0,0,0") -> (Pose) {
    angularDampingOverride=1
    angularSpringConstant=12000
    bAngularSpring=true
    bOverrideAngularDamping=true
    bUseGravityOverride=true
    gravityScale=0.35
    simulationSpace=RootRelative
    boundBone {boneName=dyn_skirt_02_back_r}
    componentLinearAccScale=<1 1 1>
    componentLinearVelScale=<1 1 1>
    gravityOverride {z=-980}
    physicsBodyDefinitions [{
      boundBone {boneName=dyn_skirt_02_back_r}
      boxExtents=<20 30 30>
      constraintSetup {angularLimitsMax=<5 10 -1>, angularLimitsMin=<-5 -15 -10>, angularTarget=<1 0 0>}
      localJointOffset=<1 0 0>
     }]
    simSpaceSettings {maxAngularAcceleration=2500, maxAngularVelocity=2500, simSpaceAngularAlpha=1}
   }
   n76 = Anim Dynamics - Bone: dyn_skirt_03_back_r!(ComponentPose=@n75, Alpha=1, SimSpaceSettings:Anim Phys Sim Space Settings Structure=@n67, ExternalForce="0,0,0") -> (Pose) {
    angularDampingOverride=1
    angularSpringConstant=12000
    bAngularSpring=true
    bOverrideAngularDamping=true
    bUseGravityOverride=true
    bUseSphericalLimits=true
    gravityScale=0.35
    simulationSpace=RootRelative
    boundBone {boneName=dyn_skirt_03_back_r}
    componentLinearAccScale=<1 1 1>
    componentLinearVelScale=<1 1 1>
    gravityOverride {z=-980}
    physicsBodyDefinitions [{
      boundBone {boneName=dyn_skirt_03_back_r}
      boxExtents=<20 30 30>
      constraintSetup {angularLimitsMax=<5 10 -1>, angularLimitsMin=<-5 -15 -10>, angularTarget=<1 0 0>}
      localJointOffset=<1 0 0>
     }]
    simSpaceSettings {maxAngularAcceleration=2500, maxAngularVelocity=2500, simSpaceAngularAlpha=1}
    sphericalLimits [{limitRadius=16, drivingBone {boneName=calf_r}, sphereLocalOffset=<10 0 0>}]
   }
   n77 = "Transform (Modify) Bone - Bone: sword_hidebone_l"!(ComponentPose=@n76, bAlphaBoolEnabled:bool=True, Translation="0,0,0", Rotation="0,0,0", Scale="1,0,0") -> (Pose) {alphaInputType=Bool, scaleMode=BMM_Replace, translationMode=BMM_Additive, boneToModify {boneName=sword_hidebone_l}, scale {y=0, z=0}}
   n78 = RigidBody!(ComponentPose=@n77, Alpha, ExternalForce) -> (Pose) {
    bClampLinearTranslationLimitToRefPose=true
    bOverrideWorldGravity=true
    overridePhysicsAsset=&/M_MED_CurbRestore_belt_Physics
    simulationSpace=BaseBoneSpace
    baseBoneRef {boneName=spine_01}
    componentAppliedLinearAccClamp=<3000 3000 3000>
    overrideWorldGravity {z=-980}
    simSpaceSettings {
     dampingAlpha=0
     maxAngularAcceleration=45
     maxAngularVelocity=9
     maxLinearAcceleration=3200
     maxLinearVelocity=800
     velocityScaleZ=0.44
     worldAlpha=1
     externalLinearDragV=<1.5 0.15 0.25>
    }
   }
   n79 = "Constraint - dyn_belt_01 (R)"(ComponentPose=@n78, Alpha=0.1, ConstraintWeights_0=1) {
    alphaBoolBlend {blendInTime=2, blendOutTime=1}
    alphaScaleBiasClamp {bInterpResult=true}
    boneToModify {boneName=dyn_belt_01}
    constraintSetup [{transformType=Rotation, targetBone {boneName=dyn_skirt_01_front_r}}]
    constraintWeights [1]
   }
   n80 = "Constraint - dyn_belt_02 (R)"(ComponentPose=@n79, Alpha=0.1, ConstraintWeights_0=1) {
    alphaBoolBlend {blendInTime=2, blendOutTime=1}
    alphaScaleBiasClamp {bInterpResult=true}
    boneToModify {boneName=dyn_belt_02}
    constraintSetup [{transformType=Rotation, targetBone {boneName=dyn_skirt_01_front_r}}]
    constraintWeights [1]
   }
   n81 = "Constraint - dyn_belt_03 (R)"(ComponentPose=@n80, Alpha=0.4, ConstraintWeights_0=1) {
    alphaBoolBlend {blendInTime=2, blendOutTime=1}
    alphaScaleBiasClamp {bInterpResult=true}
    boneToModify {boneName=dyn_belt_03}
    constraintSetup [{transformType=Rotation, targetBone {boneName=dyn_skirt_02_front_r}}]
    constraintWeights [1]
   }
   n82 = "Constraint - dyn_belt_04 (R)"(ComponentPose=@n81, Alpha=0.6, ConstraintWeights_0=1) {
    alphaBoolBlend {blendInTime=2, blendOutTime=1}
    alphaScaleBiasClamp {bInterpResult=true}
    boneToModify {boneName=dyn_belt_04}
    constraintSetup [{transformType=Rotation, targetBone {boneName=dyn_skirt_02_front_r}}]
    constraintWeights [1]
   }
   Save cached pose 'postDynamics'(Pose=Component To Local(ComponentPose=@n82)) {cacheName=postDynamics}
  ]}
 EventGraph {nodes [
   Event OnCharacterCustomizationLoaded()
   > n1 = Sequence()
   > Get Dynamic Material Instance(SkeletalMeshComponent=Get Array Item with EFort Custom Part Type())
   n2 = CharacterItemDefinitionHasAnyMetaTag(Tags=Get Tags() -> (Tags:Array of Gameplay Tag Structures))
   Set Cape Tag(execute=@n1.then_1, Cape Tag:bool=@n2) -> (Output_Get:bool)
   > Set Cape Tag Inverted(Cape Tag Inverted:bool=NOT Boolean(A=@n2)) -> (Output_Get:bool)
   Event BlueprintUpdateAnimation()
   > n3 = Sequence!(execute) -> (then_0, then_1, then_2:Exec, then_3:Exec, then_4:Exec, then_5:Exec, then_6:Exec, then_7:Exec, then_8:Exec)
   > n4 = Sequence!(execute) -> (then_0, then_1, then_2, then_3, then_4)
   n5 = Get weapon_tag() -> (weapon_tag:Array of Gameplay Tag Structures)
   n6 = Get bIsPlayingEmote()
   > Set EquippedSword(EquippedSword:bool=AND Boolean(A=PickaxeItemDefinitionHasAnyMetaTag(Tags=@n5, bRequiresEquipped=true), B=NOT Boolean(A=@n6))) -> (Output_Get:bool)
   > n7 = Branch(Condition=Get bIsInFrontEnd())
   > Set Sim Space Settings Skirt(Sim Space Settings Skirt:Anim Phys Sim Space Settings Structure=Get Sim Space Settings_FrontEnd() -> (Sim Space Settings_FrontEnd:Anim Phys Sim Space Settings Structure)) -> (Output_Get:Anim Phys Sim Space Settings Structure)
   Set Sim Space Settings Skirt(
    execute=@n7.else
    Sim Space Settings Skirt=Get Sim Space Settings_InZone() -> (Sim Space Settings_InZone:Anim Phys Sim Space Settings Structure)
   )
   n8 = Branch(execute=@n4.then_2, Condition=Get bClothEnabled())
   > n9 = "Cloth Controller Function (Parameter Asset) (Nativized)"!(
    execute
    CharacterPartAnimInstance:Custom Character Part Anim Instance Object Reference=Self-Reference()
    ClothingAssetName:Name=M_MED_CurbRestore_Clothing_0
    ClothCoreSettings:"Cloth Core Settings Structure (by ref)"=Get Cloth Core Settings_0() -> (Cloth Core Settings_0:Cloth Core Settings Structure)
    WindGustLocalSettings:"Wind Gust Settings Structure (by ref)"=Get Wind Gust Local Settings() -> (Wind Gust Local Settings:Wind Gust Settings Structure)
    ClothUpAxis:Vector="0,0,1"
    LinVelScaleFwdBwdClamp:float=0
    LinVelScaleUpDwnClamp:float=0.7
    DeltaTime:float=0
    GameTimeInSeconds:float=GetGameTimeInSeconds()
    StateLogicLayerName:Name=none
    Parameters:Fort Character Dynamics Parameters Object Reference=Get Cloth Controller Parameters() -> (Cloth Controller Parameters:Fort Character Dynamics Parameters Object Reference)
   ) -> (
    then
    OutCurrentState:Name=None
    OutClothParameters:Cloth Parameters Structure
    OutClothCoreSettings:Cloth Core Settings Structure
    OutWindGustSettings:Wind Gust Parameters Structure
    OutWindGustLocalSettings:Wind Gust Settings Structure
    OutWindVelocity:Vector
    OutLinearVelVector:Vector
   )
   > Set Wind Gust Local Settings Out(Wind Gust Local Settings Out:Wind Gust Settings Structure=@n9.OutWindGustLocalSettings) -> (Output_Get:Wind Gust Settings Structure)
   n10 = "Gravity Override State Function (Parameter Asset) (Nativized)"(
    execute=@n8.else
    CharacterPartAnimInstance=Self-Reference()
    GameTimeInSeconds=GetGameTimeInSeconds()
    Parameters=&/M_MED_CurbRestore_Parameter
   )
   > Set GravityOverride(GravityOverride=@n10)
   Set Alpha Skydive and Cape(execute=@n4.then_3, Alpha Skydive and Cape:bool=AND Boolean(A=Get bIsSkyDiving(), B=Get Cape Tag() -> (Cape Tag:bool))) -> (Output_Get:bool)
   > Set PlayMelleWithSim(PlayMelleWithSim:bool=AND Boolean(A=Get bClothEnabled(), B=Get bIsPlayingMeleeAnim())) -> (Output_Get:bool)
   n11 = NOT Boolean(A=Get bIsCrouching())
   > Set Standing movement(Standing movement:bool=AND Boolean(A=Get bIsMoving2D(), B=@n11)) -> (Output_Get:bool)
   > Set Not Crouching(Not Crouching:bool=@n11) -> (Output_Get:bool)
   Set Crouching with simulation(execute=@n4.then_4, Crouching with simulation:bool=AND Boolean(A=Get bIsCrouching(), B=Get bClothEnabled())) -> (Output_Get:bool)
   n12 = Get bIsCrouching()
   n13 = Get bIsMoving2D()
   n14 = NOT Boolean(A=Get bClothEnabled())
   n15 = NOT Boolean(A=Get bRigidBodyEnabled())
   > Set ALT Coruching Moving(ALT Coruching Moving:bool=AND Boolean(A=AND Boolean(A=@n12, B=@n13), B=AND Boolean(A=@n15, B=@n14))) -> (Output_Get:bool)
   n16 = CheckForBackbling()
   Set IsWearingBackpackOrCape(
    execute=@n4.then_1
    IsWearingBackpackOrCape:bool=OR Boolean!(A=@n16, B=@n16.bIsBackblingACape, C:bool=@n16.bIsBackblingLong) -> (ReturnValue)
   ) -> (Output_Get:bool)
   > Set Is Wearing Backbling(Is Wearing Backbling=@n16)
   n17 = Sequence!(execute=@n3.then_1) -> (then_0, then_1)
   Set RbanOrCloth(execute=@n17.then_1, RbanOrCloth:bool=OR Boolean!(A=Get bRigidBodyEnabled(), B=Get bClothEnabled()) -> (ReturnValue)) -> (Output_Get:bool)
   n18 = Branch(execute=@n17.then_0, Condition=Get Is Wearing Backbling())
   n19 = GetPartSkeletalMeshComponent(self=Self-Reference())
   > Set ClothBlendWeight!(execute, ClothBlendWeight:float=0, self:Skeletal Mesh Component Object Reference=@n19) -> (then, Output_Get:float)
   > Set bDisableClothSimulation!(execute, bDisableClothSimulation:bool=true, self:Skeletal Mesh Component Object Reference=@n19) -> (then, Output_Get:bool)
   Set ClothBlendWeight(execute=@n18.else, ClothBlendWeight=1, self=@n19)
   > Set bDisableClothSimulation(bDisableClothSimulation=false, self=@n19)
   n20 = Get bIsMoving2D()
   Set crounchingAndNotMoving(execute=@n3.then_2, crounchingAndNotMoving:bool=AND Boolean(A=NOT Boolean(A=@n20), B=Get bIsCrouching())) -> (Output_Get:bool)
   > Set crouchingAndMoving(crouchingAndMoving:bool=AND Boolean(A=Get bIsMoving2D(), B=Get bIsCrouching())) -> (Output_Get:bool)
   n21 = Get bIsParachuteOpened()
   n22 = Get bIsJumping()
   n23 = Get bIsFalling()
   > Set jumpOrFalling(jumpOrFalling:bool=AND Boolean(A=OR Boolean(A=@n22, B=@n23), B=NOT Boolean(A=@n21))) -> (Output_Get:bool)
   n24 = OR Boolean(A=Get bIsJumping(), B=Get bIsFalling())
   Set movingGround(execute=@n3.then_3, movingGround:bool=AND Boolean(A=Get bIsMoving2D(), B=NOT Boolean(A=@n24))) -> (Output_Get:bool)
   n25 = Get bIsSkydiveFloating()
   n26 = NOT Boolean(A=Get bIsSurfaceSwimming())
   n27 = AND Boolean(A=Get bIsMoving2D(), B=NOT Boolean(A=Get bIsCrouching()))
   > Set MovingStandingOnTheGround(MovingStandingOnTheGround:bool=AND Boolean(A=AND Boolean(A=@n27, B=@n26), B=NOT Boolean(A=@n25))) -> (Output_Get:bool)
   n28 = Get bIsCrouching()
   > Set MovingNotCrouching(MovingNotCrouching:bool=AND Boolean(A=Get bIsMoving2D(), B=NOT Boolean(A=@n28))) -> (Output_Get:bool)
   Set Crouching with cape(execute=@n3.then_4, Crouching with cape:bool=AND Boolean(A=Get bIsCrouching(), B=Get Cape Tag())) -> (Output_Get:bool)
   > Set Crouching without cape(Crouching without cape:bool=AND Boolean(A=Get bIsCrouching(), B=Get Cape Tag Inverted() -> (Cape Tag Inverted:bool))) -> (Output_Get:bool)
   n29 = GetOwnerSkeletalMeshComponent()
   n30 = BreakRotator(InRot=BreakTransform(InTransform=GetDeltaTransformFromRefPose(self=@n29, BoneName=upperarm_l, BaseName=clavicle_l)).Rotation)
   n31 = Get Y percentage() -> (Y percentage:float)
   n32 = Get Z percentage() -> (Z percentage:float)
   n33 = Get X percentage() -> (X percentage:float)
   Set Upperarm Rotation(
    execute=@n3.then_5
    Upperarm Rotation:Rotator=MakeRotator(
     Roll=float * float!(A=@n30, B=-1, C:double=@n33) -> (ReturnValue)
     Pitch=float * float(A=@n30.Pitch, B=-1, C=@n31)
     Yaw=float * float(A=@n30.Yaw, B=-1, C=@n32)
    )
   ) -> (Output_Get:Rotator)
   n34 = PlayerMovement(
    execute=[@n3.then_6, @n38.then_0]
    AnimInstance=Self-Reference()
    XDotProductCoefficient=1
    YDotProductCoefficient=1
    ZDotProductCoefficient=1
   )
   > n35 = Branch(Condition=Get bIsDBNOCarried())
   > Set BwdMovement(BwdMovement:double=0.5) -> (Output_Get:double)
   n36 = Branch(execute=@n35.else, Condition=AND Boolean(A=Get bIsTargeting(), B=Get bIsCrouching()))
   n37 = MapRangeClamped(
    Value=float + float(A="Absolute (Float)"(A=@n34), B=MapRangeClamped(Value=@n34.YMovement, InRangeA=-1, InRangeB=1, OutRangeA=0.3)) {is=title K2Node_CommutativeAssociativeBinaryOperator}
    InRangeA=0
    InRangeB=1.65
   )
   > Set BwdMovement(BwdMovement=float * float(A=@n37, B=0.25) {is=title K2Node_PromotableOperator})
   Set BwdMovement(execute=@n36.else, BwdMovement=@n37)
   > Branch(Condition=Get bIsSkydiveDiveMode())
   > Set BwdMovement(BwdMovement=0)
   n38 = Sequence!(execute=@n3.then_7) -> (then_0, then_1, then_2:Exec)
   n39 = Branch(execute=@n38.then_1, Condition=Get bIsSkyDiving())
   > n40 = Branch(Condition="float > float"(A=Get PawnVelocityZ(), B=50) {is=title K2Node_CallFunction})
   > Set BwdMovement(BwdMovement=float * float!(A=Get BwdMovement(), B=0) -> (ReturnValue) {is=title K2Node_CommutativeAssociativeBinaryOperator})
   Set BwdMovement(execute=[@n39.else, @n40.else], BwdMovement=float * float(A=Get BwdMovement(), B=1))
   n41 = Sequence(execute=@n3.then_8)
   > Branch(Condition=Get bIsInFrontEnd())
   > Set Custom Idle Is Playing(Custom Idle Is Playing:double=GetCurveValue(self=GetOwnerFrontEndAnimInstance(), CurveName=Frontend_Idle_Bypass)) -> (Output_Get:double)
   n42 = GetAnimInstance(self=GetOwnerSkeletalMeshComponent())
   Is Valid(exec=@n41.then_1, InputObject=@n42)
   n43 = Montage_IsActive!(
    self:Anim Instance Object Reference=@n42
    Montage:Anim Montage Object Reference=/BRCosmetics/Animation/Game/MainPlayer/Emotes/CurbRestore/Hero/Emore_CurbRestore_Hero_Pickaxe_M
   ) -> (ReturnValue:bool)
   > Set Emote is active(Emote is active:bool=@n43) -> (Output_Get:bool)
   > Set Emote is active 01 Value(Emote is active 01 Value:double=SelectFloat(A=0.1, B=0, bPickA=@n43)) -> (Output_Get:double)
   > Set Emote is active 03 Value(Emote is active 03 Value:double=SelectFloat(A=0.3, B=0, bPickA=@n43)) -> (Output_Get:double)
   > Set Emote is active 08 Value(Emote is active 08 Value:double=SelectFloat(A=0.8, B=0, bPickA=@n43)) -> (Output_Get:double)
   n44 = GetAnimInstance(self=GetOwnerSkeletalMeshComponent())
   n45 = Montage_IsPlaying(self=@n44, Montage=/BRCosmetics/Animation/Game/MainPlayer/Emotes/GoldenGlory/CMM/Emote_GoldenGlory_CMM_M)
   n46 = Montage_IsPlaying(self=@n44, Montage=/BRCosmetics/Animation/Game/MainPlayer/Emotes/GoldenGlory/CMM/Emote_GoldenGlory_CMM_FrontEnd_M)
   Set GoldenGloryEmote(execute=@n41.then_2, GoldenGloryEmote:double="To Float (Boolean)"(InBool=OR Boolean(A=@n45, B=@n46))) -> (Output_Get:double)
   Event OnBeginSkydiving()
   > Set SkydivingAlpha(SkydivingAlpha:Rotator="65,0,0") -> (Output_Get:Rotator)
   Event OnEndSkydiving()
   > Set SkydivingAlpha(SkydivingAlpha="10,0,0")
  ]}
]
variables [
 Alpha Skydive and Cape {type=boolean}
 ALT Coruching Moving {type=boolean}
 Body_Mesh {type=SkeletalMeshComponent}
 BwdMovement {type=number}
 Cape Tag {type=boolean}
 Cape Tag Inverted {type=boolean}
 Cloth Controller Parameters {
  type=FortCharacterDynamicsParameters
  default=&/Fortnite_Base_CapeCurbRestore_Parameters
 }
 Cloth Core Settings_0 {
  type=ClothCoreSettings
  default {
   collisionThickness=1.5
   dampingCoefficient=0.05
   frictionCoefficient=0.3
   localDampingCoefficient=0
   numberIterations=2
   numberSubsteps=1
   selfCollisionThickness=0
   animDriveDamping=<0 1>
   animDriveStiffness=<0 1>
   areaStiffness=<1.2 1.2>
   bendingStiffness=<1.2 1.2>
   edgeStiffness=<1.5 1.5>
   tetherStiffness=<1.2 1.2>
  }
 }
 Cloth Value {type=boolean}
 Crouching with cape {type=boolean}
 Crouching with simulation {type=boolean}
 Crouching without cape {type=boolean}
 crouchingAndMoving {type=boolean}
 crounchingAndNotMoving {type=boolean}
 Custom Idle Is Playing {type=number}
 Emote is active {type=boolean}
 Emote is active 01 Value {type=number}
 Emote is active 03 Value {type=number}
 Emote is active 08 Value {type=number}
 EquippedSword {type=boolean}
 GoldenGloryEmote {type=number}
 GravityOverride {type=Vector}
 GravityOverride Parameters {type=FortCharacterDynamicsParameters, default=null}
 Is Wearing Backbling {type=boolean}
 IsWearingBackpackOrCape {type=boolean}
 jumpOrFalling {type=boolean}
 movingGround {type=boolean}
 MovingNotCrouching {type=boolean}
 MovingStandingOnTheGround {type=boolean}
 Not Crouching {type=boolean}
 PlayMelleWithSim {type=boolean}
 RbanOrCloth {type=boolean}
 Sim Space Settings {
  type=SimSpaceSettings
  default {
   dampingAlpha=1
   maxAngularAcceleration=10000
   maxAngularVelocity=10000
   maxLinearAcceleration=10000
   maxLinearVelocity=10000
   velocityScaleZ=1
   worldAlpha=0
   externalLinearDragV=<0 0 0>
  }
 }
 Sim Space Settings Skirt {
  type=AnimPhysSimSpaceSettings
  default {maxAngularAcceleration=2500, maxAngularVelocity=2500, simSpaceAngularAlpha=1, externalAngularVelocity=<0 0 0>}
 }
 Sim Space Settings_FrontEnd {
  type=AnimPhysSimSpaceSettings
  default {maxAngularAcceleration=1000, maxAngularVelocity=1000, simSpaceAngularAlpha=0.2, externalAngularVelocity=<0 0 0>}
 }
 Sim Space Settings_InZone {
  type=AnimPhysSimSpaceSettings
  default {maxAngularAcceleration=2000, maxAngularVelocity=2000, simSpaceAngularAlpha=1, externalAngularVelocity=<0 0 0>}
 }
 SkydivingAlpha {type=Rotator}
 Standing movement {type=boolean}
 Tags {type="array<GameplayTag>", default [{tagName=Cosmetics.Character.RestrictCape}]}
 Upperarm Rotation {type=Rotator}
 weapon_tag {type="array<GameplayTag>", default [{tagName=Cosmetics.MetaTags.CurbRestore}]}
 Wind Gust Local Settings {type=WindGustSettings, default {elapsedTime=0, finalGustStrengthLocal=0, gustDuration=0, gustStrength=0, timeBetweenGusts=0}}
 Wind Gust Local Settings Out {type=WindGustSettings, default {elapsedTime=0, finalGustStrengthLocal=0, gustDuration=0, gustStrength=0, timeBetweenGusts=0}}
 X percentage {type=number, default=0.2}
 Y percentage {type=number, default=0.2}
 Z percentage {type=number, default=0.2}
]
