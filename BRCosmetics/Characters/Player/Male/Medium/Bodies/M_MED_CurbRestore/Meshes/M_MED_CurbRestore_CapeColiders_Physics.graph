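# Physics asset of cape colliders for the M_MED_CurbRestore male character body, apparently for simulated cape motion to collide against.
# Capsules cover pelvis, spine, arms, hands, legs and feet, plus the two MSTR_sword bones and the back dyn_skirt_02 bones.
# A huge sphere on root, named dyn_floor_collision_sphere, sits below the character with its top just above the origin to act as a floor.
# It has no constraints, so the bodies are not jointed as a ragdoll.
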PhysicsAsset /BRCosmetics/Characters/Player/Male/Medium/Bodies/M_MED_CurbRestore/Meshes/M_MED_CurbRestore_CapeColiders_Physics
bodies {
 MSTR_sword_l [{shapeType=Capsule, length=42.9994, radius=13.4506, center=<44.5436 0.0788 10.0674>, rotation=<rot 0 -91.2547 -89.3499>}]
 MSTR_sword_r [{shapeType=Capsule, length=37.2291, radius=13.4432, center=<34.8591 8.3907 -2.8651>, rotation=<rot 0 -91.8768 -89.5089>}]
 calf_l [{shapeType=Capsule, length=26.7275, radius=9.7232, center=<-28.3879 -2.5337 0.7747>, rotation=<rot 0 -91.3347 -89.1215>}]
 calf_r [{shapeType=Capsule, length=26.7275, radius=9.7232, center=<28.3879 2.5337 -0.7747>, rotation=<rot 0 -91.3347 90.8785>}]
 dyn_skirt_02_back_l [{shapeType=Capsule, length=20.3945, radius=17.928, center=<5.1234 13.5835 0>, rotation=<rot 90 3.1181 0>}]
 dyn_skirt_02_back_r [{shapeType=Capsule, length=20.3945, radius=17.928, center=<5.0948 11.1499 1.9889>, rotation=<rot 90 -178.2064 0>}]
 foot_l [{shapeType=Capsule, length=14.2477, radius=10.1784, center=<-3.5448 2.4526 0.4227>, rotation=<rot 0 5.4609 -85.3315>}]
 foot_r [{shapeType=Capsule, length=14.2477, radius=10.1784, center=<3.5448 -2.4526 -0.4227>, rotation=<rot 0 5.4609 94.6685>}]
 hand_l [{shapeType=Capsule, length=8.5692, radius=7.2466, center=<6.8246 0 0>, rotation=<rot 90 180 0>}]
 hand_r [{shapeType=Capsule, length=8.5692, radius=7.2466, center=<-6.8246 0 0>, rotation=<rot 90 0 0>}]
 lowerarm_l [{shapeType=Capsule, length=18.4935, radius=6.979, center=<9.0358 0 0>, rotation=<rot 90 180 0>}]
 lowerarm_r [{shapeType=Capsule, length=18.4935, radius=6.979, center=<-9.0358 0 0>, rotation=<rot 90 0 0>}]
 pelvis [{shapeType=Capsule, length=11.9124, radius=14.7281, center=<-7.0025 -2.9105 0>}]
 root [{shapeType=Sphere, radius=281.2503, shapeName=dyn_floor_collision_sphere, center=<0 0 -278.8158>}]
 spine_01 [{shapeType=Capsule, length=14.0625, radius=11.133, center=<0 -2.6211 0>}]
 spine_02 [{shapeType=Capsule, length=14.2357, radius=11.3366, shapeName=spine_02_capsule, center=<5.9356 -0.1834 0>}]
 spine_03 [{shapeType=Capsule, length=15.368, radius=12.1543, shapeName=spine_03_capsule, center=<6.3888 1.4518 0>}]
 spine_04 [{shapeType=Capsule, length=10.8612, radius=11.8765, center=<3.8067 1.1563 0.2674>, rotation=<rot -3.0034 -90 1.6141>}]
 spine_05 [{shapeType=Capsule, length=15.6656, radius=11.2395, center=<0 1.6207 0>}]
 thigh_l [{shapeType=Capsule, length=29.1008, radius=10.5792, center=<-26.0497 0 0>, rotation=<rot 90 0 0>}]
 thigh_r [{shapeType=Capsule, length=29.1008, radius=10.5792, center=<26.0497 0 0>, rotation=<rot 90 180 0>}]
 upperarm_l [{shapeType=Capsule, length=15.093, radius=5.3882, shapeName=upperarm_l_capsule, center=<12.4756 0.438 -1.5466>, rotation=<rot 90 -174.1698 0>}]
 upperarm_r [{shapeType=Capsule, length=15.093, radius=5.3882, shapeName=upperarm_r_capsule, center=<-12.4756 -0.438 1.5466>, rotation=<rot 90 5.8302 0>}]
}
constraints []
