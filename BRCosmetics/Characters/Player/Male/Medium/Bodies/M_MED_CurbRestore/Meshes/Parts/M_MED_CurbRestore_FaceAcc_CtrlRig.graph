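# Control Rig for the CurbRestore male medium face accessory that offsets the center and right hair master bones based on head pose, only when `cape` is false.
# Spherical pose readers on the head output weights that are scaled, interpolated and fed into self-relative transform offsets.
# A PistolADS-scaled pass moves both bones standing, or only the right when crouching.
# A second pass, scaled by inverse IsTargeting Alpha, runs only when not crouching.
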ControlRigBlueprint /BRCosmetics/Characters/Player/Male/Medium/Bodies/M_MED_CurbRestore/Meshes/Parts/M_MED_CurbRestore_FaceAcc_CtrlRig
graphs [
 RigVMFunctionLibrary {links [], nodes []}
 RigVMFunction_Sequence_1_ContainedGraph {nodes [
   n1 = RigVMFunctionEntryNode()
   RigVMFunction_Sequence = &RigVMFunction_Sequence(ExecuteContext=@n1.ExecuteContext)
   n2 = &RigVMFunction_Sequence(ExecuteContext=@RigVMFunction_Sequence.B)
   RigVMFunctionReturnNode(A=@RigVMFunction_Sequence.A, B=@n2.A, C=@n2.B)
  ]}
 RigVMModel {nodes [
   RigUnit_BeginExecution = &RigUnit_BeginExecution()
   RigVMFunction_ControlFlowBranch = &RigVMFunction_ControlFlowBranch(ExecuteContext=@RigUnit_BeginExecution.ExecutePin, Condition=RigVMVariableNode() {variable=cape}.Value)
   RigVMFunction_Sequence = RigVMAggregateNode(ExecuteContext=@RigVMFunction_ControlFlowBranch.False)
   n1 = &RigVMFunction_ControlFlowBranch(ExecuteContext=@RigVMFunction_Sequence.B, Condition=RigVMVariableNode() {variable=Crouching}.Value)
   n2 = &RigUnit_SphericalPoseReader(
    ActiveRegionScaleFactors=(PositiveWidth=1,NegativeWidth=1,PositiveHeight=1,NegativeHeight=1)
    ActiveRegionSize=0.1
    Debug=(bDrawDebug=True,bDraw2D=False,bDrawLocalAxes=False,DebugScale=25,DebugSegments=20,DebugThickness=0.25)
    DriverAxis=(X=-1,Y=0,Z=0)
    DriverItem=(Type=Bone,Name="head")
    FalloffRegionScaleFactors=(PositiveWidth=1,NegativeWidth=1,PositiveHeight=1,NegativeHeight=1)
    FalloffSize=0.2
    FlipHeightScaling=False
    FlipWidthScaling=False
    OptionalParentItem=(Type=Bone,Name="None")
    RotationOffset=(X=35,Y=102,Z=0)
    ExecutePin=@n1.False
   )
   Multiply = &RigVMFunction_MathDoubleMul(A=@n2.OutputParam, B=RigVMVariableNode() {variable=PistolADS}.Value)
   AlphaInterp = &RigVMFunction_AlphaInterp(InterpSpeedDecreasing=10, InterpSpeedIncreasing=10, Scale=1, bClampResult=False, Value=@Multiply.Result)
   > RigUnit_SetRelativeTransformForItem = /Script/ControlRig.RigUnit_SetRelativeTransformForItem(
    Child=(Type=Bone,Name="MSTR_hair_c")
    Parent=(Type=Bone,Name="MSTR_hair_c")
    Value=(Rotation=(X=0.0298,Y=0.0819,Z=-0.3407,W=0.9361),Translation=(X=0,Y=0,Z=0),Scale3D=(X=1,Y=1,Z=1))
    bParentInitial=False
    bPropagateToChildren=True
    Weight=@AlphaInterp.Result
   ) {
    is=struct RigVMUnitNode
    template="Set Relative Transform::Execute(in Child,in Parent,in bParentInitial,in Value,in Weight,in bPropagateToChildren)"
   }
   > &RigUnit_SetRelativeTransformForItem(
    Child=(Type=Bone,Name="MSTR_hair_r")
    Parent=(Type=Bone,Name="MSTR_hair_r")
    Value=(Rotation=(X=0.0058,Y=0.1304,Z=-0.0433,W=0.9905),Translation=(X=-1,Y=0,Z=0),Scale3D=(X=1,Y=1,Z=1))
    bParentInitial=False
    bPropagateToChildren=True
    Weight=@AlphaInterp.Result
   ) {template="Set Relative Transform::Execute(in Child,in Parent,in bParentInitial,in Value,in Weight,in bPropagateToChildren)"}
   n3 = &RigUnit_SphericalPoseReader(
    ActiveRegionScaleFactors=(PositiveWidth=1,NegativeWidth=1,PositiveHeight=1,NegativeHeight=1)
    ActiveRegionSize=0.1
    Debug=(bDrawDebug=True,bDraw2D=False,bDrawLocalAxes=False,DebugScale=25,DebugSegments=20,DebugThickness=0.25)
    DriverAxis=(X=-1,Y=0,Z=0)
    DriverItem=(Type=Bone,Name="head")
    FalloffRegionScaleFactors=(PositiveWidth=1,NegativeWidth=1,PositiveHeight=1,NegativeHeight=1)
    FalloffSize=0.2
    FlipHeightScaling=False
    FlipWidthScaling=False
    OptionalParentItem=(Type=Bone,Name="None")
    RotationOffset=(X=35,Y=102,Z=0)
    ExecutePin=@n1.True
   )
   n4 = &RigVMFunction_AlphaInterp(
    InterpSpeedDecreasing=8
    InterpSpeedIncreasing=8
    Scale=1
    bClampResult=False
    Value=&RigVMFunction_MathDoubleMul(A=@n3.OutputParam, B=RigVMVariableNode() {variable=PistolADS}.Value).Result
   )
   > &RigUnit_SetRelativeTransformForItem(
    Child=(Type=Bone,Name="MSTR_hair_r")
    Parent=(Type=Bone,Name="MSTR_hair_r")
    Value=(Rotation=(X=0,Y=0,Z=0,W=1),Translation=(X=-3,Y=0,Z=0),Scale3D=(X=1,Y=1,Z=1))
    bParentInitial=False
    bPropagateToChildren=True
    Weight=@n4.Result
   ) {template="Set Relative Transform::Execute(in Child,in Parent,in bParentInitial,in Value,in Weight,in bPropagateToChildren)"}
   n5 = &RigVMFunction_ControlFlowBranch(ExecuteContext=@RigVMFunction_Sequence.C, Condition=RigVMVariableNode() {variable=Crouching}.Value)
   n6 = &RigUnit_SphericalPoseReader(
    ActiveRegionScaleFactors=(PositiveWidth=1,NegativeWidth=1,PositiveHeight=1,NegativeHeight=1)
    ActiveRegionSize=0.1
    Debug=(bDrawDebug=True,bDraw2D=False,bDrawLocalAxes=False,DebugScale=25,DebugSegments=20,DebugThickness=0.25)
    DriverAxis=(X=-1,Y=0,Z=0)
    DriverItem=(Type=Bone,Name="head")
    FalloffRegionScaleFactors=(PositiveWidth=1,NegativeWidth=1,PositiveHeight=1,NegativeHeight=1)
    FalloffSize=0.2
    FlipHeightScaling=False
    FlipWidthScaling=False
    OptionalParentItem=(Type=Bone,Name="None")
    RotationOffset=(X=35,Y=102,Z=0)
    ExecutePin=@n5.False
   )
   Subtract = &RigVMFunction_MathDoubleSub(A=1, B=RigVMVariableNode() {variable=IsTargeting Alpha}.Value)
   n7 = &RigVMFunction_AlphaInterp(
    InterpSpeedDecreasing=8
    InterpSpeedIncreasing=8
    Scale=1
    bClampResult=False
    Value=&RigVMFunction_MathDoubleMul(A=@n6.OutputParam, B=@Subtract.Result).Result
   )
   > &RigUnit_SetRelativeTransformForItem(
    Child=(Type=Bone,Name="MSTR_hair_c")
    Parent=(Type=Bone,Name="MSTR_hair_c")
    Value=(Rotation=(X=0.0298,Y=0.0819,Z=-0.3407,W=0.9361),Translation=(X=-3,Y=0,Z=0),Scale3D=(X=1,Y=1,Z=1))
    bParentInitial=False
    bPropagateToChildren=True
    Weight=@n7.Result
   ) {template="Set Relative Transform::Execute(in Child,in Parent,in bParentInitial,in Value,in Weight,in bPropagateToChildren)"}
   > &RigUnit_SetRelativeTransformForItem(
    Child=(Type=Bone,Name="MSTR_hair_r")
    Parent=(Type=Bone,Name="MSTR_hair_r")
    Value=(Rotation=(X=0.0226,Y=0.0842,Z=-0.2578,W=0.9623),Translation=(X=-1.5,Y=0,Z=0),Scale3D=(X=1,Y=1,Z=1))
    bParentInitial=False
    bPropagateToChildren=True
    Weight=@n7.Result
   ) {template="Set Relative Transform::Execute(in Child,in Parent,in bParentInitial,in Value,in Weight,in bPropagateToChildren)"}
  ]}
]
hierarchy [
 root {loc=<0 0 0>, rot=<0 0 0 1>}
 pelvis {parent=root, loc=<0 -0.7071 93.1191>, rot=<0.0127 0.707 -0.0127 -0.707>}
 spine_01 {parent=pelvis, loc=<6.7494 -0.6388 0>, rot=<0 0 0.034 -0.9994>}
 spine_02 {parent=spine_01, loc=<7.2518 -0.035 0>, rot=<0 0 0.0238 -0.9997>}
 spine_03 {parent=spine_02, loc=<6.0104 0 0>, rot=<0 0 0.0149 -0.9999>}
 spine_04 {parent=spine_03, loc=<8.1024 0 0>, rot=<0 0 -0.0132 -0.9999>}
 spine_05 {parent=spine_04, loc=<10.0146 0 0>, rot=<0 0 -0.0015 -1>}
 neck_01 {parent=spine_05, loc=<11.3138 0.1529 0>, rot=<0 0 -0.2003 -0.9797>}
 neck_02 {parent=neck_01, loc=<4.5627 0 0>, rot=<0 0 0.0504 -0.9987>}
 head {parent=neck_02, loc=<5.3914 0 0>, rot=<0 0 0.115 -0.9934>}
 MSTR_hair_l {parent=head, loc=<4.905 0.2861 -7.6753>, rot=<0.0041 -1 0 0>}
 dyn_hair_01_l {parent=MSTR_hair_l, loc=<0 0 0>, rot=<0 0 0 1>}
 dyn_hair_02_l {parent=dyn_hair_01_l, loc=<4.3063 0 0>, rot=<0 0 0 1>}
 MSTR_hair_r {parent=head, loc=<4.905 0.2861 7.8451>, rot=<0.0041 -1 0 0>}
 dyn_hair_01_r {parent=MSTR_hair_r, loc=<0 0 0>, rot=<0 0 0 1>}
 dyn_hair_02_r {parent=dyn_hair_01_r, loc=<4.5743 0 0>, rot=<0 0 0 1>}
 MSTR_hair_c {parent=head, loc=<5.6142 -7.0804 0>, rot=<0.0041 -1 0 0>}
 dyn_hair_01_c {parent=MSTR_hair_c, loc=<0 0 0>, rot=<0 0 0 1>}
 dyn_hair_02_c {parent=dyn_hair_01_c, loc=<4.9337 0 0>, rot=<0 0 0 1>}
]
variables [
 Crouching {type=bool, default=False}
 IsTargeting Alpha {type=double, default=0}
 PistolADS {type=double, default=0}
 cape {type=bool, default=False}
]
