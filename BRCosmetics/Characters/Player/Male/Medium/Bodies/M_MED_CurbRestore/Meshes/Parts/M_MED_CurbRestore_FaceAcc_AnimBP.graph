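# Anim Blueprint for the CurbRestore male face-accessory part, adding simulated hair motion to the pose copied from the parent mesh.
# A pre-dynamics Control Rig feeds two left, right and center AnimDynamics hair chains; the RestrictCape cosmetic tag picks one, only the non-cape chain using spherical limits.
# A second Control Rig finishes the pose.
# The EventGraph feeds the rigs targeting and pistol-or-rifle aiming alphas, alongside cape and crouch states.
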AnimBlueprint /BRCosmetics/Characters/Player/Male/Medium/Bodies/M_MED_CurbRestore/Meshes/Parts/M_MED_CurbRestore_FaceAcc_AnimBP
target_skeleton=&/M_MED_CurbRestore_FaceAcc_Skeleton
graphs [
 AnimGraph {nodes [
   n1 = Anim Dynamics - Bone: dyn_hair_01_l(ComponentPose=Local To Component(LocalPose=Use cached pose 'StartPose'()), Alpha=1, ExternalForce="0,0,0") {
    angularDampingOverride=1
    angularSpringConstant=3500
    bAngularSpring=true
    bOverrideAngularDamping=true
    bUseGravityOverride=true
    gravityScale=0.35
    simulationSpace=RootRelative
    boundBone {boneName=dyn_hair_01_l}
    componentAppliedLinearAccClamp=<3000 3000 3000>
    componentLinearAccScale=<0.9 0.9 0.8>
    componentLinearVelScale=<0.8 0.8 0.8>
    gravityOverride {z=-980}
    physicsBodyDefinitions [{
      boundBone {boneName=dyn_hair_01_l}
      boxExtents=<28 28 28>
      constraintSetup {angularLimitsMax=<0 0 5>, angularLimitsMin=<0 -15 -15>, angularTarget=<1 0 0>}
      localJointOffset=<2 0 0.5>
     }]
   }
   n2 = Anim Dynamics - Bone: dyn_hair_02_l(ComponentPose=@n1, Alpha=1, ExternalForce="0,0,0") {
    angularDampingOverride=1
    angularSpringConstant=3500
    bAngularSpring=true
    bOverrideAngularDamping=true
    bUseGravityOverride=true
    gravityScale=0.35
    simulationSpace=RootRelative
    boundBone {boneName=dyn_hair_02_l}
    componentAppliedLinearAccClamp=<3000 3000 3000>
    componentLinearAccScale=<0.8 0.8 0.8>
    componentLinearVelScale=<0.8 0.8 0.8>
    gravityOverride {z=-980}
    physicsBodyDefinitions [{
      boundBone {boneName=dyn_hair_02_l}
      boxExtents=<29 29 29>
      constraintSetup {angularLimitsMax=<0 0 15>, angularLimitsMin=<0 -15 -15>, angularTarget=<1 0 0>}
      localJointOffset=<2 0 0.5>
     }]
   }
   n3 = Anim Dynamics - Bone: dyn_hair_01_r(ComponentPose=@n2, Alpha=1, ExternalForce="0,0,0") {
    angularDampingOverride=1
    angularSpringConstant=3500
    bAngularSpring=true
    bOverrideAngularDamping=true
    bUseGravityOverride=true
    gravityScale=0.35
    simulationSpace=RootRelative
    boundBone {boneName=dyn_hair_01_r}
    componentAppliedLinearAccClamp=<3000 3000 3000>
    componentLinearAccScale=<0.8 0.8 0.8>
    componentLinearVelScale=<0.8 0.8 0.8>
    gravityOverride {z=-980}
    physicsBodyDefinitions [{
      boundBone {boneName=dyn_hair_01_r}
      boxExtents=<26 26 26>
      constraintSetup {angularLimitsMax=<0 15 15>, angularLimitsMin=<0 0 -15>, angularTarget=<1 0 0>}
      localJointOffset=<1.7 0 -1.3>
     }]
   }
   n4 = Anim Dynamics - Bone: dyn_hair_02_r(ComponentPose=@n3, Alpha=1, ExternalForce="0,0,0") {
    angularDampingOverride=1
    angularSpringConstant=3500
    bAngularSpring=true
    bOverrideAngularDamping=true
    bUseGravityOverride=true
    gravityScale=0.35
    simulationSpace=RootRelative
    boundBone {boneName=dyn_hair_02_r}
    componentAppliedLinearAccClamp=<3000 3000 3000>
    componentLinearAccScale=<0.8 0.8 0.8>
    componentLinearVelScale=<0.8 0.8 0.8>
    gravityOverride {z=-980}
    physicsBodyDefinitions [{
      boundBone {boneName=dyn_hair_02_r}
      boxExtents=<27 27 27>
      constraintSetup {angularLimitsMax=<0 15 15>, angularLimitsMin=<0 0 -15>, angularTarget=<1 0 0>}
      localJointOffset=<1.7 0 -1>
     }]
   }
   n5 = Anim Dynamics - Bone: dyn_hair_01_c(ComponentPose=@n4, Alpha=1, ExternalForce="0,0,0") {
    angularDampingOverride=1
    angularSpringConstant=3500
    bAngularSpring=true
    bOverrideAngularDamping=true
    bUseGravityOverride=true
    gravityScale=0.35
    simulationSpace=RootRelative
    boundBone {boneName=dyn_hair_01_c}
    componentAppliedLinearAccClamp=<3000 3000 3000>
    componentLinearAccScale=<0.8 0.8 0.8>
    componentLinearVelScale=<0.8 0.8 0.8>
    gravityOverride {z=-980}
    physicsBodyDefinitions [{
      boundBone {boneName=dyn_hair_01_c}
      boxExtents=<27 27 27>
      constraintSetup {angularLimitsMax=<4 15 15>, angularLimitsMin=<-4 -15 -10>, angularTarget=<1 0 0>}
      localJointOffset=<1.7 1.3 0>
     }]
   }
   n6 = Anim Dynamics - Bone: dyn_hair_02_c(ComponentPose=@n5, Alpha=1, ExternalForce="0,0,0") {
    angularDampingOverride=1
    angularSpringConstant=3500
    bAngularSpring=true
    bOverrideAngularDamping=true
    bUseGravityOverride=true
    gravityScale=0.35
    simulationSpace=RootRelative
    boundBone {boneName=dyn_hair_02_c}
    componentAppliedLinearAccClamp=<3000 3000 3000>
    componentLinearAccScale=<0.8 0.8 0.8>
    componentLinearVelScale=<0.8 0.8 0.8>
    gravityOverride {z=-980}
    physicsBodyDefinitions [{
      boundBone {boneName=dyn_hair_02_c}
      boxExtents=<29 29 29>
      constraintSetup {angularLimitsMax=<4 15 15>, angularLimitsMin=<-4 -15 -15>, angularTarget=<1 0 0>}
      localJointOffset=<1.35 1 0>
     }]
   }
   n7 = Anim Dynamics - Bone: dyn_hair_01_l(ComponentPose=Local To Component(LocalPose=Use cached pose 'StartPose'()), Alpha=1, ExternalForce="0,0,0") {
    angularDampingOverride=1
    angularSpringConstant=3900
    bAngularSpring=true
    bOverrideAngularDamping=true
    bUseGravityOverride=true
    bUsePlanarLimit=false
    gravityScale=0.35
    simulationSpace=RootRelative
    boundBone {boneName=dyn_hair_01_l}
    componentAppliedLinearAccClamp=<2200 2200 2200>
    componentLinearAccScale=<0.9 0.9 0.9>
    componentLinearVelScale=<0.9 0.9 0.9>
    gravityOverride {z=-980}
    physicsBodyDefinitions [{
      boundBone {boneName=dyn_hair_01_l}
      boxExtents=<25 25 25>
      constraintSetup {angularLimitsMax=<0 0 9>, angularLimitsMin=<0 -20 -20>, angularTarget=<1 0 0>}
      localJointOffset=<2 0 0>
     }]
    relativeSpaceBone {boneName=head}
    simSpaceSettings {maxAngularAcceleration=2200, maxAngularVelocity=16}
   }
   n8 = Anim Dynamics - Bone: dyn_hair_02_l(ComponentPose=@n7, Alpha=1, ExternalForce="0,0,0") {
    angularDampingOverride=1
    angularSpringConstant=4000
    bAngularSpring=true
    bOverrideAngularDamping=true
    bUseGravityOverride=true
    bUseSphericalLimits=true
    gravityScale=0.35
    simulationSpace=RootRelative
    boundBone {boneName=dyn_hair_02_l}
    componentAppliedLinearAccClamp=<2200 2200 2200>
    componentLinearAccScale=<1 1 1>
    componentLinearVelScale=<1 1 1>
    gravityOverride {z=-980}
    physicsBodyDefinitions [{
      collisionType=CustomSphere
      boundBone {boneName=dyn_hair_02_l}
      boxExtents=<25 25 25>
      constraintSetup {angularLimitsMax=<0 0 9>, angularLimitsMin=<0 -20 -17>, angularTarget=<1 0 0>}
      localJointOffset=<2 0 0>
     }]
    relativeSpaceBone {boneName=head}
    simSpaceSettings {maxAngularAcceleration=2200, maxAngularVelocity=15}
    sphericalLimits [{limitRadius=12, drivingBone {boneName=spine_05}, sphereLocalOffset=<-3.5384 2.6896 -6.6438>}]
   }
   n9 = Anim Dynamics - Bone: dyn_hair_01_r(ComponentPose=@n8, Alpha=1, ExternalForce="0,0,0") {
    angularDampingOverride=1
    angularSpringConstant=3850
    bAngularSpring=true
    bOverrideAngularDamping=true
    bUseGravityOverride=true
    gravityScale=0.35
    simulationSpace=RootRelative
    boundBone {boneName=dyn_hair_01_r}
    componentAppliedLinearAccClamp=<2200 2200 2200>
    componentLinearAccScale=<0.9 0.9 0.9>
    componentLinearVelScale=<0.9 0.9 0.9>
    gravityOverride {z=-980}
    physicsBodyDefinitions [{
      boundBone {boneName=dyn_hair_01_r}
      boxExtents=<30 30 30>
      constraintSetup {angularLimitsMax=<0 20 9>, angularLimitsMin=<0 0 -9>, angularTarget=<1 0 0>}
      localJointOffset=<0.8 0 -0.84>
     }]
    relativeSpaceBone {boneName=head}
   }
   n10 = Anim Dynamics - Bone: dyn_hair_02_r(ComponentPose=@n9, Alpha=1, ExternalForce="0,0,0") {
    angularDampingOverride=1
    angularSpringConstant=3900
    bAngularSpring=true
    bOverrideAngularDamping=true
    bUseGravityOverride=true
    bUseSphericalLimits=true
    gravityScale=0.35
    simulationSpace=RootRelative
    boundBone {boneName=dyn_hair_02_r}
    componentAppliedLinearAccClamp=<2200 2200 2200>
    componentLinearAccScale=<1 0.8 0.9>
    componentLinearVelScale=<0.9882 0.91 0.9>
    gravityOverride {z=-980}
    physicsBodyDefinitions [{
      collisionType=CustomSphere
      sphereCollisionRadius=6
      boundBone {boneName=dyn_hair_02_r}
      boxExtents=<29 29 29>
      constraintSetup {angularLimitsMax=<0 20 9>, angularLimitsMin=<0 0 -9>, angularTarget=<1 0 0>}
      localJointOffset=<0.9 0 -0.9>
     }]
    relativeSpaceBone {boneName=head}
    sphericalLimits [{limitRadius=2, drivingBone {boneName=head}, sphereLocalOffset=<0 0 0>}]
   }
   n11 = Anim Dynamics - Bone: dyn_hair_01_c(ComponentPose=@n10, Alpha=1, ExternalForce="0,0,0") {
    angularDampingOverride=1
    angularSpringConstant=4002
    bAngularSpring=true
    bOverrideAngularDamping=true
    bUseGravityOverride=true
    gravityScale=0.35
    simulationSpace=RootRelative
    boundBone {boneName=dyn_hair_01_c}
    componentAppliedLinearAccClamp=<2200 2200 2200>
    componentLinearAccScale=<0.95 0.9 0.95>
    componentLinearVelScale=<0.95 0.9 0.95>
    gravityOverride {z=-980}
    physicsBodyDefinitions [{
      boundBone {boneName=dyn_hair_01_c}
      boxExtents=<34 34 34>
      constraintSetup {angularLimitsMax=<0 5 2>, angularLimitsMin=<0 -5 -20>, angularTarget=<1 0 0>}
      localJointOffset=<0.94 0.85 0>
     }]
    relativeSpaceBone {boneName=head}
   }
   n12 = Anim Dynamics - Bone: dyn_hair_02_c(ComponentPose=@n11, Alpha=1, ExternalForce="0,0,0") {
    angularDampingOverride=1
    angularSpringConstant=4000
    bAngularSpring=true
    bOverrideAngularDamping=true
    bUseGravityOverride=true
    bUseSphericalLimits=true
    gravityScale=0.35
    simulationSpace=RootRelative
    boundBone {boneName=dyn_hair_02_c}
    componentAppliedLinearAccClamp=<2200 2200 2200>
    componentLinearAccScale=<0.9 0.9 0.9>
    componentLinearVelScale=<0.9 0.9 0.9>
    gravityOverride {z=-980}
    physicsBodyDefinitions [{
      collisionType=CustomSphere
      sphereCollisionRadius=9
      boundBone {boneName=dyn_hair_02_c}
      boxExtents=<34 34 34>
      constraintSetup {angularLimitsMax=<0 5 2>, angularLimitsMin=<0 -5 -20>, angularTarget=<1 0 0>}
      localJointOffset=<1 0.99 0>
     }]
    relativeSpaceBone {boneName=head}
    sphericalLimits [{limitRadius=12, drivingBone {boneName=spine_05}, sphereLocalOffset=<-5.1288 -6.0238 5.3548>}]
   }
   n13 = Control Rig!(
    Source=Blend Poses by bool(
     BlendPose_0=Component To Local(ComponentPose=@n6)
     BlendPose_1=Component To Local(ComponentPose=@n12)
     bActiveValue=Get Cape Tag() -> (Cape Tag:bool)
    ) {blendType=Linear}
    Alpha
    PistolADS:double=Get Is Pistol Alpha() -> (Is Pistol Alpha:float)
    cape:bool=Get Cape Tag()
    crouching:bool=Get bIsCrouching()
   ) -> (Pose) {controlRigAssetReference {blueprintRigClass=&/M_MED_CurbRestore_FaceAcc_AfterDynamics_CtrlRig.M_MED_CurbRestore_FaceAcc_AfterDynamics_CtrlRig_C}}
   Output Pose(Result=@n13) {layerGroup=None}
   n14 = Control Rig!(
    Source=Copy Pose From Mesh() {bUseAttachedParent=true}
    Alpha
    PistolADS=Get Is Pistol Alpha()
    IsTargeting Alpha:double=Get Is Targeting Alpha() -> (Is Targeting Alpha:double)
    cape=Get Cape Tag()
    Crouching:bool=Get bIsCrouching()
   ) -> (Pose) {controlRigAssetReference {blueprintRigClass=&/M_MED_CurbRestore_FaceAcc_CtrlRig.M_MED_CurbRestore_FaceAcc_CtrlRig_C}}
   Save cached pose 'StartPose'(Pose=@n14) {cacheName=StartPose}
  ]}
 EventGraph {nodes [
   Event BlueprintUpdateAnimation()
   > n1 = Sequence()
   > n2 = Cast To FortPlayerAnimInstance(Object=GetAnimInstance(self=GetOwnerSkeletalMeshComponent()))
   > Set isPistol(isPistol=false)
   > Set isRifle(isRifle:bool=false) -> (Output_Get:bool)
   > n3 = Switch on EFortWeaponCoreAnimation!(execute, Selection:EFortWeaponCoreAnimation Enum=Get WeaponCoreAnim(self=@n2)) -> (
    Melee:Exec
    Pistol:Exec
    Shotgun:Exec
    PaperBlueprint:Exec
    Rifle:Exec
    MeleeOneHand:Exec
    MachinePistol:Exec
    RocketLauncher:Exec
    GrenadeLauncher:Exec
    GoingCommando:Exec
    AssaultRifle:Exec
    TacticalShotgun:Exec
    SniperRifle:Exec
    TrapPlacement:Exec
    ShoulderLauncher:Exec
    AbilityDecoTool:Exec
    Crossbow:Exec
    C4:Exec
    RemoteControl:Exec
    DualWield:Exec
    AR_BullPup:Exec
    AR_ForwardGrip:Exec
    MedPackPaddles:Exec
    SMG_P90:Exec
    AR_DrumGun:Exec
    Consumable_Small:Exec
    Consumable_Large:Exec
    Balloon:Exec
    MountedTurret:Exec
    CreativeTool:Exec
    ExplosiveBow:Exec
    AshtonIndigo:Exec
    AshtonChicago:Exec
    MeleeDualWield:Exec
    Unarmed:Exec
   )
   Set isRifle(execute=[@n3.AssaultRifle, @n3.Rifle, @n3.RocketLauncher, @n3.Shotgun, @n3.SniperRifle, @n3.TacticalShotgun], isRifle=true)
   Set isPistol(execute=@n3.Pistol, isPistol=true)
   n4 = OR Boolean(A=Get isPistol() -> (isPistol:bool), B=Get isRifle() -> (isRifle:bool))
   n5 = Get bIsTargeting()
   Set Is Pistol Alpha(execute=@n1.then_1, Is Pistol Alpha:float="To Float (Boolean)"(InBool=AND Boolean(A=@n4, B=@n5))) -> (Output_Get:float)
   > Set Is Targeting Alpha(Is Targeting Alpha:double="To Float (Boolean)"(InBool=@n5)) -> (Output_Get:double)
   Event OnCharacterCustomizationLoaded()
   n6 = CharacterItemDefinitionHasAnyMetaTag(Tags=Get Tags() -> (Tags:Array of Gameplay Tag Structures))
   > Set Cape Tag(Cape Tag:bool=@n6) -> (Output_Get:bool)
   > Set Cape Tag Inverted(Cape Tag Inverted:bool=NOT Boolean(A=@n6)) -> (Output_Get:bool)
  ]}
]
variables [
 Cape Tag {type=boolean}
 Cape Tag Inverted {type=boolean}
 Is Pistol Alpha {type=number}
 Is Targeting Alpha {type=number}
 isPistol {type=boolean}
 isRifle {type=boolean}
 Tags {type="array<GameplayTag>", default [{tagName=Cosmetics.Character.RestrictCape}]}
]
